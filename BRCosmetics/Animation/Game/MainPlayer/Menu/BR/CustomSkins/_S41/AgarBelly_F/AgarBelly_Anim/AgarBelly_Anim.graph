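# Level Sequence that authors the menu animation for the AgarBelly_F custom skin in Battle Royale.
# A spawned FortPlayerMannequin stays at the origin, with an empty skeletal animation track, while a Fortnite_Biped_CtrlRig parameter track drives the motion.
# Legs and arms are switched to IK and the spine and neck to FK.
# Keys span roughly frames -1 to 389, with foot roll, toe tip, pole vector and leg softness adjustments around planted feet.
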LevelSequence /BRCosmetics/Animation/Game/MainPlayer/Menu/BR/CustomSkins/_S41/AgarBelly_F/AgarBelly_Anim/AgarBelly_Anim
bindings [FortPlayerMannequin {
  id=#1
  possessed_class=FortPlayerMannequin
  tracks [
   {type=MovieSceneSpawnTrack, sections [{type=MovieSceneSpawnSection, channels [None {type=MovieSceneScriptingBoolChannel, default=true}]}]}
   {
    type=MovieScene3DTransformTrack
    sections [{
      type=MovieScene3DTransformSection
      channels [
       Location.X {type=MovieSceneScriptingDoubleChannel, default=0, const [0,4]}
       Location.Y {type=MovieSceneScriptingDoubleChannel, default=0, const [0,4]}
       Location.Z {type=MovieSceneScriptingDoubleChannel, default=0, const [0,4]}
       Rotation.X {type=MovieSceneScriptingDoubleChannel, default=0, const [0,4]}
       Rotation.Y {type=MovieSceneScriptingDoubleChannel, default=0, const [0,4]}
       Rotation.Z {type=MovieSceneScriptingDoubleChannel, default=0, const [0,4]}
       Scale.X {type=MovieSceneScriptingDoubleChannel, default=1, const [1,4]}
       Scale.Y {type=MovieSceneScriptingDoubleChannel, default=1, const [1,4]}
       Scale.Z {type=MovieSceneScriptingDoubleChannel, default=1, const [1,4]}
      ]
     }]
   }
   {type=MovieSceneSkeletalAnimationTrack, sections []}
   {
    type=MovieSceneControlRigParameterTrack
    display_name=Fortnite_Biped_CtrlRig
    sections [{
      type=MovieSceneControlRigParameterSection
      channels [
       leg_l_stretch_switch {type=MovieSceneScriptingBoolChannel, default=false, const [true,29]}
       leg_r_stretch_switch {type=MovieSceneScriptingBoolChannel, default=false, const [true,29]}
       arm_l_fk_local {type=MovieSceneScriptingBoolChannel, default=true, const [true,36]}
       Auto Interpolate_3 {type=MovieSceneScriptingBoolChannel, default=false, const [false,31]}
       Auto Interpolate {type=MovieSceneScriptingBoolChannel, default=false, const [false,29]}
       arm_l_stretch_switch {type=MovieSceneScriptingBoolChannel, default=false, const [false,28]}
       Keep Aligned {type=MovieSceneScriptingBoolChannel, default=false, const [false,28]}
       arm_r_fk_local {type=MovieSceneScriptingBoolChannel, default=true, const [true,11]}
       Auto Interpolate_4 {type=MovieSceneScriptingBoolChannel, default=false, const [false,8]}
       Auto Interpolate_2 {type=MovieSceneScriptingBoolChannel, default=false, const [false,8]}
       arm_r_stretch_switch {type=MovieSceneScriptingBoolChannel, default=false, const [true,27]}
       Keep Aligned_2 {type=MovieSceneScriptingBoolChannel, default=false, const [false,27]}
       head_local {type=MovieSceneScriptingBoolChannel, default=true, const [true,37]}
       Stretch {type=MovieSceneScriptingBoolChannel, default=false, const [true,37]}
       head_stretch {type=MovieSceneScriptingBoolChannel, default=false, const [false,29]}
       spine_stretch_switch {type=MovieSceneScriptingBoolChannel, default=false, const [false,30]}
       arm_l_fk_ik_switch {type=MovieSceneScriptingBoolChannel, default=false, const [true,34]}
       leg_l_fk_ik_switch {type=MovieSceneScriptingBoolChannel, default=true, const [true,29]}
       arm_r_fk_ik_switch {type=MovieSceneScriptingBoolChannel, default=false, const [true,33]}
       leg_r_fk_ik_switch {type=MovieSceneScriptingBoolChannel, default=true, const [true,29]}
       ShowBodyControls {type=MovieSceneScriptingBoolChannel, default=true, const [true,29]}
       spine_fk_ik_switch {type=MovieSceneScriptingBoolChannel, default=false, const [false,29]}
       neck_fk_ik_switch {type=MovieSceneScriptingBoolChannel, default=false, const [false,29]}
       space {type=MovieSceneScriptingByteChannel, default=1, const [1,37]}
       global_ctrl.Location.X {type=MovieSceneScriptingFloatChannel, default=0, const [0,29]}
       global_ctrl.Location.Y {type=MovieSceneScriptingFloatChannel, default=0, const [0,29]}
       global_ctrl.Location.Z {type=MovieSceneScriptingFloatChannel, default=0, const [0,29]}
       global_ctrl.Rotation.X {type=MovieSceneScriptingFloatChannel, default=0, const [0,29]}
       global_ctrl.Rotation.Y {type=MovieSceneScriptingFloatChannel, default=0, const [0,29]}
       global_ctrl.Rotation.Z {type=MovieSceneScriptingFloatChannel, default=0, const [0,29]}
       root_ctrl.Location.X {type=MovieSceneScriptingFloatChannel, default=0, const [0,29]}
       root_ctrl.Location.Y {type=MovieSceneScriptingFloatChannel, default=0, const [0,29]}
       root_ctrl.Location.Z {type=MovieSceneScriptingFloatChannel, default=0, const [0,29]}
       root_ctrl.Rotation.X {type=MovieSceneScriptingFloatChannel, default=0, const [0,29]}
       root_ctrl.Rotation.Y {type=MovieSceneScriptingFloatChannel, default=0, const [0,29]}
       root_ctrl.Rotation.Z {type=MovieSceneScriptingFloatChannel, default=0, const [0,29]}
       root_ctrl.Scale.X {type=MovieSceneScriptingFloatChannel, default=1, const [1,29]}
       root_ctrl.Scale.Y {type=MovieSceneScriptingFloatChannel, default=1, const [1,29]}
       root_ctrl.Scale.Z {type=MovieSceneScriptingFloatChannel, default=1, const [1,29]}
       body_offset_ctrl.Location.X {type=MovieSceneScriptingFloatChannel, default=0, const [0,29]}
       body_offset_ctrl.Location.Y {type=MovieSceneScriptingFloatChannel, default=0, const [0,29]}
       body_offset_ctrl.Location.Z {type=MovieSceneScriptingFloatChannel, default=0, const [0,29]}
       body_offset_ctrl.Rotation.X {type=MovieSceneScriptingFloatChannel, default=0, const [0,29]}
       body_offset_ctrl.Rotation.Y {type=MovieSceneScriptingFloatChannel, default=0, const [0,29]}
       body_offset_ctrl.Rotation.Z {type=MovieSceneScriptingFloatChannel, default=0, const [0,29]}
       body_offset_ctrl.Scale.X {type=MovieSceneScriptingFloatChannel, default=1, const [1,29]}
       body_offset_ctrl.Scale.Y {type=MovieSceneScriptingFloatChannel, default=1, const [1,29]}
       body_offset_ctrl.Scale.Z {type=MovieSceneScriptingFloatChannel, default=1, const [1,29]}
       thigh_l_fk_ctrl.Location.X {type=MovieSceneScriptingFloatChannel, default=0, thinned_from=29, keys [[-1,0],[389,0]]}
       thigh_l_fk_ctrl.Location.Y {type=MovieSceneScriptingFloatChannel, default=0, thinned_from=29, keys [[-1,0],[389,0]]}
       thigh_l_fk_ctrl.Location.Z {type=MovieSceneScriptingFloatChannel, default=0, thinned_from=29, keys [[-1,0],[389,0]]}
       thigh_l_fk_ctrl.Rotation.X {
        type=MovieSceneScriptingFloatChannel
        default=1.0553
        keys [[-1,-6.6354],[0,-7.9426],[21,-7.8364],[46,-7.0191],[60,-7.0191],[78,-6.4162],[84,-7.0191],[92,-6.6492],[112,-6.8355],[119,-6.8627],[135,-6.8731],[139,-7.0694],[146,-7.0537],[160,-7.5123],[172,-7.6874],[178,-14.3754],[183,-10.8061],[190,-8.6748],[210,-7.8069],[217,-5.6424],[234,-3.6654],[253,-3.6392],[267,-3.6392],[294,-5.7479],[310,-7.9426],[331,-5.1392],[357,-7.0191],[376,-5.7735],[389,-5.7771]]
       }
       thigh_l_fk_ctrl.Rotation.Y {
        type=MovieSceneScriptingFloatChannel
        default=-1.0258
        keys [[-1,-0.3409],[0,1.9447],[21,2.4619],[46,1.9669],[60,1.9669],[78,3.0931],[84,1.9669],[92,2.3897],[112,2.6456],[119,1.9872],[135,1.7358],[139,1.6411],[146,0.9136],[160,0.2939],[172,0.0572],[178,3.269],[183,-0.8871],[190,0.2088],[210,2.6412],[217,3.1419],[234,6.0775],[253,6.1166],[267,6.1166],[294,2.541],[310,1.9447],[331,0.5949],[357,1.9669],[376,-0.2525],[389,-0.6516]]
       }
       thigh_l_fk_ctrl.Rotation.Z {
        type=MovieSceneScriptingFloatChannel
        default=-0.0727
        thinned_from=29
        keys [[-1,-42.346],[21,-42.0076],[46,-42.1305],[84,-42.1305],[92,-42.6108],[112,-42.3333],[135,-42.3691],[146,-42.6543],[160,-43.142],[172,-43.3282],[178,-33.3588],[183,-41.1287],[190,-49.5569],[210,-39.5079],[217,-49.1573],[234,-48.8471],[294,-48.8782],[310,-42.3106],[331,-48.7016],[357,-42.1305],[376,-48.4933],[389,-48.4383]]
       }
       thigh_l_fk_ctrl.Scale.X {type=MovieSceneScriptingFloatChannel, default=1, const [1,29]}
       thigh_l_fk_ctrl.Scale.Y {type=MovieSceneScriptingFloatChannel, default=1, const [1,29]}
       thigh_l_fk_ctrl.Scale.Z {type=MovieSceneScriptingFloatChannel, default=1, const [1,29]}
       calf_l_fk_ctrl.Location.X {type=MovieSceneScriptingFloatChannel, default=0, thinned_from=29, keys [[-1,0],[190,0],[210,0.0002],[217,0],[389,0]]}
       calf_l_fk_ctrl.Location.Y {type=MovieSceneScriptingFloatChannel, default=0, thinned_from=29, keys [[-1,0],[190,0],[210,0.0006],[217,0],[389,0]]}
       calf_l_fk_ctrl.Location.Z {type=MovieSceneScriptingFloatChannel, default=0, thinned_from=29, keys [[-1,0],[190,0],[210,-0.0045],[217,0],[389,0]]}
       calf_l_fk_ctrl.Rotation.X {
        type=MovieSceneScriptingFloatChannel
        default=0.0724
        thinned_from=29
        keys [[-1,0.8662],[0,0.8561],[21,0.8579],[46,0.8663],[60,0.8663],[78,0.8698],[84,0.8663],[112,0.863],[119,0.861],[135,0.8602],[139,0.8626],[172,0.857],[183,0.877],[210,0.8919],[217,0.8909],[234,0.8979],[267,0.8979],[294,0.886],[310,0.8561],[331,0.8655],[357,0.8663],[376,0.8705],[389,0.8683]]
       }
       calf_l_fk_ctrl.Rotation.Y {
        type=MovieSceneScriptingFloatChannel
        default=1.0549
        thinned_from=29
        keys [[-1,0.6065],[0,0.6206],[21,0.6182],[46,0.6063],[60,0.6063],[78,0.6013],[84,0.6063],[112,0.6109],[119,0.6137],[135,0.6147],[139,0.6115],[146,0.6144],[172,0.6191],[183,0.5907],[190,0.5838],[210,0.5799],[217,0.5696],[234,0.5585],[267,0.5585],[294,0.5771],[310,0.6206],[331,0.6074],[357,0.6063],[376,0.6002],[389,0.6035]]
       }
       calf_l_fk_ctrl.Rotation.Z {
        type=MovieSceneScriptingFloatChannel
        default=0.2163
        thinned_from=29
        keys [[-1,51.2937],[0,50.3513],[21,50.5142],[46,51.3024],[60,51.3024],[78,51.6368],[84,51.3024],[112,50.9976],[119,50.8125],[135,50.7419],[139,50.9602],[146,50.764],[172,50.4458],[183,52.3322],[190,52.7789],[210,53.5415],[217,53.6995],[234,54.4078],[267,54.4078],[294,53.2133],[310,50.3513],[331,51.228],[357,51.3024],[376,51.7076],[389,51.4925]]
       }
       calf_l_fk_ctrl.Scale.X {type=MovieSceneScriptingFloatChannel, default=1, const [1,29]}
       calf_l_fk_ctrl.Scale.Y {type=MovieSceneScriptingFloatChannel, default=1, const [1,29]}
       calf_l_fk_ctrl.Scale.Z {type=MovieSceneScriptingFloatChannel, default=1, const [1,29]}
       foot_l_fk_ctrl.Location.X {type=MovieSceneScriptingFloatChannel, default=0, thinned_from=29, keys [[-1,0],[389,0]]}
       foot_l_fk_ctrl.Location.Y {
        type=MovieSceneScriptingFloatChannel
        default=0.0627
        thinned_from=29
        keys [[-1,0.0627],[21,0.0627],[46,0.0659],[178,0.0659],[183,0.0627],[234,0.0627],[253,0.0553],[267,0.0553],[294,0.0627],[331,0.0627],[357,0.0659],[376,0.0627],[389,0.0627]]
       }
       foot_l_fk_ctrl.Location.Z {type=MovieSceneScriptingFloatChannel, default=-0.8453, thinned_from=29, keys [[-1,-0.8453],[389,-0.8453]]}
       foot_l_fk_ctrl.Rotation.X {
        type=MovieSceneScriptingFloatChannel
        default=-0.9801
        thinned_from=29
        keys [[-1,-2.2152],[0,-8.9844],[21,-7.5327],[46,-6.0098],[60,-6.0098],[78,-5.9871],[84,-6.0098],[135,-6.0235],[146,-6.0307],[160,-6.0638],[172,-6.0693],[178,-8.9758],[183,-6.1086],[190,-2.3524],[210,-5.9019],[217,-2.1934],[234,-2.1366],[267,-2.1359],[294,-2.2201],[310,-8.9844],[331,-6.0752],[357,-6.0098],[376,-2.2927],[389,-2.3031]]
       }
       foot_l_fk_ctrl.Rotation.Y {
        type=MovieSceneScriptingFloatChannel
        default=0.1001
        thinned_from=29
        keys [[-1,-7.1349],[0,-7.7307],[21,-8.0593],[46,-7.5258],[60,-7.5258],[78,-7.4915],[84,-7.5258],[92,-7.6232],[112,-7.9184],[119,-7.8618],[135,-7.8401],[139,-7.9186],[146,-7.5839],[160,-7.4247],[172,-7.6433],[178,-10.1715],[183,-7.2234],[190,-5.9445],[210,-5.7861],[217,-5.3516],[234,-5.1621],[267,-5.1597],[294,-5.7074],[310,-7.7307],[331,-6.3123],[357,-7.5258],[376,-6.3131],[389,-6.3996]]
       }
       foot_l_fk_ctrl.Rotation.Z {
        type=MovieSceneScriptingFloatChannel
        default=-0.1441
        thinned_from=29
        keys [[-1,-20.169],[0,-20.5827],[21,-20.7565],[60,-20.7639],[78,-21.0146],[84,-20.7639],[92,-20.7546],[112,-20.8532],[119,-20.7143],[135,-20.6612],[139,-20.7783],[146,-20.4456],[160,-20.0654],[172,-20.1406],[178,-22.7101],[183,-20.5553],[190,-19.2495],[210,-21.1115],[217,-20.0687],[234,-20.5699],[267,-20.5672],[294,-19.9939],[310,-20.5827],[331,-19.5722],[357,-20.7639],[376,-19.3769],[389,-19.2888]]
       }
       foot_l_fk_ctrl.Scale.X {type=MovieSceneScriptingFloatChannel, default=1, const [1,29]}
       foot_l_fk_ctrl.Scale.Y {type=MovieSceneScriptingFloatChannel, default=1, const [1,29]}
       foot_l_fk_ctrl.Scale.Z {type=MovieSceneScriptingFloatChannel, default=1, const [1,29]}
       ball_l_fk_ctrl.Location.X {
        type=MovieSceneScriptingFloatChannel
        default=0
        thinned_from=29
        keys [[-1,0],[160,0],[172,1.1663],[178,1.1663],[183,0],[210,0.0002],[217,-1.1663],[294,-1.1663],[310,0],[331,-1.1663],[357,0],[376,-1.1663],[389,-1.1663]]
       }
       ball_l_fk_ctrl.Location.Y {
        type=MovieSceneScriptingFloatChannel
        default=0
        thinned_from=29
        keys [[-1,0],[160,0],[172,-7.6543],[178,-7.6543],[183,0],[210,0.0001],[217,7.6543],[294,7.6543],[310,0],[331,7.6543],[357,0],[376,7.6543],[389,7.6543]]
       }
       ball_l_fk_ctrl.Location.Z {
        type=MovieSceneScriptingFloatChannel
        default=0
        thinned_from=29
        keys [[-1,0],[160,0],[172,3.6092],[178,3.6092],[183,0],[210,0.0001],[217,-3.6092],[294,-3.6092],[310,0],[331,-3.6092],[357,0],[376,-3.6092],[389,-3.6092]]
       }
       ball_l_fk_ctrl.Rotation.X {type=MovieSceneScriptingFloatChannel, default=0, thinned_from=29, keys [[-1,0],[190,0],[210,-0.0002],[217,0],[389,0]]}
       ball_l_fk_ctrl.Rotation.Y {type=MovieSceneScriptingFloatChannel, default=0, thinned_from=29, keys [[-1,0],[190,0],[210,-0.0009],[217,0],[389,0]]}
       ball_l_fk_ctrl.Rotation.Z {type=MovieSceneScriptingFloatChannel, default=0, thinned_from=29, keys [[-1,0],[190,0],[210,-0.001],[217,0],[389,0]]}
       ball_l_fk_ctrl.Scale.X {type=MovieSceneScriptingFloatChannel, default=1, const [1,29]}
       ball_l_fk_ctrl.Scale.Y {type=MovieSceneScriptingFloatChannel, default=1, const [1,29]}
       ball_l_fk_ctrl.Scale.Z {type=MovieSceneScriptingFloatChannel, default=1, const [1,29]}
       foot_l_ik_ctrl.Location.X {type=MovieSceneScriptingFloatChannel, default=0, const [6.0009,29]}
       foot_l_ik_ctrl.Location.Y {type=MovieSceneScriptingFloatChannel, default=0, const [19.2466,29]}
       foot_l_ik_ctrl.Location.Z {type=MovieSceneScriptingFloatChannel, default=0, const [0,29]}
       foot_l_ik_ctrl.Rotation.X {type=MovieSceneScriptingFloatChannel, default=0, const [0,29]}
       foot_l_ik_ctrl.Rotation.Y {type=MovieSceneScriptingFloatChannel, default=0, const [0,29]}
       foot_l_ik_ctrl.Rotation.Z {type=MovieSceneScriptingFloatChannel, default=0, const [34.1832,29]}
       foot_l_ik_ctrl.Scale.X {type=MovieSceneScriptingFloatChannel, default=1, const [1,29]}
       foot_l_ik_ctrl.Scale.Y {type=MovieSceneScriptingFloatChannel, default=1, const [1,29]}
       foot_l_ik_ctrl.Scale.Z {type=MovieSceneScriptingFloatChannel, default=1, const [1,29]}
       ball_l_ik_ctrl.Location.X {type=MovieSceneScriptingFloatChannel, default=0, thinned_from=30, keys [[-1,0],[389,0]]}
       ball_l_ik_ctrl.Location.Y {type=MovieSceneScriptingFloatChannel, default=0, thinned_from=30, keys [[-1,0],[389,0]]}
       ball_l_ik_ctrl.Location.Z {type=MovieSceneScriptingFloatChannel, default=0, thinned_from=30, keys [[-1,0],[389,0]]}
       ball_l_ik_ctrl.Rotation.X {type=MovieSceneScriptingFloatChannel, default=0, const [0,29]}
       ball_l_ik_ctrl.Rotation.Y {type=MovieSceneScriptingFloatChannel, default=0, thinned_from=29, keys [[-1,0],[389,0]]}
       ball_l_ik_ctrl.Rotation.Z {type=MovieSceneScriptingFloatChannel, default=0, const [0,29]}
       ball_l_ik_ctrl.Scale.X {type=MovieSceneScriptingFloatChannel, default=1, const [1,29]}
       ball_l_ik_ctrl.Scale.Y {type=MovieSceneScriptingFloatChannel, default=1, const [1,29]}
       ball_l_ik_ctrl.Scale.Z {type=MovieSceneScriptingFloatChannel, default=1, const [1,29]}
       foot_bk1_l_ctrl.Location.X {type=MovieSceneScriptingFloatChannel, default=0, const [0,29]}
       foot_bk1_l_ctrl.Location.Y {type=MovieSceneScriptingFloatChannel, default=0, const [0,29]}
       foot_bk1_l_ctrl.Location.Z {type=MovieSceneScriptingFloatChannel, default=0, const [0,29]}
       foot_bk1_l_ctrl.Rotation.X {type=MovieSceneScriptingFloatChannel, default=0, const [0,29]}
       foot_bk1_l_ctrl.Rotation.Y {type=MovieSceneScriptingFloatChannel, default=0, const [0,29]}
       foot_bk1_l_ctrl.Rotation.Z {type=MovieSceneScriptingFloatChannel, default=0, const [0,29]}
       foot_bk1_l_ctrl.Scale.X {type=MovieSceneScriptingFloatChannel, default=1, const [1,29]}
       foot_bk1_l_ctrl.Scale.Y {type=MovieSceneScriptingFloatChannel, default=1, const [1,29]}
       foot_bk1_l_ctrl.Scale.Z {type=MovieSceneScriptingFloatChannel, default=1, const [1,29]}
       foot_roll_l_ctrl.Location.X {type=MovieSceneScriptingFloatChannel, default=0, const [0,29]}
       foot_roll_l_ctrl.Location.Y {type=MovieSceneScriptingFloatChannel, default=0, const [0,29]}
       foot_roll_l_ctrl.Location.Z {type=MovieSceneScriptingFloatChannel, default=0, const [0,29]}
       foot_roll_l_ctrl.Rotation.X {type=MovieSceneScriptingFloatChannel, default=0, const [0,29]}
       foot_roll_l_ctrl.Rotation.Y {type=MovieSceneScriptingFloatChannel, default=0, const [0,29]}
       foot_roll_l_ctrl.Rotation.Z {type=MovieSceneScriptingFloatChannel, default=0, thinned_from=30, keys [[-1,0],[172,0],[176,-1.3437],[178,0],[389,0]]}
       foot_roll_l_ctrl.Scale.X {type=MovieSceneScriptingFloatChannel, default=1, const [1,29]}
       foot_roll_l_ctrl.Scale.Y {type=MovieSceneScriptingFloatChannel, default=1, const [1,29]}
       foot_roll_l_ctrl.Scale.Z {type=MovieSceneScriptingFloatChannel, default=1, const [1,29]}
       roll_blend_l_ctrl {type=MovieSceneScriptingFloatChannel, default=1, const [1,29]}
       tip_l_ctrl.Location.X {type=MovieSceneScriptingFloatChannel, default=0, const [0,29]}
       tip_l_ctrl.Location.Y {type=MovieSceneScriptingFloatChannel, default=0, const [0,29]}
       tip_l_ctrl.Location.Z {type=MovieSceneScriptingFloatChannel, default=0, const [0,29]}
       tip_l_ctrl.Rotation.X {type=MovieSceneScriptingFloatChannel, default=0, thinned_from=29, keys [[-1,0],[389,0]]}
       tip_l_ctrl.Rotation.Y {
        type=MovieSceneScriptingFloatChannel
        default=0
        thinned_from=29
        keys [[-1,0],[0,9.009],[21,7.073],[135,7.073],[139,6.7439],[146,5.4371],[160,5.061],[172,5.6936],[178,8.8481],[183,9.009],[310,9.009],[331,7.073],[357,7.073],[376,0],[389,0]]
       }
       tip_l_ctrl.Rotation.Z {type=MovieSceneScriptingFloatChannel, default=0, thinned_from=29, keys [[-1,0],[389,0]]}
       tip_l_ctrl.Scale.X {type=MovieSceneScriptingFloatChannel, default=1, const [1,29]}
       tip_l_ctrl.Scale.Y {type=MovieSceneScriptingFloatChannel, default=1, const [1,29]}
       tip_l_ctrl.Scale.Z {type=MovieSceneScriptingFloatChannel, default=1, const [1,29]}
       heel_l_ctrl.Location.X {type=MovieSceneScriptingFloatChannel, default=0, const [0,29]}
       heel_l_ctrl.Location.Y {type=MovieSceneScriptingFloatChannel, default=0, const [0,29]}
       heel_l_ctrl.Location.Z {type=MovieSceneScriptingFloatChannel, default=0, const [0,29]}
       heel_l_ctrl.Rotation.X {type=MovieSceneScriptingFloatChannel, default=0, const [0,29]}
       heel_l_ctrl.Rotation.Y {type=MovieSceneScriptingFloatChannel, default=0, const [0,29]}
       heel_l_ctrl.Rotation.Z {type=MovieSceneScriptingFloatChannel, default=0, const [0,29]}
       heel_l_ctrl.Scale.X {type=MovieSceneScriptingFloatChannel, default=1, const [1,29]}
       heel_l_ctrl.Scale.Y {type=MovieSceneScriptingFloatChannel, default=1, const [1,29]}
       heel_l_ctrl.Scale.Z {type=MovieSceneScriptingFloatChannel, default=1, const [1,29]}
       leg_l_softness {type=MovieSceneScriptingFloatChannel, default=0, thinned_from=30, keys [[-1,0.2588],[0,1],[178,1],[183,0.2588],[253,0.2588],[266,1],[389,1]]}
       leg_l_pv_ik_ctrl.Location.X {
        type=MovieSceneScriptingFloatChannel
        default=0
        thinned_from=29
        keys [[-1,0],[0,0],[11,-0.3432],[21,-2.6128],[46,-3.6603],[60,-3.7484],[78,-3.5707],[84,-3.5319],[160,-3.5319],[172,-3.0904],[183,0],[210,0],[217,-0.1081],[234,-1.5403],[253,-2.0537],[267,-2.0537],[294,-0.7702],[310,0],[389,0]]
       }
       leg_l_pv_ik_ctrl.Location.Y {type=MovieSceneScriptingFloatChannel, default=0, thinned_from=29, keys [[-1,0],[389,0]]}
       leg_l_pv_ik_ctrl.Location.Z {type=MovieSceneScriptingFloatChannel, default=0, thinned_from=29, keys [[-1,0],[389,0]]}
       leg_l_pv_ik_ctrl.Rotation.X {type=MovieSceneScriptingFloatChannel, default=0, const [0,11]}
       leg_l_pv_ik_ctrl.Rotation.Y {type=MovieSceneScriptingFloatChannel, default=0, thinned_from=11, keys [[-1,0],[389,0]]}
       leg_l_pv_ik_ctrl.Rotation.Z {type=MovieSceneScriptingFloatChannel, default=0, const [0,11]}
       leg_l_pv_ik_ctrl.Scale.X {type=MovieSceneScriptingFloatChannel, default=1, const [1,29]}
       leg_l_pv_ik_ctrl.Scale.Y {type=MovieSceneScriptingFloatChannel, default=1, const [1,29]}
       leg_l_pv_ik_ctrl.Scale.Z {type=MovieSceneScriptingFloatChannel, default=1, const [1,29]}
       thigh_r_fk_ctrl.Location.X {type=MovieSceneScriptingFloatChannel, default=0, thinned_from=29, keys [[-1,0],[389,0]]}
       thigh_r_fk_ctrl.Location.Y {type=MovieSceneScriptingFloatChannel, default=0, thinned_from=29, keys [[-1,0],[389,0]]}
       thigh_r_fk_ctrl.Location.Z {type=MovieSceneScriptingFloatChannel, default=0, thinned_from=29, keys [[-1,0],[389,0]]}
       thigh_r_fk_ctrl.Rotation.X {
        type=MovieSceneScriptingFloatChannel
        default=1.0528
        thinned_from=29
        keys [[-1,-21.565],[0,-23.6333],[21,-23.2055],[46,-23.3652],[60,-23.3652],[78,-24.2132],[84,-23.3652],[92,-23.723],[112,-23.5823],[119,-23.4182],[135,-23.3556],[139,-23.1568],[146,-23.0038],[160,-22.1868],[172,-22.0815],[178,-17.6984],[183,-19.2952],[190,-20.1736],[210,-24.6665],[217,-23.9535],[234,-26.622],[267,-26.657],[294,-23.6586],[310,-23.6333],[331,-25.2978],[357,-23.3652],[376,-22.8746],[389,-22.7626]]
       }
       thigh_r_fk_ctrl.Rotation.Y {
        type=MovieSceneScriptingFloatChannel
        default=-1.026
        keys [[-1,1.8988],[0,1.2278],[21,0.3072],[46,0.5488],[60,0.5488],[78,-0.4051],[84,0.5488],[92,0.2724],[112,-0.0511],[119,0.5824],[135,0.8242],[139,0.8297],[146,1.6161],[160,2.6261],[172,2.3842],[178,-1.8787],[183,1.8808],[190,0.9366],[210,-1.2761],[217,-1.0101],[234,-3.2212],[253,-3.2504],[267,-3.2504],[294,-0.5485],[310,1.2278],[331,2.3021],[357,0.5488],[376,1.9645],[389,2.3235]]
       }
       thigh_r_fk_ctrl.Rotation.Z {
        type=MovieSceneScriptingFloatChannel
        default=-0.07
        thinned_from=29
        keys [[-1,-29.5491],[0,-30.1921],[21,-30.1456],[46,-29.9303],[60,-29.9303],[78,-30.1378],[84,-29.9303],[92,-30.7158],[135,-30.6301],[139,-30.6927],[146,-30.5511],[160,-30.9953],[172,-30.9604],[178,-20.5543],[183,-26.6685],[190,-33.8525],[210,-24.7485],[217,-34.3377],[234,-34.7813],[267,-34.7871],[294,-34.4389],[310,-30.1921],[331,-35.9313],[357,-29.9303],[376,-34.9008],[389,-34.9668]]
       }
       thigh_r_fk_ctrl.Scale.X {type=MovieSceneScriptingFloatChannel, default=1, const [1,29]}
       thigh_r_fk_ctrl.Scale.Y {type=MovieSceneScriptingFloatChannel, default=1, const [1,29]}
       thigh_r_fk_ctrl.Scale.Z {type=MovieSceneScriptingFloatChannel, default=1, const [1,29]}
       calf_r_fk_ctrl.Location.X {
        type=MovieSceneScriptingFloatChannel
        default=0
        thinned_from=29
        keys [[-1,0],[0,-0.0001],[21,0],[46,0.001],[178,0.001],[183,0],[294,0],[310,-0.0001],[331,0],[357,0.001],[376,0],[389,0]]
       }
       calf_r_fk_ctrl.Location.Y {
        type=MovieSceneScriptingFloatChannel
        default=0
        thinned_from=29
        keys [[-1,0],[0,-0.001],[21,0],[46,0.0076],[178,0.0076],[183,0],[190,0],[210,0.0013],[217,0],[294,0],[310,-0.001],[331,0],[357,0.0076],[376,0],[389,0]]
       }
       calf_r_fk_ctrl.Location.Z {
        type=MovieSceneScriptingFloatChannel
        default=0
        thinned_from=29
        keys [[-1,0],[0,0.0001],[21,0],[46,0.0021],[178,0.0021],[183,0],[190,0],[210,0.0049],[217,0],[310,0.0001],[331,0],[357,0.0021],[376,0],[389,0]]
       }
       calf_r_fk_ctrl.Rotation.X {
        type=MovieSceneScriptingFloatChannel
        default=0.0752
        thinned_from=29
        keys [[-1,0.8312],[0,0.815],[21,0.8226],[60,0.8214],[78,0.8193],[84,0.8214],[119,0.8255],[135,0.8259],[139,0.8295],[146,0.8247],[160,0.8242],[172,0.826],[178,0.8369],[183,0.8286],[190,0.8052],[210,0.7915],[217,0.7945],[234,0.7864],[267,0.7864],[294,0.8024],[310,0.815],[331,0.8151],[357,0.8214],[376,0.8222],[389,0.8251]]
       }
       calf_r_fk_ctrl.Rotation.Y {
        type=MovieSceneScriptingFloatChannel
        default=1.0359
        thinned_from=29
        keys [[-1,0.6228],[0,0.6435],[21,0.6341],[46,0.6311],[60,0.6311],[78,0.6338],[84,0.6311],[119,0.6255],[135,0.625],[139,0.6202],[146,0.6266],[160,0.6273],[172,0.6249],[178,0.6098],[183,0.6262],[190,0.656],[210,0.6619],[217,0.669],[234,0.6785],[267,0.6785],[294,0.6595],[310,0.6435],[331,0.6439],[357,0.6311],[376,0.6345],[389,0.6309]]
       }
       calf_r_fk_ctrl.Rotation.Z {
        type=MovieSceneScriptingFloatChannel
        default=0.0678
        thinned_from=29
        keys [[-1,49.0767],[0,47.6183],[21,48.2904],[60,48.4046],[78,48.2133],[84,48.4046],[119,48.7866],[135,48.8247],[139,49.1522],[146,48.71],[160,48.6634],[172,48.8262],[178,49.8584],[183,48.8409],[190,46.7486],[210,45.8125],[217,45.8196],[234,45.1318],[267,45.1318],[294,46.5032],[310,47.6183],[331,47.6128],[357,48.4046],[376,48.2635],[389,48.515]]
       }
       calf_r_fk_ctrl.Scale.X {type=MovieSceneScriptingFloatChannel, default=1, const [1,29]}
       calf_r_fk_ctrl.Scale.Y {type=MovieSceneScriptingFloatChannel, default=1, const [1,29]}
       calf_r_fk_ctrl.Scale.Z {type=MovieSceneScriptingFloatChannel, default=1, const [1,29]}
       foot_r_fk_ctrl.Location.X {type=MovieSceneScriptingFloatChannel, default=0, thinned_from=29, keys [[-1,0],[389,0]]}
       foot_r_fk_ctrl.Location.Y {
        type=MovieSceneScriptingFloatChannel
        default=-0.0543
        thinned_from=29
        keys [[-1,-0.0543],[234,-0.0543],[253,-0.0465],[267,-0.0465],[294,-0.0543],[389,-0.0543]]
       }
       foot_r_fk_ctrl.Location.Z {type=MovieSceneScriptingFloatChannel, default=-0.8609, thinned_from=29, keys [[-1,-0.8609],[389,-0.8609]]}
       foot_r_fk_ctrl.Rotation.X {
        type=MovieSceneScriptingFloatChannel
        default=-0.9802
        thinned_from=29
        keys [[-1,19.8326],[0,19.7592],[60,19.7629],[78,19.7247],[84,19.7629],[92,19.7365],[172,19.7981],[178,18.613],[183,20.9814],[190,21.1163],[210,22.4244],[217,21.0164],[234,20.8776],[267,20.8757],[294,21.0519],[310,19.7592],[331,21.0545],[357,19.7629],[376,21.1219],[389,21.1319]]
       }
       foot_r_fk_ctrl.Rotation.Y {
        type=MovieSceneScriptingFloatChannel
        default=0.1214
        thinned_from=29
        keys [[-1,-9.256],[0,-10.3585],[21,-9.6389],[46,-9.8837],[60,-9.8837],[78,-9.9635],[84,-9.8837],[92,-9.8166],[112,-9.4691],[119,-9.5162],[135,-9.5342],[139,-9.4672],[146,-9.8156],[160,-10.0061],[172,-9.7482],[178,-7.4678],[183,-10.0682],[190,-10.6877],[210,-11.4752],[217,-11.3789],[234,-11.7858],[267,-11.7913],[294,-10.9747],[310,-10.3585],[331,-11.348],[357,-9.8837],[376,-10.3046],[389,-10.2089]]
       }
       foot_r_fk_ctrl.Rotation.Z {
        type=MovieSceneScriptingFloatChannel
        default=0.0013
        thinned_from=29
        keys [[-1,-31.0885],[0,-30.1491],[21,-30.4666],[60,-30.5088],[78,-30.2674],[84,-30.5088],[92,-30.4231],[112,-30.5639],[119,-30.6794],[135,-30.7235],[139,-30.9273],[146,-30.7206],[160,-30.7361],[172,-30.8421],[178,-32.2348],[183,-31.0911],[190,-29.6483],[210,-29.4232],[217,-28.8985],[234,-28.1007],[267,-28.0995],[294,-29.3079],[310,-30.1491],[331,-29.7991],[357,-30.5088],[376,-30.3462],[389,-30.4945]]
       }
       foot_r_fk_ctrl.Scale.X {type=MovieSceneScriptingFloatChannel, default=1, const [1,29]}
       foot_r_fk_ctrl.Scale.Y {type=MovieSceneScriptingFloatChannel, default=1, const [1,29]}
       foot_r_fk_ctrl.Scale.Z {type=MovieSceneScriptingFloatChannel, default=1, const [1,29]}
       ball_r_fk_ctrl.Location.X {type=MovieSceneScriptingFloatChannel, default=0, thinned_from=29, keys [[-1,0],[0,0.0007],[21,0],[294,0],[310,0.0007],[331,0],[389,0]]}
       ball_r_fk_ctrl.Location.Y {type=MovieSceneScriptingFloatChannel, default=0, thinned_from=29, keys [[-1,0],[0,0.0002],[21,0],[294,0],[310,0.0002],[331,0],[389,0]]}
       ball_r_fk_ctrl.Location.Z {type=MovieSceneScriptingFloatChannel, default=0, thinned_from=29, keys [[-1,0],[389,0]]}
       ball_r_fk_ctrl.Rotation.X {type=MovieSceneScriptingFloatChannel, default=0, thinned_from=29, keys [[-1,0],[0,0.0001],[294,0],[310,0.0001],[389,0]]}
       ball_r_fk_ctrl.Rotation.Y {type=MovieSceneScriptingFloatChannel, default=0, thinned_from=29, keys [[-1,0],[0,0.0006],[21,0],[294,0],[310,0.0006],[331,0],[389,0]]}
       ball_r_fk_ctrl.Rotation.Z {type=MovieSceneScriptingFloatChannel, default=0, thinned_from=29, keys [[-1,0],[0,-0.0037],[21,0],[294,0],[310,-0.0037],[331,0],[389,0]]}
       ball_r_fk_ctrl.Scale.X {type=MovieSceneScriptingFloatChannel, default=1, const [1,29]}
       ball_r_fk_ctrl.Scale.Y {type=MovieSceneScriptingFloatChannel, default=1, const [1,29]}
       ball_r_fk_ctrl.Scale.Z {type=MovieSceneScriptingFloatChannel, default=1, const [1,29]}
       foot_r_ik_ctrl.Location.X {type=MovieSceneScriptingFloatChannel, default=0, thinned_from=29, keys [[-1,4.8731],[389,4.8731]]}
       foot_r_ik_ctrl.Location.Y {type=MovieSceneScriptingFloatChannel, default=0, const [-15.684,29]}
       foot_r_ik_ctrl.Location.Z {type=MovieSceneScriptingFloatChannel, default=0, const [0,29]}
       foot_r_ik_ctrl.Rotation.X {type=MovieSceneScriptingFloatChannel, default=0, const [0,29]}
       foot_r_ik_ctrl.Rotation.Y {type=MovieSceneScriptingFloatChannel, default=0, const [0,29]}
       foot_r_ik_ctrl.Rotation.Z {type=MovieSceneScriptingFloatChannel, default=0, const [-26.123,29]}
       foot_r_ik_ctrl.Scale.X {type=MovieSceneScriptingFloatChannel, default=1, const [1,29]}
       foot_r_ik_ctrl.Scale.Y {type=MovieSceneScriptingFloatChannel, default=1, const [1,29]}
       foot_r_ik_ctrl.Scale.Z {type=MovieSceneScriptingFloatChannel, default=1, const [1,29]}
       ball_r_ik_ctrl.Location.X {type=MovieSceneScriptingFloatChannel, default=0, const [0,29]}
       ball_r_ik_ctrl.Location.Y {type=MovieSceneScriptingFloatChannel, default=0, const [0,29]}
       ball_r_ik_ctrl.Location.Z {type=MovieSceneScriptingFloatChannel, default=0, const [0,29]}
       ball_r_ik_ctrl.Rotation.X {type=MovieSceneScriptingFloatChannel, default=0, thinned_from=29, keys [[-1,0],[389,0]]}
       ball_r_ik_ctrl.Rotation.Y {type=MovieSceneScriptingFloatChannel, default=0, const [0,29]}
       ball_r_ik_ctrl.Rotation.Z {type=MovieSceneScriptingFloatChannel, default=0, const [0,29]}
       ball_r_ik_ctrl.Scale.X {type=MovieSceneScriptingFloatChannel, default=1, const [1,29]}
       ball_r_ik_ctrl.Scale.Y {type=MovieSceneScriptingFloatChannel, default=1, const [1,29]}
       ball_r_ik_ctrl.Scale.Z {type=MovieSceneScriptingFloatChannel, default=1, const [1,29]}
       foot_bk1_r_ctrl.Location.X {type=MovieSceneScriptingFloatChannel, default=0, const [0,29]}
       foot_bk1_r_ctrl.Location.Y {type=MovieSceneScriptingFloatChannel, default=0, const [0,29]}
       foot_bk1_r_ctrl.Location.Z {type=MovieSceneScriptingFloatChannel, default=0, const [0,29]}
       foot_bk1_r_ctrl.Rotation.X {type=MovieSceneScriptingFloatChannel, default=0, thinned_from=29, keys [[-1,0],[389,0]]}
       foot_bk1_r_ctrl.Rotation.Y {type=MovieSceneScriptingFloatChannel, default=0, const [0,29]}
       foot_bk1_r_ctrl.Rotation.Z {type=MovieSceneScriptingFloatChannel, default=0, thinned_from=29, keys [[-1,0],[389,0]]}
       foot_bk1_r_ctrl.Scale.X {type=MovieSceneScriptingFloatChannel, default=1, const [1,29]}
       foot_bk1_r_ctrl.Scale.Y {type=MovieSceneScriptingFloatChannel, default=1, const [1,29]}
       foot_bk1_r_ctrl.Scale.Z {type=MovieSceneScriptingFloatChannel, default=1, const [1,29]}
       foot_roll_r_ctrl.Location.X {type=MovieSceneScriptingFloatChannel, default=0, const [0,29]}
       foot_roll_r_ctrl.Location.Y {type=MovieSceneScriptingFloatChannel, default=0, const [0,29]}
       foot_roll_r_ctrl.Location.Z {type=MovieSceneScriptingFloatChannel, default=0, const [0,29]}
       foot_roll_r_ctrl.Rotation.X {type=MovieSceneScriptingFloatChannel, default=0, const [0,29]}
       foot_roll_r_ctrl.Rotation.Y {type=MovieSceneScriptingFloatChannel, default=0, const [0,29]}
       foot_roll_r_ctrl.Rotation.Z {type=MovieSceneScriptingFloatChannel, default=0, thinned_from=30, keys [[-1,0],[253,0],[265,-0.2176],[272,0],[389,0]]}
       foot_roll_r_ctrl.Scale.X {type=MovieSceneScriptingFloatChannel, default=1, const [1,29]}
       foot_roll_r_ctrl.Scale.Y {type=MovieSceneScriptingFloatChannel, default=1, const [1,29]}
       foot_roll_r_ctrl.Scale.Z {type=MovieSceneScriptingFloatChannel, default=1, const [1,29]}
       roll_blend_r_ctrl {type=MovieSceneScriptingFloatChannel, default=1, const [1,28]}
       tip_r_ctrl.Location.X {type=MovieSceneScriptingFloatChannel, default=0, const [0,31]}
       tip_r_ctrl.Location.Y {type=MovieSceneScriptingFloatChannel, default=0, const [0,31]}
       tip_r_ctrl.Location.Z {type=MovieSceneScriptingFloatChannel, default=0, const [0,31]}
       tip_r_ctrl.Rotation.X {type=MovieSceneScriptingFloatChannel, default=0, const [0,31]}
       tip_r_ctrl.Rotation.Y {
        type=MovieSceneScriptingFloatChannel
        default=0
        thinned_from=31
        keys [[-1,0],[178,0],[183,-0.8255],[190,-4.2616],[210,-4.5721],[253,-4.5721],[265,-3.4927],[272,-0.5715],[280,0],[389,0]]
       }
       tip_r_ctrl.Rotation.Z {type=MovieSceneScriptingFloatChannel, default=0, const [0,31]}
       tip_r_ctrl.Scale.X {type=MovieSceneScriptingFloatChannel, default=1, const [1,31]}
       tip_r_ctrl.Scale.Y {type=MovieSceneScriptingFloatChannel, default=1, const [1,31]}
       tip_r_ctrl.Scale.Z {type=MovieSceneScriptingFloatChannel, default=1, const [1,31]}
       heel_r_ctrl.Location.X {type=MovieSceneScriptingFloatChannel, default=0, thinned_from=29, keys [[-1,0],[389,0]]}
       heel_r_ctrl.Location.Y {type=MovieSceneScriptingFloatChannel, default=0, thinned_from=29, keys [[-1,0],[389,0]]}
       heel_r_ctrl.Location.Z {type=MovieSceneScriptingFloatChannel, default=0, thinned_from=29, keys [[-1,0],[389,0]]}
       heel_r_ctrl.Rotation.X {type=MovieSceneScriptingFloatChannel, default=0, const [0,29]}
       heel_r_ctrl.Rotation.Y {type=MovieSceneScriptingFloatChannel, default=0, thinned_from=29, keys [[-1,0],[389,0]]}
       heel_r_ctrl.Rotation.Z {type=MovieSceneScriptingFloatChannel, default=0, const [0,29]}
       heel_r_ctrl.Scale.X {type=MovieSceneScriptingFloatChannel, default=1, const [1,29]}
       heel_r_ctrl.Scale.Y {type=MovieSceneScriptingFloatChannel, default=1, const [1,29]}
       heel_r_ctrl.Scale.Z {type=MovieSceneScriptingFloatChannel, default=1, const [1,29]}
       leg_r_softness {type=MovieSceneScriptingFloatChannel, default=0, thinned_from=30, keys [[-1,0.2588],[0,1],[389,1]]}
       leg_r_pv_ik_ctrl.Location.X {
        type=MovieSceneScriptingFloatChannel
        default=0
        thinned_from=30
        keys [[-1,-24.3112],[0,-24.3112],[11,-24.6543],[21,-26.9238],[46,-27.9713],[60,-28.155],[84,-27.8428],[172,-27.8428],[178,-27.3463],[183,-24.8459],[190,-24.3112],[210,-24.3112],[217,-24.1407],[234,-23.0281],[253,-22.695],[267,-22.695],[294,-23.9071],[310,-24.3112],[389,-24.3112]]
       }
       leg_r_pv_ik_ctrl.Location.Y {type=MovieSceneScriptingFloatChannel, default=0, thinned_from=30, keys [[-1,17.6473],[389,17.6473]]}
       leg_r_pv_ik_ctrl.Location.Z {type=MovieSceneScriptingFloatChannel, default=0, thinned_from=30, keys [[-1,23.7842],[46,23.8182],[389,23.7842]]}
       leg_r_pv_ik_ctrl.Rotation.X {type=MovieSceneScriptingFloatChannel, default=0, const [-27.2161,30]}
       leg_r_pv_ik_ctrl.Rotation.Y {type=MovieSceneScriptingFloatChannel, default=0, thinned_from=30, keys [[-1,-0.5322],[389,-0.5322]]}
       leg_r_pv_ik_ctrl.Rotation.Z {type=MovieSceneScriptingFloatChannel, default=0, thinned_from=30, keys [[-1,0.274],[389,0.274]]}
       leg_r_pv_ik_ctrl.Scale.X {type=MovieSceneScriptingFloatChannel, default=1, const [1,30]}
       leg_r_pv_ik_ctrl.Scale.Y {type=MovieSceneScriptingFloatChannel, default=1, const [1,30]}
       leg_r_pv_ik_ctrl.Scale.Z {type=MovieSceneScriptingFloatChannel, default=1, const [1,30]}
       upperarm_l_fk_ctrl.Location.X {type=MovieSceneScriptingFloatChannel, default=0, thinned_from=38, keys [[-1,0],[389,0]]}
       upperarm_l_fk_ctrl.Location.Y {type=MovieSceneScriptingFloatChannel, default=0, thinned_from=38, keys [[-1,0],[389,0]]}
       upperarm_l_fk_ctrl.Location.Z {type=MovieSceneScriptingFloatChannel, default=0, thinned_from=38, keys [[-1,0],[389,0]]}
       upperarm_l_fk_ctrl.Rotation.X {
        type=MovieSceneScriptingFloatChannel
        default=0
        thinned_from=38
        keys [[-1,18.7969],[0,18.6273],[21,20.3925],[46,20.7219],[60,20.7014],[78,20.5576],[139,20.5576],[142,21.8479],[146,20.5576],[160,22.1231],[172,22.1231],[178,17.9698],[183,17.6495],[190,21.1975],[217,21.1975],[221,19.5766],[225,20.6356],[234,23.6991],[245,24.4032],[253,24.5038],[265,28.2563],[267,27.4762],[272,24.4877],[280,30.8107],[286,30.7496],[294,24.4024],[310,18.6273],[331,19.4135],[357,20.5576],[376,20.2361],[388,20.692],[389,19.632]]
       }
       upperarm_l_fk_ctrl.Rotation.Y {
        type=MovieSceneScriptingFloatChannel
        default=0
        thinned_from=38
        keys [[-1,-8.1037],[0,-9.2002],[21,-9.0846],[60,-9.1665],[78,-9.4074],[139,-9.4074],[142,-7.9155],[146,-9.4074],[160,-5.4858],[172,-5.4858],[178,-16.2129],[183,-1.1504],[190,10.8178],[217,10.8178],[221,12.1621],[225,11.1511],[234,6.0536],[245,4.9753],[253,4.8213],[265,0.4136],[267,1.3265],[272,5.2282],[280,2.918],[286,3.5609],[294,6.8469],[310,-9.2002],[331,5.7129],[357,-9.4074],[376,6.8884],[388,-8.6676],[389,6.8962]]
       }
       upperarm_l_fk_ctrl.Rotation.Z {
        type=MovieSceneScriptingFloatChannel
        default=0
        thinned_from=38
        keys [[-1,-7.0661],[0,-13.2855],[21,-14.3756],[46,-14.8843],[60,-14.8991],[78,-15.0026],[139,-15.0026],[142,-14.2496],[146,-15.0026],[160,-17.2563],[172,-17.2563],[178,-37.6545],[183,-11.2278],[190,7.638],[217,7.638],[221,4.759],[225,5.7895],[234,10.3951],[245,10.9253],[253,11.001],[265,11.8695],[267,11.6761],[272,10.3198],[280,7.2738],[286,6.1928],[294,7.6058],[310,-13.2855],[331,3.6316],[357,-15.0026],[376,9.6006],[388,-8.492],[389,9.8917]]
       }
       upperarm_l_fk_ctrl.Scale.X {type=MovieSceneScriptingFloatChannel, default=1, const [1,38]}
       upperarm_l_fk_ctrl.Scale.Y {type=MovieSceneScriptingFloatChannel, default=1, const [1,38]}
       upperarm_l_fk_ctrl.Scale.Z {type=MovieSceneScriptingFloatChannel, default=1, const [1,38]}
       lowerarm_l_fk_ctrl.Location.X {type=MovieSceneScriptingFloatChannel, default=0, thinned_from=11, keys [[-1,0],[21,0],[46,-0.0004],[183,0],[388,0]]}
       lowerarm_l_fk_ctrl.Location.Y {type=MovieSceneScriptingFloatChannel, default=0, thinned_from=11, keys [[-1,0],[21,0],[46,-0.0007],[183,0],[388,0]]}
       lowerarm_l_fk_ctrl.Location.Z {type=MovieSceneScriptingFloatChannel, default=0, thinned_from=11, keys [[-1,0],[21,0],[46,0.0011],[183,0],[388,0]]}
       lowerarm_l_fk_ctrl.Rotation.X {
        type=MovieSceneScriptingFloatChannel
        default=0
        thinned_from=11
        keys [[-1,0.379],[0,0.3769],[21,0.3715],[46,0.3726],[178,0.3695],[183,0.3861],[245,0.3781],[310,0.3769],[357,0.3695],[388,0.3711]]
       }
       lowerarm_l_fk_ctrl.Rotation.Y {
        type=MovieSceneScriptingFloatChannel
        default=0
        thinned_from=11
        keys [[-1,-0.2603],[0,-0.2633],[21,-0.2709],[46,-0.2743],[178,-0.2736],[183,-0.2497],[245,-0.2616],[310,-0.2633],[357,-0.2736],[388,-0.2715]]
       }
       lowerarm_l_fk_ctrl.Rotation.Z {
        type=MovieSceneScriptingFloatChannel
        default=0
        thinned_from=11
        keys [[-1,111.5985],[0,112.0579],[21,113.2281],[46,113.6442],[178,113.6459],[183,110.0128],[245,111.8023],[310,112.0579],[357,113.6459],[388,113.3106]]
       }
       lowerarm_l_fk_ctrl.Scale.X {type=MovieSceneScriptingFloatChannel, default=1, const [1,11]}
       lowerarm_l_fk_ctrl.Scale.Y {type=MovieSceneScriptingFloatChannel, default=1, const [1,11]}
       lowerarm_l_fk_ctrl.Scale.Z {type=MovieSceneScriptingFloatChannel, default=1, const [1,11]}
       hand_l_fk_ctrl.Location.X {type=MovieSceneScriptingFloatChannel, default=0, thinned_from=36, keys [[-1,0],[389,0]]}
       hand_l_fk_ctrl.Location.Y {type=MovieSceneScriptingFloatChannel, default=0, thinned_from=36, keys [[-1,-0.1788],[389,-0.1788]]}
       hand_l_fk_ctrl.Location.Z {type=MovieSceneScriptingFloatChannel, default=0, thinned_from=36, keys [[-1,0.0249],[389,0.0249]]}
       hand_l_fk_ctrl.Rotation.X {
        type=MovieSceneScriptingFloatChannel
        default=0
        thinned_from=36
        keys [[-1,-2.2111],[0,-1.997],[21,-0.5125],[46,0.0543],[60,0.0601],[78,0.1002],[139,0.1002],[142,-0.2565],[146,0.1002],[172,0.1002],[178,16.1824],[183,-0.6158],[190,-13.198],[217,-13.198],[221,-14.6832],[225,-14.7346],[234,-13.8275],[253,-13.8276],[265,-13.267],[267,-13.3665],[272,-13.8276],[294,-13.8275],[310,-1.997],[331,-12.9671],[357,0.1002],[376,-13.269],[388,0.1486],[389,-13.198]]
       }
       hand_l_fk_ctrl.Rotation.Y {
        type=MovieSceneScriptingFloatChannel
        default=0
        thinned_from=36
        keys [[-1,21.6846],[0,26.5322],[21,26.947],[46,27.1172],[139,27.1334],[142,26.5967],[146,27.1334],[172,27.1334],[178,38.3344],[183,24.4044],[190,13.2798],[217,13.2798],[221,12.5031],[225,12.4312],[234,12.7298],[253,12.7298],[265,14.1989],[267,13.9196],[272,12.7298],[294,12.7298],[310,26.5322],[331,18.1335],[357,27.1334],[376,13.2179],[388,22.072],[389,13.2798]]
       }
       hand_l_fk_ctrl.Rotation.Z {
        type=MovieSceneScriptingFloatChannel
        default=0
        thinned_from=36
        keys [[-1,-27.1697],[0,-22.1289],[21,-22.2862],[46,-21.8781],[60,-21.832],[78,-21.5092],[139,-21.5092],[142,-23.5723],[146,-21.5092],[172,-21.5092],[178,-11.5729],[183,-38.3189],[190,-44.8835],[217,-44.8835],[221,-54.9678],[225,-55.2919],[234,-45.6997],[253,-45.6997],[265,-46.8367],[272,-45.6997],[294,-45.6997],[310,-22.1289],[331,-39.7271],[357,-21.5092],[376,-44.9788],[388,-26.4448],[389,-44.8835]]
       }
       hand_l_fk_ctrl.Scale.X {type=MovieSceneScriptingFloatChannel, default=1, const [1,36]}
       hand_l_fk_ctrl.Scale.Y {type=MovieSceneScriptingFloatChannel, default=1, const [1,36]}
       hand_l_fk_ctrl.Scale.Z {type=MovieSceneScriptingFloatChannel, default=1, const [1,36]}
       thumb_01_l_ctrl.Location.X {type=MovieSceneScriptingFloatChannel, default=0, thinned_from=29, keys [[-1,0],[389,0]]}
       thumb_01_l_ctrl.Location.Y {type=MovieSceneScriptingFloatChannel, default=0, thinned_from=29, keys [[-1,0],[389,0]]}
       thumb_01_l_ctrl.Location.Z {type=MovieSceneScriptingFloatChannel, default=0, thinned_from=29, keys [[-1,0],[389,0]]}
       thumb_01_l_ctrl.Rotation.X {type=MovieSceneScriptingFloatChannel, default=0, const [-11.8305,29]}
       thumb_01_l_ctrl.Rotation.Y {type=MovieSceneScriptingFloatChannel, default=0, thinned_from=29, keys [[-1,-6.959],[389,-6.959]]}
       thumb_01_l_ctrl.Rotation.Z {type=MovieSceneScriptingFloatChannel, default=0, thinned_from=29, keys [[-1,4.4404],[389,4.4404]]}
       thumb_01_l_ctrl.Scale.X {type=MovieSceneScriptingFloatChannel, default=1, const [1,29]}
       thumb_01_l_ctrl.Scale.Y {type=MovieSceneScriptingFloatChannel, default=1, const [1,29]}
       thumb_01_l_ctrl.Scale.Z {type=MovieSceneScriptingFloatChannel, default=1, const [1,29]}
       thumb_02_l_ctrl.Location.X {type=MovieSceneScriptingFloatChannel, default=0, thinned_from=29, keys [[-1,0],[389,0]]}
       thumb_02_l_ctrl.Location.Y {type=MovieSceneScriptingFloatChannel, default=0, thinned_from=29, keys [[-1,0],[389,0]]}
       thumb_02_l_ctrl.Location.Z {type=MovieSceneScriptingFloatChannel, default=0, thinned_from=29, keys [[-1,0],[389,0]]}
       thumb_02_l_ctrl.Rotation.X {type=MovieSceneScriptingFloatChannel, default=0, thinned_from=29, keys [[-1,0],[389,0]]}
       thumb_02_l_ctrl.Rotation.Y {type=MovieSceneScriptingFloatChannel, default=0, thinned_from=29, keys [[-1,0],[389,0]]}
       thumb_02_l_ctrl.Rotation.Z {type=MovieSceneScriptingFloatChannel, default=0, const [-12.3705,29]}
       thumb_02_l_ctrl.Scale.X {type=MovieSceneScriptingFloatChannel, default=1, const [1,29]}
       thumb_02_l_ctrl.Scale.Y {type=MovieSceneScriptingFloatChannel, default=1, const [1,29]}
       thumb_02_l_ctrl.Scale.Z {type=MovieSceneScriptingFloatChannel, default=1, const [1,29]}
       thumb_03_l_ctrl.Location.X {type=MovieSceneScriptingFloatChannel, default=0, thinned_from=29, keys [[-1,-0.5596],[389,-0.5596]]}
       thumb_03_l_ctrl.Location.Y {type=MovieSceneScriptingFloatChannel, default=0, thinned_from=29, keys [[-1,-0.2442],[389,-0.2442]]}
       thumb_03_l_ctrl.Location.Z {type=MovieSceneScriptingFloatChannel, default=0, thinned_from=29, keys [[-1,0],[389,0]]}
       thumb_03_l_ctrl.Rotation.X {type=MovieSceneScriptingFloatChannel, default=0, thinned_from=29, keys [[-1,0],[389,0]]}
       thumb_03_l_ctrl.Rotation.Y {type=MovieSceneScriptingFloatChannel, default=0, thinned_from=29, keys [[-1,0],[389,0]]}
       thumb_03_l_ctrl.Rotation.Z {type=MovieSceneScriptingFloatChannel, default=0, const [-66.4234,29]}
       thumb_03_l_ctrl.Scale.X {type=MovieSceneScriptingFloatChannel, default=1, const [1,29]}
       thumb_03_l_ctrl.Scale.Y {type=MovieSceneScriptingFloatChannel, default=1, const [1,29]}
       thumb_03_l_ctrl.Scale.Z {type=MovieSceneScriptingFloatChannel, default=1, const [1,29]}
       index_metacarpal_l_ctrl.Location.X {type=MovieSceneScriptingFloatChannel, default=0, thinned_from=29, keys [[-1,0],[389,0]]}
       index_metacarpal_l_ctrl.Location.Y {type=MovieSceneScriptingFloatChannel, default=0, thinned_from=29, keys [[-1,0],[389,0]]}
       index_metacarpal_l_ctrl.Location.Z {type=MovieSceneScriptingFloatChannel, default=0, thinned_from=29, keys [[-1,0],[389,0]]}
       index_metacarpal_l_ctrl.Rotation.X {type=MovieSceneScriptingFloatChannel, default=0, const [-2.7256,29]}
       index_metacarpal_l_ctrl.Rotation.Y {type=MovieSceneScriptingFloatChannel, default=0, thinned_from=29, keys [[-1,0],[389,0]]}
       index_metacarpal_l_ctrl.Rotation.Z {type=MovieSceneScriptingFloatChannel, default=0, thinned_from=29, keys [[-1,-7.3778],[389,-7.3778]]}
       index_metacarpal_l_ctrl.Scale.X {type=MovieSceneScriptingFloatChannel, default=1, const [1,29]}
       index_metacarpal_l_ctrl.Scale.Y {type=MovieSceneScriptingFloatChannel, default=1, const [1,29]}
       index_metacarpal_l_ctrl.Scale.Z {type=MovieSceneScriptingFloatChannel, default=1, const [1,29]}
       index_01_l_ctrl.Location.X {type=MovieSceneScriptingFloatChannel, default=0, thinned_from=31, keys [[-1,0],[389,0]]}
       index_01_l_ctrl.Location.Y {type=MovieSceneScriptingFloatChannel, default=0, thinned_from=31, keys [[-1,0],[389,0]]}
       index_01_l_ctrl.Location.Z {type=MovieSceneScriptingFloatChannel, default=0, thinned_from=31, keys [[-1,0],[389,0]]}
       index_01_l_ctrl.Rotation.X {type=MovieSceneScriptingFloatChannel, default=0, thinned_from=31, keys [[-1,0],[217,0],[223,3.034],[234,0],[389,0]]}
       index_01_l_ctrl.Rotation.Y {type=MovieSceneScriptingFloatChannel, default=0, thinned_from=31, keys [[-1,0],[217,0],[223,0.8215],[234,0],[389,0]]}
       index_01_l_ctrl.Rotation.Z {
        type=MovieSceneScriptingFloatChannel
        default=0
        thinned_from=32
        keys [[-1,-64.9002],[210,-64.9002],[217,-65.8315],[223,-39.1645],[228,-37.7239],[234,-61.0198],[242,-68.8438],[253,-65.3932],[267,-64.9002],[389,-64.9002]]
       }
       index_01_l_ctrl.Scale.X {type=MovieSceneScriptingFloatChannel, default=1, const [1,31]}
       index_01_l_ctrl.Scale.Y {type=MovieSceneScriptingFloatChannel, default=1, const [1,31]}
       index_01_l_ctrl.Scale.Z {type=MovieSceneScriptingFloatChannel, default=1, const [1,31]}
       index_02_l_ctrl.Location.X {type=MovieSceneScriptingFloatChannel, default=0, const [0,29]}
       index_02_l_ctrl.Location.Y {type=MovieSceneScriptingFloatChannel, default=0, const [0,29]}
       index_02_l_ctrl.Location.Z {type=MovieSceneScriptingFloatChannel, default=0, const [0,29]}
       index_02_l_ctrl.Rotation.X {type=MovieSceneScriptingFloatChannel, default=0, thinned_from=29, keys [[-1,0],[389,0]]}
       index_02_l_ctrl.Rotation.Y {type=MovieSceneScriptingFloatChannel, default=0, thinned_from=29, keys [[-1,0],[389,0]]}
       index_02_l_ctrl.Rotation.Z {type=MovieSceneScriptingFloatChannel, default=0, const [-49.5772,29]}
       index_02_l_ctrl.Scale.X {type=MovieSceneScriptingFloatChannel, default=1, const [1,29]}
       index_02_l_ctrl.Scale.Y {type=MovieSceneScriptingFloatChannel, default=1, const [1,29]}
       index_02_l_ctrl.Scale.Z {type=MovieSceneScriptingFloatChannel, default=1, const [1,29]}
       index_03_l_ctrl.Location.X {type=MovieSceneScriptingFloatChannel, default=0, const [0,29]}
       index_03_l_ctrl.Location.Y {type=MovieSceneScriptingFloatChannel, default=0, const [0,29]}
       index_03_l_ctrl.Location.Z {type=MovieSceneScriptingFloatChannel, default=0, const [0,29]}
       index_03_l_ctrl.Rotation.X {type=MovieSceneScriptingFloatChannel, default=0, thinned_from=29, keys [[-1,0],[389,0]]}
       index_03_l_ctrl.Rotation.Y {type=MovieSceneScriptingFloatChannel, default=0, thinned_from=29, keys [[-1,0],[389,0]]}
       index_03_l_ctrl.Rotation.Z {type=MovieSceneScriptingFloatChannel, default=0, const [-102.5714,29]}
       index_03_l_ctrl.Scale.X {type=MovieSceneScriptingFloatChannel, default=1, const [1,29]}
       index_03_l_ctrl.Scale.Y {type=MovieSceneScriptingFloatChannel, default=1, const [1,29]}
       index_03_l_ctrl.Scale.Z {type=MovieSceneScriptingFloatChannel, default=1, const [1,29]}
       middle_metacarpal_l_ctrl.Location.X {type=MovieSceneScriptingFloatChannel, default=0, const [0,29]}
       middle_metacarpal_l_ctrl.Location.Y {type=MovieSceneScriptingFloatChannel, default=0, const [0,29]}
       middle_metacarpal_l_ctrl.Location.Z {type=MovieSceneScriptingFloatChannel, default=0, const [0,29]}
       middle_metacarpal_l_ctrl.Rotation.X {type=MovieSceneScriptingFloatChannel, default=0, const [-2.7256,29]}
       middle_metacarpal_l_ctrl.Rotation.Y {type=MovieSceneScriptingFloatChannel, default=0, const [0,29]}
       middle_metacarpal_l_ctrl.Rotation.Z {type=MovieSceneScriptingFloatChannel, default=0, thinned_from=29, keys [[-1,-7.3778],[389,-7.3778]]}
       middle_metacarpal_l_ctrl.Scale.X {type=MovieSceneScriptingFloatChannel, default=1, const [1,29]}
       middle_metacarpal_l_ctrl.Scale.Y {type=MovieSceneScriptingFloatChannel, default=1, const [1,29]}
       middle_metacarpal_l_ctrl.Scale.Z {type=MovieSceneScriptingFloatChannel, default=1, const [1,29]}
       middle_01_l_ctrl.Location.X {type=MovieSceneScriptingFloatChannel, default=0, const [0,31]}
       middle_01_l_ctrl.Location.Y {type=MovieSceneScriptingFloatChannel, default=0, const [0,31]}
       middle_01_l_ctrl.Location.Z {type=MovieSceneScriptingFloatChannel, default=0, const [0,31]}
       middle_01_l_ctrl.Rotation.X {type=MovieSceneScriptingFloatChannel, default=0, thinned_from=31, keys [[-1,0],[217,0],[223,2.3229],[234,0],[389,0]]}
       middle_01_l_ctrl.Rotation.Y {type=MovieSceneScriptingFloatChannel, default=0, thinned_from=31, keys [[-1,0],[217,0],[223,2.118],[234,0],[389,0]]}
       middle_01_l_ctrl.Rotation.Z {
        type=MovieSceneScriptingFloatChannel
        default=0
        thinned_from=32
        keys [[-1,-64.9002],[210,-64.9002],[217,-60.5579],[223,-51.3288],[228,-62.9397],[234,-72.9876],[242,-69.0787],[253,-65.4225],[267,-64.9002],[389,-64.9002]]
       }
       middle_01_l_ctrl.Scale.X {type=MovieSceneScriptingFloatChannel, default=1, const [1,31]}
       middle_01_l_ctrl.Scale.Y {type=MovieSceneScriptingFloatChannel, default=1, const [1,31]}
       middle_01_l_ctrl.Scale.Z {type=MovieSceneScriptingFloatChannel, default=1, const [1,31]}
       middle_02_l_ctrl.Location.X {type=MovieSceneScriptingFloatChannel, default=0, const [0,29]}
       middle_02_l_ctrl.Location.Y {type=MovieSceneScriptingFloatChannel, default=0, const [0,29]}
       middle_02_l_ctrl.Location.Z {type=MovieSceneScriptingFloatChannel, default=0, const [0,29]}
       middle_02_l_ctrl.Rotation.X {type=MovieSceneScriptingFloatChannel, default=0, thinned_from=29, keys [[-1,0],[389,0]]}
       middle_02_l_ctrl.Rotation.Y {type=MovieSceneScriptingFloatChannel, default=0, thinned_from=29, keys [[-1,0],[389,0]]}
       middle_02_l_ctrl.Rotation.Z {type=MovieSceneScriptingFloatChannel, default=0, const [-49.5772,29]}
       middle_02_l_ctrl.Scale.X {type=MovieSceneScriptingFloatChannel, default=1, const [1,29]}
       middle_02_l_ctrl.Scale.Y {type=MovieSceneScriptingFloatChannel, default=1, const [1,29]}
       middle_02_l_ctrl.Scale.Z {type=MovieSceneScriptingFloatChannel, default=1, const [1,29]}
       middle_03_l_ctrl.Location.X {type=MovieSceneScriptingFloatChannel, default=0, const [0,29]}
       middle_03_l_ctrl.Location.Y {type=MovieSceneScriptingFloatChannel, default=0, const [0,29]}
       middle_03_l_ctrl.Location.Z {type=MovieSceneScriptingFloatChannel, default=0, const [0,29]}
       middle_03_l_ctrl.Rotation.X {type=MovieSceneScriptingFloatChannel, default=0, const [0,29]}
       middle_03_l_ctrl.Rotation.Y {type=MovieSceneScriptingFloatChannel, default=0, thinned_from=29, keys [[-1,0],[389,0]]}
       middle_03_l_ctrl.Rotation.Z {type=MovieSceneScriptingFloatChannel, default=0, const [-102.5714,29]}
       middle_03_l_ctrl.Scale.X {type=MovieSceneScriptingFloatChannel, default=1, const [1,29]}
       middle_03_l_ctrl.Scale.Y {type=MovieSceneScriptingFloatChannel, default=1, const [1,29]}
       middle_03_l_ctrl.Scale.Z {type=MovieSceneScriptingFloatChannel, default=1, const [1,29]}
       ring_metacarpal_l_ctrl.Location.X {type=MovieSceneScriptingFloatChannel, default=0, const [0,29]}
       ring_metacarpal_l_ctrl.Location.Y {type=MovieSceneScriptingFloatChannel, default=0, const [0,29]}
       ring_metacarpal_l_ctrl.Location.Z {type=MovieSceneScriptingFloatChannel, default=0, const [0,29]}
       ring_metacarpal_l_ctrl.Rotation.X {type=MovieSceneScriptingFloatChannel, default=0, thinned_from=29, keys [[-1,1.2867],[389,1.2867]]}
       ring_metacarpal_l_ctrl.Rotation.Y {type=MovieSceneScriptingFloatChannel, default=0, thinned_from=29, keys [[-1,0],[389,0]]}
       ring_metacarpal_l_ctrl.Rotation.Z {type=MovieSceneScriptingFloatChannel, default=0, thinned_from=29, keys [[-1,-7.3778],[389,-7.3778]]}
       ring_metacarpal_l_ctrl.Scale.X {type=MovieSceneScriptingFloatChannel, default=1, const [1,29]}
       ring_metacarpal_l_ctrl.Scale.Y {type=MovieSceneScriptingFloatChannel, default=1, const [1,29]}
       ring_metacarpal_l_ctrl.Scale.Z {type=MovieSceneScriptingFloatChannel, default=1, const [1,29]}
       ring_01_l_ctrl.Location.X {type=MovieSceneScriptingFloatChannel, default=0, const [0,31]}
       ring_01_l_ctrl.Location.Y {type=MovieSceneScriptingFloatChannel, default=0, const [0,31]}
       ring_01_l_ctrl.Location.Z {type=MovieSceneScriptingFloatChannel, default=0, const [0,31]}
       ring_01_l_ctrl.Rotation.X {type=MovieSceneScriptingFloatChannel, default=0, thinned_from=31, keys [[-1,0],[217,0],[223,2.5974],[234,0],[389,0]]}
       ring_01_l_ctrl.Rotation.Y {type=MovieSceneScriptingFloatChannel, default=0, thinned_from=31, keys [[-1,0],[217,0],[223,1.7707],[234,0],[389,0]]}
       ring_01_l_ctrl.Rotation.Z {type=MovieSceneScriptingFloatChannel, default=0, const [-64.9002,32]}
       ring_01_l_ctrl.Scale.X {type=MovieSceneScriptingFloatChannel, default=1, const [1,31]}
       ring_01_l_ctrl.Scale.Y {type=MovieSceneScriptingFloatChannel, default=1, const [1,31]}
       ring_01_l_ctrl.Scale.Z {type=MovieSceneScriptingFloatChannel, default=1, const [1,31]}
       ring_02_l_ctrl.Location.X {type=MovieSceneScriptingFloatChannel, default=0, const [0,29]}
       ring_02_l_ctrl.Location.Y {type=MovieSceneScriptingFloatChannel, default=0, const [0,29]}
       ring_02_l_ctrl.Location.Z {type=MovieSceneScriptingFloatChannel, default=0, const [0,29]}
       ring_02_l_ctrl.Rotation.X {type=MovieSceneScriptingFloatChannel, default=0, thinned_from=29, keys [[-1,0],[389,0]]}
       ring_02_l_ctrl.Rotation.Y {type=MovieSceneScriptingFloatChannel, default=0, thinned_from=29, keys [[-1,0],[389,0]]}
       ring_02_l_ctrl.Rotation.Z {type=MovieSceneScriptingFloatChannel, default=0, const [-49.5772,29]}
       ring_02_l_ctrl.Scale.X {type=MovieSceneScriptingFloatChannel, default=1, const [1,29]}
       ring_02_l_ctrl.Scale.Y {type=MovieSceneScriptingFloatChannel, default=1, const [1,29]}
       ring_02_l_ctrl.Scale.Z {type=MovieSceneScriptingFloatChannel, default=1, const [1,29]}
       ring_03_l_ctrl.Location.X {type=MovieSceneScriptingFloatChannel, default=0, const [0,29]}
       ring_03_l_ctrl.Location.Y {type=MovieSceneScriptingFloatChannel, default=0, const [0,29]}
       ring_03_l_ctrl.Location.Z {type=MovieSceneScriptingFloatChannel, default=0, const [0,29]}
       ring_03_l_ctrl.Rotation.X {type=MovieSceneScriptingFloatChannel, default=0, thinned_from=29, keys [[-1,0],[389,0]]}
       ring_03_l_ctrl.Rotation.Y {type=MovieSceneScriptingFloatChannel, default=0, thinned_from=29, keys [[-1,0],[389,0]]}
       ring_03_l_ctrl.Rotation.Z {type=MovieSceneScriptingFloatChannel, default=0, const [-102.5714,29]}
       ring_03_l_ctrl.Scale.X {type=MovieSceneScriptingFloatChannel, default=1, const [1,29]}
       ring_03_l_ctrl.Scale.Y {type=MovieSceneScriptingFloatChannel, default=1, const [1,29]}
       ring_03_l_ctrl.Scale.Z {type=MovieSceneScriptingFloatChannel, default=1, const [1,29]}
       pinky_metacarpal_l_ctrl.Location.X {type=MovieSceneScriptingFloatChannel, default=0, const [0,29]}
       pinky_metacarpal_l_ctrl.Location.Y {type=MovieSceneScriptingFloatChannel, default=0, const [0,29]}
       pinky_metacarpal_l_ctrl.Location.Z {type=MovieSceneScriptingFloatChannel, default=0, const [0,29]}
       pinky_metacarpal_l_ctrl.Rotation.X {type=MovieSceneScriptingFloatChannel, default=0, thinned_from=29, keys [[-1,1.2867],[389,1.2867]]}
       pinky_metacarpal_l_ctrl.Rotation.Y {type=MovieSceneScriptingFloatChannel, default=0, const [0,29]}
       pinky_metacarpal_l_ctrl.Rotation.Z {type=MovieSceneScriptingFloatChannel, default=0, thinned_from=29, keys [[-1,-7.3778],[389,-7.3778]]}
       pinky_metacarpal_l_ctrl.Scale.X {type=MovieSceneScriptingFloatChannel, default=1, const [1,29]}
       pinky_metacarpal_l_ctrl.Scale.Y {type=MovieSceneScriptingFloatChannel, default=1, const [1,29]}
       pinky_metacarpal_l_ctrl.Scale.Z {type=MovieSceneScriptingFloatChannel, default=1, const [1,29]}
       pinky_01_l_ctrl.Location.X {type=MovieSceneScriptingFloatChannel, default=0, const [0,31]}
       pinky_01_l_ctrl.Location.Y {type=MovieSceneScriptingFloatChannel, default=0, const [0,31]}
       pinky_01_l_ctrl.Location.Z {type=MovieSceneScriptingFloatChannel, default=0, const [0,31]}
       pinky_01_l_ctrl.Rotation.X {type=MovieSceneScriptingFloatChannel, default=0, thinned_from=31, keys [[-1,0],[217,0],[223,2.1541],[234,0],[389,0]]}
       pinky_01_l_ctrl.Rotation.Y {type=MovieSceneScriptingFloatChannel, default=0, thinned_from=31, keys [[-1,0],[217,0],[223,2.2895],[234,0],[389,0]]}
       pinky_01_l_ctrl.Rotation.Z {
        type=MovieSceneScriptingFloatChannel
        default=0
        thinned_from=32
        keys [[-1,-64.9002],[210,-64.9002],[217,-51.8908],[223,-55.7053],[228,-73.8488],[234,-71.309],[242,-65.8065],[253,-65.0135],[267,-64.9002],[389,-64.9002]]
       }
       pinky_01_l_ctrl.Scale.X {type=MovieSceneScriptingFloatChannel, default=1, const [1,31]}
       pinky_01_l_ctrl.Scale.Y {type=MovieSceneScriptingFloatChannel, default=1, const [1,31]}
       pinky_01_l_ctrl.Scale.Z {type=MovieSceneScriptingFloatChannel, default=1, const [1,31]}
       pinky_02_l_ctrl.Location.X {type=MovieSceneScriptingFloatChannel, default=0, const [0,29]}
       pinky_02_l_ctrl.Location.Y {type=MovieSceneScriptingFloatChannel, default=0, const [0,29]}
       pinky_02_l_ctrl.Location.Z {type=MovieSceneScriptingFloatChannel, default=0, const [0,29]}
       pinky_02_l_ctrl.Rotation.X {type=MovieSceneScriptingFloatChannel, default=0, const [0,29]}
       pinky_02_l_ctrl.Rotation.Y {type=MovieSceneScriptingFloatChannel, default=0, thinned_from=29, keys [[-1,0],[389,0]]}
       pinky_02_l_ctrl.Rotation.Z {type=MovieSceneScriptingFloatChannel, default=0, const [-49.5772,29]}
       pinky_02_l_ctrl.Scale.X {type=MovieSceneScriptingFloatChannel, default=1, const [1,29]}
       pinky_02_l_ctrl.Scale.Y {type=MovieSceneScriptingFloatChannel, default=1, const [1,29]}
       pinky_02_l_ctrl.Scale.Z {type=MovieSceneScriptingFloatChannel, default=1, const [1,29]}
       pinky_03_l_ctrl.Location.X {type=MovieSceneScriptingFloatChannel, default=0, const [0,29]}
       pinky_03_l_ctrl.Location.Y {type=MovieSceneScriptingFloatChannel, default=0, const [0,29]}
       pinky_03_l_ctrl.Location.Z {type=MovieSceneScriptingFloatChannel, default=0, const [0,29]}
       pinky_03_l_ctrl.Rotation.X {type=MovieSceneScriptingFloatChannel, default=0, thinned_from=29, keys [[-1,0],[389,0]]}
       pinky_03_l_ctrl.Rotation.Y {type=MovieSceneScriptingFloatChannel, default=0, thinned_from=29, keys [[-1,0],[389,0]]}
       pinky_03_l_ctrl.Rotation.Z {type=MovieSceneScriptingFloatChannel, default=0, const [-102.5714,29]}
       pinky_03_l_ctrl.Scale.X {type=MovieSceneScriptingFloatChannel, default=1, const [1,29]}
       pinky_03_l_ctrl.Scale.Y {type=MovieSceneScriptingFloatChannel, default=1, const [1,29]}
       pinky_03_l_ctrl.Scale.Z {type=MovieSceneScriptingFloatChannel, default=1, const [1,29]}
       hand_l_ik_ctrl.Location.X {
        type=MovieSceneScriptingFloatChannel
        default=78.0535
        thinned_from=29
        keys [[-1,73.9728],[21,73.6225],[135,73.4226],[146,72.8404],[172,73.0281],[178,73.246],[183,73.616],[190,73.3079],[221,73.0136],[225,73.124],[230,72.6695],[234,72.518],[272,72.5461],[294,73.4577],[310,73.9728],[331,73.6225],[388,73.5057],[389,73.9728]]
       }
       hand_l_ik_ctrl.Location.Y {
        type=MovieSceneScriptingFloatChannel
        default=81.7171
        thinned_from=29
        keys [[-1,81.9303],[21,81.5081],[112,81.3674],[135,81.5937],[146,82.1175],[172,81.9074],[178,82.0173],[183,82.4691],[201,81.8752],[221,81.654],[230,82.421],[234,82.5879],[245,82.2679],[310,81.9303],[331,81.5081],[388,81.3674],[389,81.9303]]
       }
       hand_l_ik_ctrl.Location.Z {
        type=MovieSceneScriptingFloatChannel
        default=-17.9785
        thinned_from=29
        keys [[-1,-22.4069],[21,-22.8083],[46,-22.9421],[112,-22.9421],[135,-22.9913],[146,-23.3356],[160,-23.6854],[178,-23.9319],[183,-23.8109],[190,-23.1854],[201,-23.2992],[221,-23.3562],[225,-23.506],[230,-23.6303],[234,-23.6717],[245,-23.5334],[272,-23.4858],[294,-22.7964],[310,-22.4069],[331,-22.8083],[357,-22.9421],[388,-22.9421],[389,-22.4069]]
       }
       hand_l_ik_ctrl.Rotation.X {
        type=MovieSceneScriptingFloatChannel
        default=-155.8948
        thinned_from=28
        keys [[-1,-155.8948],[135,-155.7364],[146,-154.6271],[160,-151.9582],[172,-150.8239],[178,-149.798],[183,-150.4801],[201,-154.661],[221,-156.7515],[272,-156.7299],[310,-155.8948],[389,-155.8948]]
       }
       hand_l_ik_ctrl.Rotation.Y {type=MovieSceneScriptingFloatChannel, default=11.7099, const [11.7099,28]}
       hand_l_ik_ctrl.Rotation.Z {type=MovieSceneScriptingFloatChannel, default=-48.4099, const [-48.4099,28]}
       hand_l_ik_ctrl.Scale.X {type=MovieSceneScriptingFloatChannel, default=1, const [1,28]}
       hand_l_ik_ctrl.Scale.Y {type=MovieSceneScriptingFloatChannel, default=1, const [1,28]}
       hand_l_ik_ctrl.Scale.Z {type=MovieSceneScriptingFloatChannel, default=1, const [1,28]}
       Softness {type=MovieSceneScriptingFloatChannel, default=0, const [0,28]}
       hand_l_ikorient_ctrl.Rotation.Roll {type=MovieSceneScriptingFloatChannel, default=0, const [0,29]}
       hand_l_ikorient_ctrl.Rotation.Pitch {type=MovieSceneScriptingFloatChannel, default=0, const [0,29]}
       hand_l_ikorient_ctrl.Rotation.Yaw {type=MovieSceneScriptingFloatChannel, default=0, const [0,29]}
       arm_l_pv_ik_ctrl.Location.X {
        type=MovieSceneScriptingFloatChannel
        default=0
        thinned_from=32
        keys [[-1,9.2632],[21,9.2632],[46,8.8102],[60,8.8298],[78,8.9672],[84,8.8495],[92,8.4636],[112,7.7884],[119,7.5736],[135,7.1148],[146,6.6191],[160,8.0488],[172,8.2536],[178,8.6913],[183,13.7664],[190,16.0058],[210,18.3758],[217,18.6513],[221,18.6513],[234,12.724],[245,10.9951],[253,10.7482],[272,10.7482],[294,10.2403],[310,9.2632],[331,9.2632],[357,8.8102],[388,8.8102],[389,9.2632]]
       }
       arm_l_pv_ik_ctrl.Location.Y {
        type=MovieSceneScriptingFloatChannel
        default=0
        thinned_from=32
        keys [[-1,16.3996],[46,16.383],[60,16.4811],[78,17.1679],[84,17.1537],[92,16.7018],[112,16.3715],[135,16.0886],[139,15.7207],[146,14.6322],[160,13.0041],[172,12.595],[178,13.0286],[183,14.8977],[190,17.3339],[210,17.2169],[217,17.5254],[221,17.5254],[234,21.5085],[245,22.6702],[253,22.8362],[272,22.8362],[294,18.0088],[310,16.3996],[388,16.383],[389,16.3996]]
       }
       arm_l_pv_ik_ctrl.Location.Z {
        type=MovieSceneScriptingFloatChannel
        default=0
        thinned_from=32
        keys [[-1,-3.7175],[21,-3.7175],[46,-3.331],[60,-3.3768],[78,-3.6977],[84,-3.6944],[92,-3.6275],[112,-2.8583],[119,-2.6655],[135,-2.3332],[139,-2.3309],[146,-1.7278],[160,2.2311],[172,2.8942],[178,2.6683],[183,0.7763],[190,-2.3585],[210,-3.3217],[217,-3.4887],[221,-3.4887],[234,-9.6229],[245,-11.412],[253,-11.6676],[272,-11.6676],[294,-5.705],[310,-3.7175],[331,-3.7175],[357,-3.331],[388,-3.331],[389,-3.7175]]
       }
       arm_l_pv_ik_ctrl.Rotation.X {type=MovieSceneScriptingFloatChannel, default=0, const [0,33]}
       arm_l_pv_ik_ctrl.Rotation.Y {type=MovieSceneScriptingFloatChannel, default=0, const [0,33]}
       arm_l_pv_ik_ctrl.Rotation.Z {type=MovieSceneScriptingFloatChannel, default=0, const [0,33]}
       arm_l_pv_ik_ctrl.Scale.X {type=MovieSceneScriptingFloatChannel, default=1, const [1,32]}
       arm_l_pv_ik_ctrl.Scale.Y {type=MovieSceneScriptingFloatChannel, default=1, const [1,32]}
       arm_l_pv_ik_ctrl.Scale.Z {type=MovieSceneScriptingFloatChannel, default=1, const [1,32]}
       upperarm_r_fk_ctrl.Location.X {type=MovieSceneScriptingFloatChannel, default=0, thinned_from=32, keys [[-1,0],[390,0]]}
       upperarm_r_fk_ctrl.Location.Y {type=MovieSceneScriptingFloatChannel, default=0, thinned_from=32, keys [[-1,0],[390,0]]}
       upperarm_r_fk_ctrl.Location.Z {type=MovieSceneScriptingFloatChannel, default=0, thinned_from=32, keys [[-1,0],[390,0]]}
       upperarm_r_fk_ctrl.Rotation.X {
        type=MovieSceneScriptingFloatChannel
        default=0
        thinned_from=32
        keys [[-1,36.8644],[0,28.9205],[21,30.3151],[46,30.9465],[60,30.9727],[84,31.1561],[156,31.1561],[159,34.677],[172,34.2666],[178,44.9168],[183,43.8259],[190,33.18],[210,32.4747],[216,32.641],[234,31.9769],[245,31.4408],[253,31.3642],[267,30.9562],[294,30.3093],[310,28.9205],[357,31.1561],[376,29.8252],[388,37.9288],[390,29.7976]]
       }
       upperarm_r_fk_ctrl.Rotation.Y {
        type=MovieSceneScriptingFloatChannel
        default=0
        thinned_from=32
        keys [[-1,-3.5649],[0,-17.3824],[21,-18.0721],[46,-18.3327],[60,-18.2961],[84,-18.0396],[156,-18.0396],[159,-24.3962],[160,-24.4352],[172,-24.3929],[178,-30.6788],[183,-6.2727],[190,-0.0441],[210,0.0195],[216,0.4489],[217,0.4293],[225,0.4098],[234,0.3991],[245,0.4886],[253,0.5014],[267,0.65],[294,1.1773],[310,-17.3824],[357,-18.0396],[376,1.8751],[388,-5.4686],[390,2.0736]]
       }
       upperarm_r_fk_ctrl.Rotation.Z {
        type=MovieSceneScriptingFloatChannel
        default=0
        thinned_from=32
        keys [[-1,10.4015],[0,0.867],[21,-0.0407],[46,-0.2117],[60,-0.1895],[84,-0.0344],[156,-0.0344],[159,0.5777],[160,0.552],[172,0.468],[178,-1.999],[183,8.1661],[190,10.1577],[210,9.8991],[216,10.0446],[225,9.8933],[234,9.7625],[245,9.5598],[253,9.5309],[267,9.3758],[294,9.1443],[310,0.867],[357,-0.0344],[376,8.9757],[388,9.2936],[390,8.9832]]
       }
       upperarm_r_fk_ctrl.Scale.X {type=MovieSceneScriptingFloatChannel, default=1, const [1,32]}
       upperarm_r_fk_ctrl.Scale.Y {type=MovieSceneScriptingFloatChannel, default=1, const [1,32]}
       upperarm_r_fk_ctrl.Scale.Z {type=MovieSceneScriptingFloatChannel, default=1, const [1,32]}
       lowerarm_r_fk_ctrl.Location.X {type=MovieSceneScriptingFloatChannel, default=0, thinned_from=32, keys [[-1,0],[21,0],[46,0.0004],[60,0.0004],[92,0],[390,0]]}
       lowerarm_r_fk_ctrl.Location.Y {type=MovieSceneScriptingFloatChannel, default=0, thinned_from=32, keys [[-1,0],[21,0],[46,0.0008],[60,0.0007],[84,0],[390,0]]}
       lowerarm_r_fk_ctrl.Location.Z {type=MovieSceneScriptingFloatChannel, default=0, thinned_from=32, keys [[-1,0],[21,0],[46,-0.0008],[60,-0.0007],[84,0],[390,0]]}
       lowerarm_r_fk_ctrl.Rotation.X {
        type=MovieSceneScriptingFloatChannel
        default=0
        thinned_from=32
        keys [[-1,0.4407],[0,0.4431],[46,0.438],[60,0.4383],[84,0.4405],[172,0.4405],[178,0.4534],[183,0.4381],[190,0.4253],[294,0.4253],[310,0.4431],[357,0.4405],[376,0.4253],[388,0.4377],[390,0.4253]]
       }
       lowerarm_r_fk_ctrl.Rotation.Y {
        type=MovieSceneScriptingFloatChannel
        default=0
        thinned_from=32
        keys [[-1,-0.1312],[0,-0.1226],[21,-0.1296],[46,-0.1324],[60,-0.1324],[84,-0.1318],[172,-0.1318],[178,-0.0962],[183,-0.1394],[190,-0.1748],[294,-0.1748],[310,-0.1226],[357,-0.1318],[376,-0.1748],[388,-0.1407],[390,-0.1748]]
       }
       lowerarm_r_fk_ctrl.Rotation.Z {
        type=MovieSceneScriptingFloatChannel
        default=0
        thinned_from=32
        keys [[-1,93.7001],[0,92.5934],[21,93.4921],[46,93.7828],[172,93.7808],[178,89.0582],[183,94.7728],[190,99.4643],[294,99.4643],[310,92.5934],[357,93.7808],[376,99.4643],[388,94.9391],[390,99.4643]]
       }
       lowerarm_r_fk_ctrl.Scale.X {type=MovieSceneScriptingFloatChannel, default=1, const [1,32]}
       lowerarm_r_fk_ctrl.Scale.Y {type=MovieSceneScriptingFloatChannel, default=1, const [1,32]}
       lowerarm_r_fk_ctrl.Scale.Z {type=MovieSceneScriptingFloatChannel, default=1, const [1,32]}
       hand_r_fk_ctrl.Location.X {type=MovieSceneScriptingFloatChannel, default=0, thinned_from=28, keys [[-1,0],[388,0]]}
       hand_r_fk_ctrl.Location.Y {type=MovieSceneScriptingFloatChannel, default=0, thinned_from=28, keys [[-1,-0.1788],[388,-0.1788]]}
       hand_r_fk_ctrl.Location.Z {type=MovieSceneScriptingFloatChannel, default=0, thinned_from=28, keys [[-1,0.0249],[388,0.0249]]}
       hand_r_fk_ctrl.Rotation.X {
        type=MovieSceneScriptingFloatChannel
        default=0
        thinned_from=28
        keys [[-1,8.733],[0,15.4458],[21,17.1941],[46,17.6551],[119,17.4841],[146,17.4841],[159,18.4871],[172,18.5183],[178,25.1614],[183,11.0841],[190,4.3142],[210,4.3297],[216,4.2604],[217,4.2758],[245,4.3204],[253,4.3223],[267,4.3378],[294,4.3533],[310,15.4458],[357,17.4841],[376,4.3844],[388,11.5562]]
       }
       hand_r_fk_ctrl.Rotation.Y {
        type=MovieSceneScriptingFloatChannel
        default=0
        thinned_from=28
        keys [[-1,22.9898],[0,28.7142],[146,28.6554],[159,29.2919],[172,29.3103],[178,37.3661],[183,25.7219],[190,18.0528],[210,18.0622],[217,18.0297],[294,18.0763],[310,28.7142],[357,28.6554],[376,18.0949],[388,23.2367]]
       }
       hand_r_fk_ctrl.Rotation.Z {
        type=MovieSceneScriptingFloatChannel
        default=0
        thinned_from=28
        keys [[-1,-37.4272],[0,-19.4242],[21,-19.0223],[46,-19.2434],[60,-19.3039],[119,-19.7271],[146,-19.7271],[159,-15.8438],[160,-15.7846],[172,-15.7254],[178,-16.7951],[183,-33.1711],[190,-32.344],[210,-32.2848],[216,-32.5498],[217,-32.4906],[267,-32.254],[294,-32.1949],[310,-19.4242],[357,-19.7271],[376,-32.0765],[388,-36.1292]]
       }
       hand_r_fk_ctrl.Scale.X {type=MovieSceneScriptingFloatChannel, default=1, const [1,28]}
       hand_r_fk_ctrl.Scale.Y {type=MovieSceneScriptingFloatChannel, default=1, const [1,28]}
       hand_r_fk_ctrl.Scale.Z {type=MovieSceneScriptingFloatChannel, default=1, const [1,28]}
       thumb_01_r_ctrl.Location.X {type=MovieSceneScriptingFloatChannel, default=0, thinned_from=8, keys [[-1,0],[357,0]]}
       thumb_01_r_ctrl.Location.Y {type=MovieSceneScriptingFloatChannel, default=0, thinned_from=8, keys [[-1,0],[357,0]]}
       thumb_01_r_ctrl.Location.Z {type=MovieSceneScriptingFloatChannel, default=0, thinned_from=8, keys [[-1,0],[357,0]]}
       thumb_01_r_ctrl.Rotation.X {type=MovieSceneScriptingFloatChannel, default=0, const [-11.8305,8]}
       thumb_01_r_ctrl.Rotation.Y {type=MovieSceneScriptingFloatChannel, default=0, thinned_from=8, keys [[-1,-6.959],[357,-6.959]]}
       thumb_01_r_ctrl.Rotation.Z {type=MovieSceneScriptingFloatChannel, default=0, thinned_from=8, keys [[-1,4.4404],[357,4.4404]]}
       thumb_01_r_ctrl.Scale.X {type=MovieSceneScriptingFloatChannel, default=1, const [1,8]}
       thumb_01_r_ctrl.Scale.Y {type=MovieSceneScriptingFloatChannel, default=1, const [1,8]}
       thumb_01_r_ctrl.Scale.Z {type=MovieSceneScriptingFloatChannel, default=1, const [1,8]}
       thumb_02_r_ctrl.Location.X {type=MovieSceneScriptingFloatChannel, default=0, thinned_from=8, keys [[-1,0],[357,0]]}
       thumb_02_r_ctrl.Location.Y {type=MovieSceneScriptingFloatChannel, default=0, thinned_from=8, keys [[-1,0],[357,0]]}
       thumb_02_r_ctrl.Location.Z {type=MovieSceneScriptingFloatChannel, default=0, thinned_from=8, keys [[-1,0],[357,0]]}
       thumb_02_r_ctrl.Rotation.X {type=MovieSceneScriptingFloatChannel, default=0, thinned_from=8, keys [[-1,0],[357,0]]}
       thumb_02_r_ctrl.Rotation.Y {type=MovieSceneScriptingFloatChannel, default=0, thinned_from=8, keys [[-1,0],[357,0]]}
       thumb_02_r_ctrl.Rotation.Z {type=MovieSceneScriptingFloatChannel, default=0, const [-12.3705,8]}
       thumb_02_r_ctrl.Scale.X {type=MovieSceneScriptingFloatChannel, default=1, const [1,8]}
       thumb_02_r_ctrl.Scale.Y {type=MovieSceneScriptingFloatChannel, default=1, const [1,8]}
       thumb_02_r_ctrl.Scale.Z {type=MovieSceneScriptingFloatChannel, default=1, const [1,8]}
       thumb_03_r_ctrl.Location.X {type=MovieSceneScriptingFloatChannel, default=0, thinned_from=8, keys [[-1,0.5596],[357,0.5596]]}
       thumb_03_r_ctrl.Location.Y {type=MovieSceneScriptingFloatChannel, default=0, thinned_from=8, keys [[-1,0.2442],[357,0.2442]]}
       thumb_03_r_ctrl.Location.Z {type=MovieSceneScriptingFloatChannel, default=0, thinned_from=8, keys [[-1,0],[357,0]]}
       thumb_03_r_ctrl.Rotation.X {type=MovieSceneScriptingFloatChannel, default=0, thinned_from=8, keys [[-1,0],[357,0]]}
       thumb_03_r_ctrl.Rotation.Y {type=MovieSceneScriptingFloatChannel, default=0, thinned_from=8, keys [[-1,0],[357,0]]}
       thumb_03_r_ctrl.Rotation.Z {type=MovieSceneScriptingFloatChannel, default=0, const [-66.4234,8]}
       thumb_03_r_ctrl.Scale.X {type=MovieSceneScriptingFloatChannel, default=1, const [1,8]}
       thumb_03_r_ctrl.Scale.Y {type=MovieSceneScriptingFloatChannel, default=1, const [1,8]}
       thumb_03_r_ctrl.Scale.Z {type=MovieSceneScriptingFloatChannel, default=1, const [1,8]}
       index_metacarpal_r_ctrl.Location.X {type=MovieSceneScriptingFloatChannel, default=0, thinned_from=8, keys [[-1,0],[357,0]]}
       index_metacarpal_r_ctrl.Location.Y {type=MovieSceneScriptingFloatChannel, default=0, thinned_from=8, keys [[-1,0],[357,0]]}
       index_metacarpal_r_ctrl.Location.Z {type=MovieSceneScriptingFloatChannel, default=0, thinned_from=8, keys [[-1,0],[357,0]]}
       index_metacarpal_r_ctrl.Rotation.X {type=MovieSceneScriptingFloatChannel, default=0, const [-2.7256,8]}
       index_metacarpal_r_ctrl.Rotation.Y {type=MovieSceneScriptingFloatChannel, default=0, thinned_from=8, keys [[-1,0],[357,0]]}
       index_metacarpal_r_ctrl.Rotation.Z {type=MovieSceneScriptingFloatChannel, default=0, thinned_from=8, keys [[-1,-7.3778],[357,-7.3778]]}
       index_metacarpal_r_ctrl.Scale.X {type=MovieSceneScriptingFloatChannel, default=1, const [1,8]}
       index_metacarpal_r_ctrl.Scale.Y {type=MovieSceneScriptingFloatChannel, default=1, const [1,8]}
       index_metacarpal_r_ctrl.Scale.Z {type=MovieSceneScriptingFloatChannel, default=1, const [1,8]}
       index_01_r_ctrl.Location.X {type=MovieSceneScriptingFloatChannel, default=0, thinned_from=8, keys [[-1,0],[357,0]]}
       index_01_r_ctrl.Location.Y {type=MovieSceneScriptingFloatChannel, default=0, thinned_from=8, keys [[-1,0],[357,0]]}
       index_01_r_ctrl.Location.Z {type=MovieSceneScriptingFloatChannel, default=0, thinned_from=8, keys [[-1,0],[357,0]]}
       index_01_r_ctrl.Rotation.X {type=MovieSceneScriptingFloatChannel, default=0, thinned_from=8, keys [[-1,0],[357,0]]}
       index_01_r_ctrl.Rotation.Y {type=MovieSceneScriptingFloatChannel, default=0, thinned_from=8, keys [[-1,0],[357,0]]}
       index_01_r_ctrl.Rotation.Z {type=MovieSceneScriptingFloatChannel, default=0, const [-64.9002,8]}
       index_01_r_ctrl.Scale.X {type=MovieSceneScriptingFloatChannel, default=1, const [1,8]}
       index_01_r_ctrl.Scale.Y {type=MovieSceneScriptingFloatChannel, default=1, const [1,8]}
       index_01_r_ctrl.Scale.Z {type=MovieSceneScriptingFloatChannel, default=1, const [1,8]}
       index_02_r_ctrl.Location.X {type=MovieSceneScriptingFloatChannel, default=0, const [0,8]}
       index_02_r_ctrl.Location.Y {type=MovieSceneScriptingFloatChannel, default=0, const [0,8]}
       index_02_r_ctrl.Location.Z {type=MovieSceneScriptingFloatChannel, default=0, const [0,8]}
       index_02_r_ctrl.Rotation.X {type=MovieSceneScriptingFloatChannel, default=0, const [0,8]}
       index_02_r_ctrl.Rotation.Y {type=MovieSceneScriptingFloatChannel, default=0, const [0,8]}
       index_02_r_ctrl.Rotation.Z {type=MovieSceneScriptingFloatChannel, default=0, const [-49.5772,8]}
       index_02_r_ctrl.Scale.X {type=MovieSceneScriptingFloatChannel, default=1, const [1,8]}
       index_02_r_ctrl.Scale.Y {type=MovieSceneScriptingFloatChannel, default=1, const [1,8]}
       index_02_r_ctrl.Scale.Z {type=MovieSceneScriptingFloatChannel, default=1, const [1,8]}
       index_03_r_ctrl.Location.X {type=MovieSceneScriptingFloatChannel, default=0, const [0,8]}
       index_03_r_ctrl.Location.Y {type=MovieSceneScriptingFloatChannel, default=0, const [0,8]}
       index_03_r_ctrl.Location.Z {type=MovieSceneScriptingFloatChannel, default=0, const [0,8]}
       index_03_r_ctrl.Rotation.X {type=MovieSceneScriptingFloatChannel, default=0, thinned_from=8, keys [[-1,0],[357,0]]}
       index_03_r_ctrl.Rotation.Y {type=MovieSceneScriptingFloatChannel, default=0, thinned_from=8, keys [[-1,0],[357,0]]}
       index_03_r_ctrl.Rotation.Z {type=MovieSceneScriptingFloatChannel, default=0, const [-102.5714,8]}
       index_03_r_ctrl.Scale.X {type=MovieSceneScriptingFloatChannel, default=1, const [1,8]}
       index_03_r_ctrl.Scale.Y {type=MovieSceneScriptingFloatChannel, default=1, const [1,8]}
       index_03_r_ctrl.Scale.Z {type=MovieSceneScriptingFloatChannel, default=1, const [1,8]}
       middle_metacarpal_r_ctrl.Location.X {type=MovieSceneScriptingFloatChannel, default=0, const [0,8]}
       middle_metacarpal_r_ctrl.Location.Y {type=MovieSceneScriptingFloatChannel, default=0, const [0,8]}
       middle_metacarpal_r_ctrl.Location.Z {type=MovieSceneScriptingFloatChannel, default=0, const [0,8]}
       middle_metacarpal_r_ctrl.Rotation.X {type=MovieSceneScriptingFloatChannel, default=0, const [-2.7256,8]}
       middle_metacarpal_r_ctrl.Rotation.Y {type=MovieSceneScriptingFloatChannel, default=0, const [0,8]}
       middle_metacarpal_r_ctrl.Rotation.Z {type=MovieSceneScriptingFloatChannel, default=0, thinned_from=8, keys [[-1,-7.3778],[357,-7.3778]]}
       middle_metacarpal_r_ctrl.Scale.X {type=MovieSceneScriptingFloatChannel, default=1, const [1,8]}
       middle_metacarpal_r_ctrl.Scale.Y {type=MovieSceneScriptingFloatChannel, default=1, const [1,8]}
       middle_metacarpal_r_ctrl.Scale.Z {type=MovieSceneScriptingFloatChannel, default=1, const [1,8]}
       middle_01_r_ctrl.Location.X {type=MovieSceneScriptingFloatChannel, default=0, const [0,8]}
       middle_01_r_ctrl.Location.Y {type=MovieSceneScriptingFloatChannel, default=0, const [0,8]}
       middle_01_r_ctrl.Location.Z {type=MovieSceneScriptingFloatChannel, default=0, const [0,8]}
       middle_01_r_ctrl.Rotation.X {type=MovieSceneScriptingFloatChannel, default=0, thinned_from=8, keys [[-1,0],[357,0]]}
       middle_01_r_ctrl.Rotation.Y {type=MovieSceneScriptingFloatChannel, default=0, thinned_from=8, keys [[-1,0],[357,0]]}
       middle_01_r_ctrl.Rotation.Z {type=MovieSceneScriptingFloatChannel, default=0, const [-64.9002,8]}
       middle_01_r_ctrl.Scale.X {type=MovieSceneScriptingFloatChannel, default=1, const [1,8]}
       middle_01_r_ctrl.Scale.Y {type=MovieSceneScriptingFloatChannel, default=1, const [1,8]}
       middle_01_r_ctrl.Scale.Z {type=MovieSceneScriptingFloatChannel, default=1, const [1,8]}
       middle_02_r_ctrl.Location.X {type=MovieSceneScriptingFloatChannel, default=0, const [0,8]}
       middle_02_r_ctrl.Location.Y {type=MovieSceneScriptingFloatChannel, default=0, const [0,8]}
       middle_02_r_ctrl.Location.Z {type=MovieSceneScriptingFloatChannel, default=0, const [0,8]}
       middle_02_r_ctrl.Rotation.X {type=MovieSceneScriptingFloatChannel, default=0, const [0,8]}
       middle_02_r_ctrl.Rotation.Y {type=MovieSceneScriptingFloatChannel, default=0, const [0,8]}
       middle_02_r_ctrl.Rotation.Z {type=MovieSceneScriptingFloatChannel, default=0, const [-49.5772,8]}
       middle_02_r_ctrl.Scale.X {type=MovieSceneScriptingFloatChannel, default=1, const [1,8]}
       middle_02_r_ctrl.Scale.Y {type=MovieSceneScriptingFloatChannel, default=1, const [1,8]}
       middle_02_r_ctrl.Scale.Z {type=MovieSceneScriptingFloatChannel, default=1, const [1,8]}
       middle_03_r_ctrl.Location.X {type=MovieSceneScriptingFloatChannel, default=0, const [0,8]}
       middle_03_r_ctrl.Location.Y {type=MovieSceneScriptingFloatChannel, default=0, const [0,8]}
       middle_03_r_ctrl.Location.Z {type=MovieSceneScriptingFloatChannel, default=0, const [0,8]}
       middle_03_r_ctrl.Rotation.X {type=MovieSceneScriptingFloatChannel, default=0, thinned_from=8, keys [[-1,0],[357,0]]}
       middle_03_r_ctrl.Rotation.Y {type=MovieSceneScriptingFloatChannel, default=0, thinned_from=8, keys [[-1,0],[357,0]]}
       middle_03_r_ctrl.Rotation.Z {type=MovieSceneScriptingFloatChannel, default=0, const [-102.5714,8]}
       middle_03_r_ctrl.Scale.X {type=MovieSceneScriptingFloatChannel, default=1, const [1,8]}
       middle_03_r_ctrl.Scale.Y {type=MovieSceneScriptingFloatChannel, default=1, const [1,8]}
       middle_03_r_ctrl.Scale.Z {type=MovieSceneScriptingFloatChannel, default=1, const [1,8]}
       ring_metacarpal_r_ctrl.Location.X {type=MovieSceneScriptingFloatChannel, default=0, const [0,8]}
       ring_metacarpal_r_ctrl.Location.Y {type=MovieSceneScriptingFloatChannel, default=0, const [0,8]}
       ring_metacarpal_r_ctrl.Location.Z {type=MovieSceneScriptingFloatChannel, default=0, const [0,8]}
       ring_metacarpal_r_ctrl.Rotation.X {type=MovieSceneScriptingFloatChannel, default=0, thinned_from=8, keys [[-1,1.2867],[357,1.2867]]}
       ring_metacarpal_r_ctrl.Rotation.Y {type=MovieSceneScriptingFloatChannel, default=0, const [0,8]}
       ring_metacarpal_r_ctrl.Rotation.Z {type=MovieSceneScriptingFloatChannel, default=0, thinned_from=8, keys [[-1,-7.3778],[357,-7.3778]]}
       ring_metacarpal_r_ctrl.Scale.X {type=MovieSceneScriptingFloatChannel, default=1, const [1,8]}
       ring_metacarpal_r_ctrl.Scale.Y {type=MovieSceneScriptingFloatChannel, default=1, const [1,8]}
       ring_metacarpal_r_ctrl.Scale.Z {type=MovieSceneScriptingFloatChannel, default=1, const [1,8]}
       ring_01_r_ctrl.Location.X {type=MovieSceneScriptingFloatChannel, default=0, const [0,8]}
       ring_01_r_ctrl.Location.Y {type=MovieSceneScriptingFloatChannel, default=0, const [0,8]}
       ring_01_r_ctrl.Location.Z {type=MovieSceneScriptingFloatChannel, default=0, const [0,8]}
       ring_01_r_ctrl.Rotation.X {type=MovieSceneScriptingFloatChannel, default=0, thinned_from=8, keys [[-1,0],[357,0]]}
       ring_01_r_ctrl.Rotation.Y {type=MovieSceneScriptingFloatChannel, default=0, thinned_from=8, keys [[-1,0],[357,0]]}
       ring_01_r_ctrl.Rotation.Z {type=MovieSceneScriptingFloatChannel, default=0, const [-64.9002,8]}
       ring_01_r_ctrl.Scale.X {type=MovieSceneScriptingFloatChannel, default=1, const [1,8]}
       ring_01_r_ctrl.Scale.Y {type=MovieSceneScriptingFloatChannel, default=1, const [1,8]}
       ring_01_r_ctrl.Scale.Z {type=MovieSceneScriptingFloatChannel, default=1, const [1,8]}
       ring_02_r_ctrl.Location.X {type=MovieSceneScriptingFloatChannel, default=0, const [0,8]}
       ring_02_r_ctrl.Location.Y {type=MovieSceneScriptingFloatChannel, default=0, const [0,8]}
       ring_02_r_ctrl.Location.Z {type=MovieSceneScriptingFloatChannel, default=0, const [0,8]}
       ring_02_r_ctrl.Rotation.X {type=MovieSceneScriptingFloatChannel, default=0, thinned_from=8, keys [[-1,0],[357,0]]}
       ring_02_r_ctrl.Rotation.Y {type=MovieSceneScriptingFloatChannel, default=0, thinned_from=8, keys [[-1,0],[357,0]]}
       ring_02_r_ctrl.Rotation.Z {type=MovieSceneScriptingFloatChannel, default=0, const [-49.5772,8]}
       ring_02_r_ctrl.Scale.X {type=MovieSceneScriptingFloatChannel, default=1, const [1,8]}
       ring_02_r_ctrl.Scale.Y {type=MovieSceneScriptingFloatChannel, default=1, const [1,8]}
       ring_02_r_ctrl.Scale.Z {type=MovieSceneScriptingFloatChannel, default=1, const [1,8]}
       ring_03_r_ctrl.Location.X {type=MovieSceneScriptingFloatChannel, default=0, const [0,8]}
       ring_03_r_ctrl.Location.Y {type=MovieSceneScriptingFloatChannel, default=0, const [0,8]}
       ring_03_r_ctrl.Location.Z {type=MovieSceneScriptingFloatChannel, default=0, const [0,8]}
       ring_03_r_ctrl.Rotation.X {type=MovieSceneScriptingFloatChannel, default=0, const [0,8]}
       ring_03_r_ctrl.Rotation.Y {type=MovieSceneScriptingFloatChannel, default=0, const [0,8]}
       ring_03_r_ctrl.Rotation.Z {type=MovieSceneScriptingFloatChannel, default=0, const [-102.5714,8]}
       ring_03_r_ctrl.Scale.X {type=MovieSceneScriptingFloatChannel, default=1, const [1,8]}
       ring_03_r_ctrl.Scale.Y {type=MovieSceneScriptingFloatChannel, default=1, const [1,8]}
       ring_03_r_ctrl.Scale.Z {type=MovieSceneScriptingFloatChannel, default=1, const [1,8]}
       pinky_metacarpal_r_ctrl.Location.X {type=MovieSceneScriptingFloatChannel, default=0, const [0,8]}
       pinky_metacarpal_r_ctrl.Location.Y {type=MovieSceneScriptingFloatChannel, default=0, const [0,8]}
       pinky_metacarpal_r_ctrl.Location.Z {type=MovieSceneScriptingFloatChannel, default=0, const [0,8]}
       pinky_metacarpal_r_ctrl.Rotation.X {type=MovieSceneScriptingFloatChannel, default=0, thinned_from=8, keys [[-1,1.2867],[357,1.2867]]}
       pinky_metacarpal_r_ctrl.Rotation.Y {type=MovieSceneScriptingFloatChannel, default=0, thinned_from=8, keys [[-1,0],[357,0]]}
       pinky_metacarpal_r_ctrl.Rotation.Z {type=MovieSceneScriptingFloatChannel, default=0, thinned_from=8, keys [[-1,-7.3778],[357,-7.3778]]}
       pinky_metacarpal_r_ctrl.Scale.X {type=MovieSceneScriptingFloatChannel, default=1, const [1,8]}
       pinky_metacarpal_r_ctrl.Scale.Y {type=MovieSceneScriptingFloatChannel, default=1, const [1,8]}
       pinky_metacarpal_r_ctrl.Scale.Z {type=MovieSceneScriptingFloatChannel, default=1, const [1,8]}
       pinky_01_r_ctrl.Location.X {type=MovieSceneScriptingFloatChannel, default=0, const [0,8]}
       pinky_01_r_ctrl.Location.Y {type=MovieSceneScriptingFloatChannel, default=0, const [0,8]}
       pinky_01_r_ctrl.Location.Z {type=MovieSceneScriptingFloatChannel, default=0, const [0,8]}
       pinky_01_r_ctrl.Rotation.X {type=MovieSceneScriptingFloatChannel, default=0, thinned_from=8, keys [[-1,0],[357,0]]}
       pinky_01_r_ctrl.Rotation.Y {type=MovieSceneScriptingFloatChannel, default=0, thinned_from=8, keys [[-1,0],[357,0]]}
       pinky_01_r_ctrl.Rotation.Z {type=MovieSceneScriptingFloatChannel, default=0, const [-64.9002,8]}
       pinky_01_r_ctrl.Scale.X {type=MovieSceneScriptingFloatChannel, default=1, const [1,8]}
       pinky_01_r_ctrl.Scale.Y {type=MovieSceneScriptingFloatChannel, default=1, const [1,8]}
       pinky_01_r_ctrl.Scale.Z {type=MovieSceneScriptingFloatChannel, default=1, const [1,8]}
       pinky_02_r_ctrl.Location.X {type=MovieSceneScriptingFloatChannel, default=0, const [0,8]}
       pinky_02_r_ctrl.Location.Y {type=MovieSceneScriptingFloatChannel, default=0, const [0,8]}
       pinky_02_r_ctrl.Location.Z {type=MovieSceneScriptingFloatChannel, default=0, const [0,8]}
       pinky_02_r_ctrl.Rotation.X {type=MovieSceneScriptingFloatChannel, default=0, const [0,8]}
       pinky_02_r_ctrl.Rotation.Y {type=MovieSceneScriptingFloatChannel, default=0, const [0,8]}
       pinky_02_r_ctrl.Rotation.Z {type=MovieSceneScriptingFloatChannel, default=0, const [-49.5772,8]}
       pinky_02_r_ctrl.Scale.X {type=MovieSceneScriptingFloatChannel, default=1, const [1,8]}
       pinky_02_r_ctrl.Scale.Y {type=MovieSceneScriptingFloatChannel, default=1, const [1,8]}
       pinky_02_r_ctrl.Scale.Z {type=MovieSceneScriptingFloatChannel, default=1, const [1,8]}
       pinky_03_r_ctrl.Location.X {type=MovieSceneScriptingFloatChannel, default=0, const [0,8]}
       pinky_03_r_ctrl.Location.Y {type=MovieSceneScriptingFloatChannel, default=0, const [0,8]}
       pinky_03_r_ctrl.Location.Z {type=MovieSceneScriptingFloatChannel, default=0, const [0,8]}
       pinky_03_r_ctrl.Rotation.X {type=MovieSceneScriptingFloatChannel, default=0, const [0,8]}
       pinky_03_r_ctrl.Rotation.Y {type=MovieSceneScriptingFloatChannel, default=0, const [0,8]}
       pinky_03_r_ctrl.Rotation.Z {type=MovieSceneScriptingFloatChannel, default=0, const [-102.5714,8]}
       pinky_03_r_ctrl.Scale.X {type=MovieSceneScriptingFloatChannel, default=1, const [1,8]}
       pinky_03_r_ctrl.Scale.Y {type=MovieSceneScriptingFloatChannel, default=1, const [1,8]}
       pinky_03_r_ctrl.Scale.Z {type=MovieSceneScriptingFloatChannel, default=1, const [1,8]}
       hand_r_ik_ctrl.Location.X {
        type=MovieSceneScriptingFloatChannel
        default=0
        thinned_from=28
        keys [[-1,56.2707],[21,56.4709],[139,56.5146],[146,56.4059],[160,55.7175],[172,55.488],[178,55.4991],[183,55.6113],[190,56.6037],[210,56.9345],[225,56.6446],[234,55.818],[245,55.6071],[267,55.5425],[294,56.2707],[310,56.2707],[331,56.4709],[357,56.5376],[376,56.2707],[388,56.5376],[390,56.2707]]
       }
       hand_r_ik_ctrl.Location.Y {
        type=MovieSceneScriptingFloatChannel
        default=0
        thinned_from=28
        keys [[-1,76.2896],[146,76.4273],[210,76.684],[225,76.8653],[234,76.632],[267,76.5543],[294,76.2896],[390,76.2896]]
       }
       hand_r_ik_ctrl.Location.Z {
        type=MovieSceneScriptingFloatChannel
        default=0
        thinned_from=28
        keys [[-1,-13.235],[0,-13.235],[21,-13.8844],[46,-14.1009],[135,-14.1009],[146,-13.8802],[160,-13.9657],[183,-13.9915],[190,-13.6656],[210,-13.557],[234,-13.3294],[267,-13.2998],[294,-13.235],[310,-13.235],[331,-13.8844],[357,-14.1009],[376,-13.235],[388,-14.1009],[390,-13.235]]
       }
       hand_r_ik_ctrl.Rotation.X {
        type=MovieSceneScriptingFloatChannel
        default=0
        thinned_from=28
        keys [[-1,-197.1047],[225,-197.1047],[267,-197.6401],[294,-197.1047],[390,-197.1047]]
       }
       hand_r_ik_ctrl.Rotation.Y {
        type=MovieSceneScriptingFloatChannel
        default=0
        thinned_from=28
        keys [[-1,-12.6591],[225,-12.6591],[234,-17.2211],[245,-18.3854],[253,-18.5517],[267,-18.7418],[294,-12.6591],[390,-12.6591]]
       }
       hand_r_ik_ctrl.Rotation.Z {
        type=MovieSceneScriptingFloatChannel
        default=0
        thinned_from=28
        keys [[-1,-219.2809],[225,-219.2809],[234,-217.7992],[267,-217.3053],[294,-219.2809],[390,-219.2809]]
       }
       hand_r_ik_ctrl.Scale.X {type=MovieSceneScriptingFloatChannel, default=1, const [-1,28]}
       hand_r_ik_ctrl.Scale.Y {type=MovieSceneScriptingFloatChannel, default=1, const [1,28]}
       hand_r_ik_ctrl.Scale.Z {type=MovieSceneScriptingFloatChannel, default=1, const [1,28]}
       Softness_2 {type=MovieSceneScriptingFloatChannel, default=0, const [0,27]}
       hand_r_ikorient_ctrl.Rotation.Roll {type=MovieSceneScriptingFloatChannel, default=0, const [3.4754,29]}
       hand_r_ikorient_ctrl.Rotation.Pitch {type=MovieSceneScriptingFloatChannel, default=0, const [-17.5139,29]}
       hand_r_ikorient_ctrl.Rotation.Yaw {type=MovieSceneScriptingFloatChannel, default=0, const [35.6029,29]}
       arm_r_pv_ik_ctrl.Location.X {
        type=MovieSceneScriptingFloatChannel
        default=0
        thinned_from=30
        keys [[-1,2.1151],[0,-12.8312],[21,-12.8852],[84,-12.9165],[92,-12.1546],[112,-11.4185],[119,-11.2027],[135,-10.7478],[139,-10.3101],[146,-8.9179],[159,-9.0768],[160,-9.0765],[172,-9.4217],[178,-10.3223],[183,-9.0928],[190,-3.9502],[210,-5.8567],[225,-6.8838],[234,-11.2336],[245,-12.573],[253,-12.7643],[267,-13.031],[294,-11.9797],[310,-12.8312],[357,-12.9165],[376,1.3011],[388,2.0297],[390,2.1151]]
       }
       arm_r_pv_ik_ctrl.Location.Y {
        type=MovieSceneScriptingFloatChannel
        default=0
        thinned_from=30
        keys [[-1,-3.3084],[0,-16.3332],[21,-17.1036],[46,-17.5494],[84,-17.5494],[92,-17.2741],[119,-17.1927],[135,-17.1893],[139,-16.9595],[146,-15.9686],[159,-12.8654],[160,-12.8103],[172,-12.3986],[178,-12.6187],[183,-11.2464],[190,-5.6186],[210,-6.7837],[225,-7.9098],[234,-15.3969],[245,-17.9071],[253,-18.2657],[267,-18.7621],[294,-16.9404],[310,-16.3332],[357,-17.5494],[376,-4.7632],[388,-4.5246],[390,-3.3084]]
       }
       arm_r_pv_ik_ctrl.Location.Z {
        type=MovieSceneScriptingFloatChannel
        default=0
        thinned_from=30
        keys [[-1,-11.7493],[0,-33.5314],[46,-32.9475],[84,-32.9475],[92,-33.0099],[135,-33.6845],[139,-33.9448],[146,-34.5448],[159,-38.2758],[172,-39.4587],[178,-40.1159],[183,-36.3837],[190,-31.5295],[210,-30.3162],[225,-30.6811],[234,-33.3673],[245,-37.2837],[267,-38.849],[294,-34.8608],[310,-33.5314],[357,-32.9475],[376,-10.1817],[388,-11.1654],[390,-11.7493]]
       }
       arm_r_pv_ik_ctrl.Rotation.X {
        type=MovieSceneScriptingFloatChannel
        default=0
        thinned_from=30
        keys [[-1,13.7492],[0,18.1938],[21,19.1525],[46,19.7072],[84,19.7072],[92,19.7732],[119,19.5592],[135,19.4887],[139,19.4064],[146,19.1958],[159,15.5692],[172,14.8829],[178,14.7994],[183,14.9818],[190,16.2587],[210,17.1255],[225,17.5152],[234,18.345],[245,18.4864],[267,18.5297],[294,18.2778],[310,18.1938],[357,19.7072],[376,15.9655],[388,15.2626],[390,13.7492]]
       }
       arm_r_pv_ik_ctrl.Rotation.Y {
        type=MovieSceneScriptingFloatChannel
        default=0
        thinned_from=30
        keys [[-1,-7.8978],[0,-3.1678],[21,-3.2279],[46,-3.3067],[84,-3.3067],[92,-3.5192],[112,-4.3498],[119,-4.5605],[135,-4.8154],[139,-4.9428],[146,-5.2733],[159,-11.4042],[172,-12.2026],[178,-12.3861],[183,-11.8477],[190,-8.0787],[210,-5.4208],[225,-4.572],[234,-2.7761],[245,-2.4676],[253,-2.4236],[267,-2.3732],[294,-2.9691],[310,-3.1678],[357,-3.3067],[376,-6.7689],[388,-8.0367],[390,-7.8978]]
       }
       arm_r_pv_ik_ctrl.Rotation.Z {
        type=MovieSceneScriptingFloatChannel
        default=0
        thinned_from=30
        keys [[-1,5.8001],[0,-2.2732],[21,-3.4206],[46,-4.0844],[84,-4.0844],[92,-3.5222],[112,-3.2362],[119,-3.1849],[135,-3.0233],[139,-2.5702],[146,-1.7001],[159,3.372],[172,3.3385],[178,3.0284],[183,3.4599],[190,6.4802],[210,4.2154],[225,3.3254],[234,-1.6191],[245,-2.8364],[253,-3.0103],[267,-3.209],[294,-2.5072],[310,-2.2732],[357,-4.0844],[376,3.8281],[388,3.989],[390,5.8001]]
       }
       arm_r_pv_ik_ctrl.Scale.X {type=MovieSceneScriptingFloatChannel, default=1, const [1,30]}
       arm_r_pv_ik_ctrl.Scale.Y {type=MovieSceneScriptingFloatChannel, default=1, const [1,30]}
       arm_r_pv_ik_ctrl.Scale.Z {type=MovieSceneScriptingFloatChannel, default=1, const [1,30]}
       body_ctrl.Location.X {
        type=MovieSceneScriptingFloatChannel
        default=0
        keys [[0,-0.8509],[11,-1.5383],[12,-1.6362],[22,-2.3623],[47,-2.6043],[61,-2.5768],[79,-2.3846],[85,-2.6227],[93,-3.3369],[113,-4.323],[136,-5.0224],[140,-5.2504],[147,-5.8868],[161,-6.2069],[173,-5.5946],[179,-4.4839],[184,-0.8903],[191,1.6038],[211,3.1003],[217,3.4361],[221,3.628],[225,3.8589],[234,3.9762],[245,3.7563],[253,3.7249],[265,3.4471],[272,2.8068],[280,1.3309],[294,-0.2997],[310,-0.8509],[321,-1.6362],[331,-2.3623],[357,-2.6043],[376,-2.5741],[388,-4.4322],[389,-3.1161]]
       }
       body_ctrl.Location.Y {
        type=MovieSceneScriptingFloatChannel
        default=0
        thinned_from=36
        keys [[-1,2.1288],[0,2.5255],[11,2.5718],[21,2.4034],[46,2.3472],[84,2.3452],[112,2.325],[172,2.325],[178,2.2507],[183,2.1443],[190,2.4609],[208,2.6241],[265,2.6191],[272,2.6097],[294,2.5339],[310,2.5255],[321,2.5718],[331,2.4034],[357,2.3472],[376,2.1784],[388,1.9505],[389,2.1288]]
       }
       body_ctrl.Location.Z {
        type=MovieSceneScriptingFloatChannel
        default=0
        thinned_from=37
        keys [[-1,-12.8916],[0,-12.7283],[11,-12.6024],[21,-12.8602],[46,-12.9461],[84,-12.9721],[92,-13.0132],[112,-12.8738],[135,-12.7512],[139,-12.8551],[146,-12.7881],[160,-12.7535],[172,-12.8919],[178,-13.3389],[183,-13.2133],[190,-12.9199],[210,-12.8231],[221,-12.8905],[225,-13.1799],[234,-13.4138],[245,-13.4821],[265,-13.4757],[272,-13.5267],[280,-13.8902],[286,-13.6126],[294,-12.9673],[310,-12.7283],[321,-12.6024],[331,-12.8602],[357,-12.9461],[376,-12.8847],[388,-12.9567],[389,-12.8916]]
       }
       body_ctrl.Rotation.X {
        type=MovieSceneScriptingFloatChannel
        default=0
        thinned_from=35
        keys [[-1,11.4255],[0,11.9479],[21,13.1347],[46,13.2096],[78,13.2296],[84,13.2812],[92,13.436],[112,13.7102],[135,13.8149],[146,13.7742],[160,12.9384],[172,12.8731],[178,12.9113],[183,12.7747],[190,11.7445],[210,11.1265],[221,11.1131],[225,11.0195],[234,10.9857],[265,10.9899],[272,11.0118],[280,11.1739],[294,11.8291],[310,11.9479],[331,13.1347],[357,13.2096],[376,11.3953],[388,12.6872],[389,11.4255]]
       }
       body_ctrl.Rotation.Y {
        type=MovieSceneScriptingFloatChannel
        default=0
        keys [[-1,3.2221],[0,2.0376],[11,2.4179],[21,2.8836],[46,3.3268],[60,3.2442],[78,2.9965],[84,3.0351],[92,3.1509],[112,3.9116],[135,4.3103],[139,4.4262],[146,4.7815],[160,5.5326],[172,6.5478],[178,6.7753],[183,5.8682],[190,1.5546],[210,-1.0335],[217,-1.1268],[221,-1.358],[225,-2.2064],[234,-3.3884],[245,-3.7598],[253,-3.8129],[265,-3.5706],[272,-3.1301],[280,-1.7938],[294,1.4494],[310,2.0376],[331,2.8836],[357,3.3268],[376,2.751],[388,4.5114],[389,3.1803]]
       }
       body_ctrl.Rotation.Z {
        type=MovieSceneScriptingFloatChannel
        default=0
        thinned_from=35
        keys [[-1,26.0415],[0,28.2656],[21,28.4371],[60,28.8787],[78,29.3843],[84,29.3552],[92,28.7502],[135,28.1977],[146,26.8349],[160,25.6508],[172,25.4709],[178,25.6237],[183,24.5782],[190,26.9477],[210,28.3693],[221,28.9422],[225,29.2563],[234,33.2646],[245,34.4336],[253,34.6006],[265,34.0797],[272,32.574],[280,29.7464],[294,28.4929],[310,28.2656],[331,28.4371],[357,28.5269],[376,26.2763],[388,26.3028],[389,26.0415]]
       }
       body_ctrl.Scale.X {type=MovieSceneScriptingFloatChannel, default=1, const [1,36]}
       body_ctrl.Scale.Y {type=MovieSceneScriptingFloatChannel, default=1, const [1,36]}
       body_ctrl.Scale.Z {type=MovieSceneScriptingFloatChannel, default=1, const [1,36]}
       hips_ctrl.Location.X {type=MovieSceneScriptingFloatChannel, default=0, const [0,34]}
       hips_ctrl.Location.Y {type=MovieSceneScriptingFloatChannel, default=0, const [0,34]}
       hips_ctrl.Location.Z {type=MovieSceneScriptingFloatChannel, default=0, const [0,34]}
       hips_ctrl.Rotation.X {
        type=MovieSceneScriptingFloatChannel
        default=0
        thinned_from=33
        keys [[-1,-1.3105],[0,-1.2429],[21,-2.5344],[46,-2.9751],[92,-2.9751],[112,-3.193],[135,-3.2895],[139,-3.368],[146,-4.7799],[160,-4.9911],[178,-4.9911],[183,-5.2683],[217,-5.2683],[221,-5.4237],[234,-7.2962],[245,-7.6281],[253,-7.6756],[265,-6.889],[272,-5.8249],[280,-4.2463],[294,-2.0398],[310,-1.2429],[331,-2.5344],[357,-2.9751],[376,-1.2429],[388,-3.0426],[389,-1.2429]]
       }
       hips_ctrl.Rotation.Y {
        type=MovieSceneScriptingFloatChannel
        default=0
        thinned_from=33
        keys [[-1,0],[0,0],[21,-1.0841],[46,-1.454],[84,-1.454],[112,-2.8257],[139,-3.1585],[146,-3.306],[160,-3.7328],[172,-4.467],[178,-5.1425],[183,-4.1645],[190,-1.9543],[210,0],[310,0],[331,-1.0841],[357,-1.454],[376,0],[388,-1.454],[389,0]]
       }
       hips_ctrl.Rotation.Z {type=MovieSceneScriptingFloatChannel, default=0, thinned_from=34, keys [[-1,0],[389,0]]}
       hips_ctrl.Scale.X {type=MovieSceneScriptingFloatChannel, default=1, const [1,34]}
       hips_ctrl.Scale.Y {type=MovieSceneScriptingFloatChannel, default=1, const [1,34]}
       hips_ctrl.Scale.Z {type=MovieSceneScriptingFloatChannel, default=1, const [1,34]}
       hips_tan_ctrl.Location.X {
        type=MovieSceneScriptingFloatChannel
        default=0
        keys [[-1,-0.1598],[0,-0.2436],[21,-0.3749],[46,-0.4269],[60,-0.4269],[78,-0.6016],[84,-0.4274],[92,-0.4269],[112,-0.7729],[119,-0.7729],[135,-0.8066],[139,-0.8126],[146,-0.8869],[160,-0.338],[172,-0.338],[178,-1.1804],[183,-0.6309],[190,-0.0708],[210,-0.0689],[217,-0.0708],[221,0],[234,0.1116],[253,0.1174],[265,0.1166],[267,0.1174],[272,0],[294,-0.1613],[310,-0.2436],[331,-0.3305],[357,-0.4269],[376,-0.0202],[389,0]]
       }
       hips_tan_ctrl.Location.Y {
        type=MovieSceneScriptingFloatChannel
        default=0.0032
        thinned_from=32
        keys [[-1,0.4471],[0,0.4494],[21,0.6058],[46,0.6555],[60,0.6555],[78,1.4924],[84,0.6916],[92,0.6555],[112,1.4716],[135,1.4715],[139,1.5214],[146,1.722],[160,0.6714],[172,0.6714],[178,1.7814],[183,0.9446],[190,-0.1471],[210,0.9383],[217,-0.1471],[221,-0.1634],[234,-0.1624],[253,-0.1614],[267,-0.1614],[272,-0.1634],[294,-0.1629],[310,0.4494],[331,0.5542],[357,0.6555],[376,-0.1633],[389,-0.1634]]
       }
       hips_tan_ctrl.Location.Z {
        type=MovieSceneScriptingFloatChannel
        default=-0.0015
        thinned_from=32
        keys [[-1,-0.0003],[0,-0.0084],[21,-0.0063],[46,-0.0271],[60,-0.0271],[78,-0.0343],[84,-0.0263],[92,-0.0271],[112,-0.0429],[119,-0.0429],[135,-0.0455],[139,-0.0491],[146,-0.0753],[160,-0.022],[172,-0.022],[178,-0.122],[183,-0.0595],[190,-0.0052],[210,-0.0277],[217,-0.0052],[221,-0.0045],[234,-0.0073],[253,-0.0076],[267,-0.0076],[272,-0.0045],[294,-0.0104],[310,-0.0084],[331,-0.0147],[357,-0.0271],[376,-0.0053],[389,-0.0045]]
       }
       hips_tan_ctrl.Rotation.X {type=MovieSceneScriptingFloatChannel, default=0, const [0,32]}
       hips_tan_ctrl.Rotation.Y {type=MovieSceneScriptingFloatChannel, default=0, const [0,32]}
       hips_tan_ctrl.Rotation.Z {type=MovieSceneScriptingFloatChannel, default=0, const [0,32]}
       hips_tan_ctrl.Scale.X {type=MovieSceneScriptingFloatChannel, default=1, const [1,32]}
       hips_tan_ctrl.Scale.Y {type=MovieSceneScriptingFloatChannel, default=1, const [1,32]}
       hips_tan_ctrl.Scale.Z {type=MovieSceneScriptingFloatChannel, default=1, const [1,32]}
       spine_01_ctrl.Location.X {type=MovieSceneScriptingFloatChannel, default=0, const [0,36]}
       spine_01_ctrl.Location.Y {type=MovieSceneScriptingFloatChannel, default=0, const [0,36]}
       spine_01_ctrl.Location.Z {type=MovieSceneScriptingFloatChannel, default=0, const [0,36]}
       spine_01_ctrl.Rotation.X {type=MovieSceneScriptingFloatChannel, default=0, const [0,36]}
       spine_01_ctrl.Rotation.Y {type=MovieSceneScriptingFloatChannel, default=0, thinned_from=36, keys [[-1,0],[389,0]]}
       spine_01_ctrl.Rotation.Z {type=MovieSceneScriptingFloatChannel, default=0, const [0,36]}
       spine_01_ctrl.Scale.X {type=MovieSceneScriptingFloatChannel, default=1, const [1,36]}
       spine_01_ctrl.Scale.Y {type=MovieSceneScriptingFloatChannel, default=1, const [1,36]}
       spine_01_ctrl.Scale.Z {type=MovieSceneScriptingFloatChannel, default=1, const [1,36]}
       spine_02_ctrl.Location.X {type=MovieSceneScriptingFloatChannel, default=0, const [0,35]}
       spine_02_ctrl.Location.Y {type=MovieSceneScriptingFloatChannel, default=0, const [0,35]}
       spine_02_ctrl.Location.Z {type=MovieSceneScriptingFloatChannel, default=0, const [0,35]}
       spine_02_ctrl.Rotation.X {
        type=MovieSceneScriptingFloatChannel
        default=0
        thinned_from=34
        keys [[-1,7.4245],[0,7.7926],[46,7.7926],[60,7.8979],[78,8.214],[92,8.099],[112,7.758],[119,7.7093],[139,7.7093],[146,7.7404],[160,7.8558],[178,7.9623],[183,7.6359],[190,7.4645],[210,7.2984],[217,7.3785],[221,7.6857],[225,7.977],[234,7.1991],[245,6.8768],[253,6.8307],[265,6.9202],[272,7.2409],[280,10.1074],[294,8.3062],[310,7.7926],[357,7.7926],[376,7.4357],[389,7.4245]]
       }
       spine_02_ctrl.Rotation.Y {
        type=MovieSceneScriptingFloatChannel
        default=0
        thinned_from=34
        keys [[-1,4.1332],[0,6.2767],[21,6.2767],[46,6.5305],[60,6.4186],[78,6.0828],[92,5.8537],[119,5.8537],[135,6.4433],[139,6.6399],[146,7.4319],[160,8.9064],[172,9.4571],[178,9.6859],[183,5.0796],[190,3.1331],[210,1.7099],[217,1.4112],[221,1.1433],[225,-0.0841],[234,-2.8608],[245,-3.2078],[253,-3.2574],[265,-2.8056],[272,-1.8718],[280,1.0586],[294,5.715],[310,6.2767],[331,6.2767],[357,6.5305],[376,4.1332],[388,4.387],[389,4.1332]]
       }
       spine_02_ctrl.Rotation.Z {
        type=MovieSceneScriptingFloatChannel
        default=0
        thinned_from=34
        keys [[-1,0],[0,0.87],[139,0.87],[146,-1.9252],[160,-5.7165],[172,-7.1325],[178,-9.5603],[183,-10.0368],[190,-8.5183],[210,-8.0026],[217,-8.0026],[221,-7.074],[225,-2.5324],[234,5.6588],[245,8.0478],[253,8.3891],[265,8.6632],[272,7.4046],[280,4.1274],[294,1.3477],[310,0.87],[357,0.87],[376,0.1476],[388,0],[389,0]]
       }
       spine_02_ctrl.Scale.X {type=MovieSceneScriptingFloatChannel, default=1, const [1,35]}
       spine_02_ctrl.Scale.Y {type=MovieSceneScriptingFloatChannel, default=1, const [1,35]}
       spine_02_ctrl.Scale.Z {type=MovieSceneScriptingFloatChannel, default=1, const [1,35]}
       spine_03_ctrl.Location.X {type=MovieSceneScriptingFloatChannel, default=0, const [0,36]}
       spine_03_ctrl.Location.Y {type=MovieSceneScriptingFloatChannel, default=0, const [0,36]}
       spine_03_ctrl.Location.Z {
        type=MovieSceneScriptingFloatChannel
        default=0
        thinned_from=38
        keys [[-1,0],[0,0],[11,0.1803],[21,0.0554],[46,0],[310,0],[321,0.1803],[331,0.0554],[357,0],[389,0]]
       }
       spine_03_ctrl.Rotation.X {
        type=MovieSceneScriptingFloatChannel
        default=0
        thinned_from=36
        keys [[-1,0],[135,0],[139,0.0256],[146,0.1272],[160,0.5798],[172,0.8514],[253,0.8514],[265,0.8721],[272,1.017],[280,1.5574],[286,0.7068],[294,0.2192],[310,0],[389,0]]
       }
       spine_03_ctrl.Rotation.Y {type=MovieSceneScriptingFloatChannel, default=0, thinned_from=35, keys [[-1,0],[389,0]]}
       spine_03_ctrl.Rotation.Z {type=MovieSceneScriptingFloatChannel, default=0, const [0,36]}
       spine_03_ctrl.Scale.X {type=MovieSceneScriptingFloatChannel, default=1, const [1,36]}
       spine_03_ctrl.Scale.Y {type=MovieSceneScriptingFloatChannel, default=1, const [1,36]}
       spine_03_ctrl.Scale.Z {type=MovieSceneScriptingFloatChannel, default=1, const [1,36]}
       spine_04_ctrl.Location.X {type=MovieSceneScriptingFloatChannel, default=0, const [0,35]}
       spine_04_ctrl.Location.Y {type=MovieSceneScriptingFloatChannel, default=0, const [0,35]}
       spine_04_ctrl.Location.Z {type=MovieSceneScriptingFloatChannel, default=0, const [0,35]}
       spine_04_ctrl.Rotation.X {type=MovieSceneScriptingFloatChannel, default=0, const [17.4723,33]}
       spine_04_ctrl.Rotation.Y {type=MovieSceneScriptingFloatChannel, default=0, thinned_from=33, keys [[-1,0.7539],[389,0.7539]]}
       spine_04_ctrl.Rotation.Z {type=MovieSceneScriptingFloatChannel, default=0, const [9.078,33]}
       spine_04_ctrl.Scale.X {type=MovieSceneScriptingFloatChannel, default=1, const [1,35]}
       spine_04_ctrl.Scale.Y {type=MovieSceneScriptingFloatChannel, default=1, const [1,35]}
       spine_04_ctrl.Scale.Z {type=MovieSceneScriptingFloatChannel, default=1, const [1,35]}
       head_ctrl.Location.X {
        type=MovieSceneScriptingFloatChannel
        default=0
        thinned_from=39
        keys [[-1,0],[78,0],[84,-0.0057],[92,-0.0459],[112,-0.0466],[119,-0.0459],[178,-0.0459],[183,0],[190,0],[210,-0.0015],[217,0],[225,0],[234,0.5043],[245,0.063],[253,0],[272,0],[280,-0.0004],[294,0],[389,0]]
       }
       head_ctrl.Location.Y {
        type=MovieSceneScriptingFloatChannel
        default=0
        thinned_from=39
        keys [[-1,-0.0875],[78,-0.0875],[84,-0.0893],[92,-0.1018],[112,-0.1292],[115,-0.1184],[119,-0.1018],[178,-0.1018],[183,-0.0875],[190,-0.0875],[210,-0.0902],[217,-0.0875],[225,-0.0875],[234,0.016],[245,-0.0746],[253,-0.0875],[389,-0.0875]]
       }
       head_ctrl.Location.Z {
        type=MovieSceneScriptingFloatChannel
        default=0
        thinned_from=39
        keys [[-1,-0.3874],[92,-0.3874],[112,-0.3787],[119,-0.3873],[190,-0.3874],[210,-0.3879],[225,-0.3874],[234,-0.4028],[245,-0.3893],[253,-0.3874],[389,-0.3874]]
       }
       head_ctrl.Rotation.X {
        type=MovieSceneScriptingFloatChannel
        default=0
        thinned_from=39
        keys [[-1,3.8354],[0,-4.2745],[21,-6.0959],[46,-6.3561],[78,-6.3561],[84,-6.4037],[92,-6.7368],[115,-6.8611],[119,-7.6459],[123,-7.8554],[139,-8.193],[142,-8.4155],[146,-10.021],[160,-10.8252],[172,-10.8337],[178,-9.8026],[183,-3.0305],[190,-0.6102],[210,-1.7145],[217,-2.1269],[221,-1.8207],[225,-2.287],[234,-3.9936],[245,-5.3371],[253,-5.529],[265,-4.9208],[272,-5.0821],[280,-4.8482],[294,-4.3276],[310,-4.2745],[331,-6.0959],[357,-6.3561],[376,3.4539],[388,1.7539],[389,3.8354]]
       }
       head_ctrl.Rotation.Y {
        type=MovieSceneScriptingFloatChannel
        default=0
        thinned_from=39
        keys [[-1,-8.3104],[0,-4.3557],[78,-4.3557],[84,-4.3643],[92,-4.424],[112,-4.7889],[123,-4.7738],[135,-4.7703],[139,-4.7315],[142,-4.1233],[146,-3.7344],[160,-3.237],[172,-2.9838],[178,-2.5929],[183,-4.5776],[190,-5.8176],[210,-6.0273],[217,-5.8741],[221,-5.6188],[225,-4.1376],[234,-3.6022],[245,-3.2323],[253,-3.1795],[265,-2.4095],[272,-0.2806],[280,-2.9582],[294,-3.8839],[310,-4.3557],[357,-4.3557],[376,-7.2991],[388,-8.3104],[389,-8.3104]]
       }
       head_ctrl.Rotation.Z {
        type=MovieSceneScriptingFloatChannel
        default=0
        keys [[-1,-17.7188],[0,-9.8348],[21,-9.7004],[46,-9.2908],[52,-7.5207],[60,-6.9306],[78,-6.9306],[84,-7.7365],[92,-13.3777],[112,-15.1302],[115,-17.1399],[119,-20.7109],[123,-21.8027],[135,-22.7235],[139,-24.7657],[142,-32.5253],[146,-41.6232],[160,-45.8592],[172,-47.3509],[178,-46.55],[183,-25.1176],[190,-16.9153],[208,-13.323],[215,-12.8807],[219,-10.8189],[223,4.6422],[232,20.848],[245,24.0627],[253,24.522],[265,24.3506],[272,18.1027],[280,-1.9121],[294,-6.7166],[310,-9.8348],[331,-9.7004],[357,-9.2908],[376,-16.5377],[388,-17.1747],[389,-17.7188]]
       }
       head_ctrl.Scale.X {type=MovieSceneScriptingFloatChannel, default=1, const [1,39]}
       head_ctrl.Scale.Y {type=MovieSceneScriptingFloatChannel, default=1, const [1,39]}
       head_ctrl.Scale.Z {type=MovieSceneScriptingFloatChannel, default=1, const [1,39]}
       clavicle_l_ctrl.Location.X {type=MovieSceneScriptingFloatChannel, default=0, thinned_from=34, keys [[-1,0],[389,0]]}
       clavicle_l_ctrl.Location.Y {type=MovieSceneScriptingFloatChannel, default=0, thinned_from=34, keys [[-1,0],[389,0]]}
       clavicle_l_ctrl.Location.Z {type=MovieSceneScriptingFloatChannel, default=0, thinned_from=34, keys [[-1,0],[389,0]]}
       clavicle_l_ctrl.Rotation.X {type=MovieSceneScriptingFloatChannel, default=0, thinned_from=35, keys [[-1,0.2452],[389,0.2452]]}
       clavicle_l_ctrl.Rotation.Y {
        type=MovieSceneScriptingFloatChannel
        default=0
        thinned_from=24
        keys [[-1,-5.1113],[0,2.5884],[21,3.8639],[46,4.2891],[84,4.2891],[92,3.8605],[112,4.5302],[135,4.7534],[172,4.7534],[178,4.0231],[183,-4.232],[190,-1.8074],[210,-0.9992],[265,-0.9992],[294,1.6915],[310,2.5884],[331,3.8639],[357,4.2891],[376,-3.0918],[388,-3.4106],[389,-5.1113]]
       }
       clavicle_l_ctrl.Rotation.Z {
        type=MovieSceneScriptingFloatChannel
        default=0
        thinned_from=35
        keys [[-1,11.7619],[0,15.778],[21,15.778],[46,15.8742],[135,15.8742],[139,15.2608],[142,12.3811],[146,10.8764],[160,9.8755],[172,9.8755],[178,9.0089],[183,3.0336],[190,5.8545],[210,6.7948],[217,6.7948],[227,7.816],[234,13.0167],[245,15.4414],[253,15.7878],[331,15.778],[357,15.8742],[376,13.4647],[388,11.8581],[389,11.7619]]
       }
       clavicle_l_ctrl.Scale.X {type=MovieSceneScriptingFloatChannel, default=1, const [1,34]}
       clavicle_l_ctrl.Scale.Y {type=MovieSceneScriptingFloatChannel, default=1, const [1,34]}
       clavicle_l_ctrl.Scale.Z {type=MovieSceneScriptingFloatChannel, default=1, const [1,34]}
       clavicle_r_ctrl.Location.X {type=MovieSceneScriptingFloatChannel, default=0, const [0,23]}
       clavicle_r_ctrl.Location.Y {type=MovieSceneScriptingFloatChannel, default=0, const [0,23]}
       clavicle_r_ctrl.Location.Z {type=MovieSceneScriptingFloatChannel, default=0, const [0,23]}
       clavicle_r_ctrl.Rotation.X {type=MovieSceneScriptingFloatChannel, default=0, thinned_from=23, keys [[-1,0.6254],[390,0.6254]]}
       clavicle_r_ctrl.Rotation.Y {
        type=MovieSceneScriptingFloatChannel
        default=0
        thinned_from=23
        keys [[-1,-8.2547],[0,-2.0163],[21,-0.8313],[46,-0.4363],[84,-0.4363],[92,-0.8694],[112,-0.1997],[135,0.0235],[172,0.0235],[178,-0.7086],[183,-7.5038],[190,-5.6556],[210,-5.0395],[265,-5.0395],[294,-2.7721],[310,-2.0163],[331,-0.8313],[357,-0.4363],[388,-6.6748],[390,-8.2547]]
       }
       clavicle_r_ctrl.Rotation.Z {type=MovieSceneScriptingFloatChannel, default=0, const [10.3592,23]}
       clavicle_r_ctrl.Scale.X {type=MovieSceneScriptingFloatChannel, default=1, const [1,23]}
       clavicle_r_ctrl.Scale.Y {type=MovieSceneScriptingFloatChannel, default=1, const [1,23]}
       clavicle_r_ctrl.Scale.Z {type=MovieSceneScriptingFloatChannel, default=1, const [1,23]}
       neck_01_ctrl.Location.X {type=MovieSceneScriptingFloatChannel, default=0, const [0,14]}
       neck_01_ctrl.Location.Y {type=MovieSceneScriptingFloatChannel, default=0, thinned_from=14, keys [[-1,0],[389,0]]}
       neck_01_ctrl.Location.Z {type=MovieSceneScriptingFloatChannel, default=0, thinned_from=14, keys [[-1,0],[389,0]]}
       neck_01_ctrl.Rotation.X {
        type=MovieSceneScriptingFloatChannel
        default=0
        thinned_from=35
        keys [[-1,-2.6694],[0,0.8992],[21,4.2223],[46,4.7427],[84,4.7368],[92,4.7104],[135,4.7016],[139,4.4052],[146,1.6243],[160,0.661],[178,0.661],[183,-2.9102],[210,-2.9102],[217,-2.893],[221,-2.8642],[225,-0.7523],[234,0.5843],[245,0.6112],[253,0.615],[265,0.615],[272,0.642],[280,0.8307],[294,1.8454],[310,0.8992],[331,4.2223],[357,4.7427],[376,-2.5949],[388,1.174],[389,-2.5949]]
       }
       neck_01_ctrl.Rotation.Y {
        type=MovieSceneScriptingFloatChannel
        default=0
        thinned_from=35
        keys [[-1,2.0231],[0,7.4021],[21,5.1718],[46,4.8937],[60,4.9035],[78,4.9717],[84,5.2616],[92,6.7305],[112,7.9185],[178,7.9185],[183,2.4851],[210,2.4851],[217,2.502],[221,2.5301],[225,2.845],[234,2.935],[245,2.9612],[265,2.965],[272,3.0424],[280,3.5842],[294,4.7591],[310,7.4021],[331,5.1718],[357,4.8937],[376,2.4078],[388,-0.4852],[389,2.0231]]
       }
       neck_01_ctrl.Rotation.Z {
        type=MovieSceneScriptingFloatChannel
        default=0
        thinned_from=35
        keys [[-1,-1.7035],[0,-1.5319],[21,-7.1013],[46,-8.9578],[60,-8.97],[78,-9.0552],[84,-9.4409],[92,-10.9035],[112,-11.5784],[210,-11.5784],[217,-11.2165],[221,-10.6133],[225,-4.3007],[234,-2.3705],[245,-1.8076],[253,-1.7271],[272,-1.7263],[280,-1.7204],[294,-1.7035],[310,-1.5319],[331,-7.1013],[357,-8.9578],[376,-1.7035],[388,-9.1294],[389,-1.7035]]
       }
       neck_01_ctrl.Scale.X {type=MovieSceneScriptingFloatChannel, default=1, const [1,14]}
       neck_01_ctrl.Scale.Y {type=MovieSceneScriptingFloatChannel, default=1, const [1,14]}
       neck_01_ctrl.Scale.Z {type=MovieSceneScriptingFloatChannel, default=1, const [1,14]}
       neck_02_ctrl.Location.X {type=MovieSceneScriptingFloatChannel, default=0, const [0,36]}
       neck_02_ctrl.Location.Y {
        type=MovieSceneScriptingFloatChannel
        default=0
        thinned_from=36
        keys [[-1,0.1375],[0,0],[21,-0.1031],[46,-0.1375],[178,-0.1375],[183,0],[310,0],[331,-0.1031],[357,-0.1375],[376,0.1203],[388,0],[389,0.1375]]
       }
       neck_02_ctrl.Location.Z {
        type=MovieSceneScriptingFloatChannel
        default=0
        thinned_from=36
        keys [[-1,0.4106],[0,0],[21,-0.308],[46,-0.4106],[178,-0.4106],[183,0],[310,0],[331,-0.308],[357,-0.4106],[376,0.3593],[388,0],[389,0.4106]]
       }
       neck_02_ctrl.Rotation.X {type=MovieSceneScriptingFloatChannel, default=0, const [0,35]}
       neck_02_ctrl.Rotation.Y {
        type=MovieSceneScriptingFloatChannel
        default=0
        thinned_from=35
        keys [[-1,-6.2034],[0,0],[21,4.6525],[46,6.2034],[178,6.2034],[183,0],[310,0],[331,4.6525],[357,6.2034],[376,-5.4279],[388,0],[389,-6.2034]]
       }
       neck_02_ctrl.Rotation.Z {type=MovieSceneScriptingFloatChannel, default=0, const [0,35]}
       neck_02_ctrl.Scale.X {type=MovieSceneScriptingFloatChannel, default=1, const [1,36]}
       neck_02_ctrl.Scale.Y {type=MovieSceneScriptingFloatChannel, default=1, const [1,36]}
       neck_02_ctrl.Scale.Z {type=MovieSceneScriptingFloatChannel, default=1, const [1,36]}
       head_ik_ctrl.Location.X {
        type=MovieSceneScriptingFloatChannel
        default=0
        thinned_from=29
        keys [[-1,0.2514],[0,-1.1317],[21,-0.7957],[46,-0.7456],[60,-0.7456],[78,-0.7174],[84,-0.7456],[92,-0.9014],[146,-0.9014],[160,-0.9344],[172,-0.9344],[178,-0.9794],[183,0.0259],[190,0.0489],[210,0.0345],[217,0.0489],[234,-0.299],[253,-0.2998],[267,-0.2998],[294,-0.7258],[310,-1.1317],[331,-1.1385],[357,-0.7456],[376,0.4953],[389,0.6698]]
       }
       head_ik_ctrl.Location.Y {
        type=MovieSceneScriptingFloatChannel
        default=0
        thinned_from=29
        keys [[-1,-0.1323],[0,-0.1761],[21,0.1364],[46,0.1475],[60,0.1475],[78,0.0979],[84,0.1475],[92,0.1553],[146,0.1553],[160,0.1487],[172,0.1487],[178,0.193],[183,-0.4385],[190,-0.4797],[210,-0.4408],[217,-0.4797],[234,-0.5421],[253,-0.5469],[267,-0.5469],[294,-0.6149],[310,-0.1761],[331,-0.4573],[357,0.1475],[376,-1.1941],[389,-1.2768]]
       }
       head_ik_ctrl.Location.Z {
        type=MovieSceneScriptingFloatChannel
        default=0
        thinned_from=29
        keys [[-1,0.2364],[0,-0.6913],[21,-1.3246],[46,-1.4907],[60,-1.4907],[78,-1.6268],[84,-1.4907],[92,-1.5174],[172,-1.5191],[178,-1.4573],[183,-0.0858],[190,-0.143],[210,-0.0875],[217,-0.143],[234,-0.1655],[253,-0.1646],[267,-0.1646],[294,-0.2617],[310,-0.6913],[331,-0.9586],[357,-1.4907],[376,0.595],[389,0.7174]]
       }
       head_ik_ctrl.Rotation.X {
        type=MovieSceneScriptingFloatChannel
        default=0
        thinned_from=29
        keys [[-1,3.8354],[0,-4.2745],[21,-6.0959],[46,-6.3561],[84,-6.3561],[92,-6.9778],[112,-6.9778],[119,-7.9845],[135,-8.3689],[139,-8.4488],[146,-10.2769],[160,-11.0811],[172,-11.0811],[178,-13.2305],[183,-3.0305],[190,0.3508],[210,0.3313],[217,0.3508],[234,-0.8351],[253,-0.8689],[267,-0.8689],[294,0.7833],[310,-4.2745],[331,-4.2745],[357,-6.3561],[376,0.1538],[389,0.0639]]
       }
       head_ik_ctrl.Rotation.Y {
        type=MovieSceneScriptingFloatChannel
        default=0
        thinned_from=29
        keys [[-1,-8.3104],[0,-4.3557],[84,-4.3557],[92,-4.3105],[112,-4.3105],[119,-4.297],[135,-4.2919],[139,-4.2531],[146,-3.256],[160,-2.7586],[172,-2.7586],[178,-4.8676],[183,-7.1767],[190,-6.0271],[217,-6.0271],[234,-3.4611],[253,-3.4238],[267,-3.4238],[294,-0.2195],[310,-4.3557],[357,-4.3557],[376,-3.4699],[389,-3.9343]]
       }
       head_ik_ctrl.Rotation.Z {
        type=MovieSceneScriptingFloatChannel
        default=0
        thinned_from=29
        keys [[-1,-17.7188],[0,-9.8348],[21,-9.7004],[46,-9.2908],[84,-9.2908],[92,-11.9429],[112,-13.77],[119,-17.3239],[135,-18.6808],[139,-20.723],[146,-40.2631],[160,-44.4991],[172,-44.4991],[178,-43.5838],[183,-25.1176],[190,-14.1812],[210,-13.323],[217,-14.1812],[234,23.8247],[253,24.522],[267,24.522],[294,-8.27],[310,-9.8348],[331,-9.8348],[357,-9.2908],[376,-16.2512],[389,-17.3914]]
       }
       head_ik_ctrl.Scale.X {type=MovieSceneScriptingFloatChannel, default=1, const [1,29]}
       head_ik_ctrl.Scale.Y {type=MovieSceneScriptingFloatChannel, default=1, const [1,29]}
       head_ik_ctrl.Scale.Z {type=MovieSceneScriptingFloatChannel, default=1, const [1,29]}
       chest_ctrl.Location.X {
        type=MovieSceneScriptingFloatChannel
        default=0
        thinned_from=32
        keys [[-1,-0.5014],[0,-0.7643],[21,-0.7693],[46,-0.7949],[60,-0.7949],[78,-0.7975],[92,-0.7949],[139,-0.7949],[146,-0.9742],[160,-0.5158],[172,-0.5158],[178,-1.214],[183,-0.4188],[190,-0.2225],[210,-0.2135],[217,-0.2223],[221,0],[234,0.3501],[253,0.3682],[265,0.3659],[267,0.3682],[272,0],[294,-0.506],[310,-0.7643],[331,-0.7643],[357,-0.7949],[376,-0.0633],[389,0]]
       }
       chest_ctrl.Location.Y {
        type=MovieSceneScriptingFloatChannel
        default=0
        thinned_from=32
        keys [[-1,0.9008],[0,0.9339],[21,0.9399],[46,0.9334],[60,0.9334],[78,0.9969],[84,1.0467],[92,0.9334],[139,0.9334],[146,0.9707],[160,0.9845],[172,0.9845],[178,1.0909],[183,0.9877],[190,0.9518],[210,0.9589],[217,0.9518],[221,0.9008],[253,0.9065],[267,0.9065],[272,0.9008],[294,0.9013],[310,0.9339],[357,0.9334],[376,0.9009],[389,0.9008]]
       }
       chest_ctrl.Location.Z {
        type=MovieSceneScriptingFloatChannel
        default=0
        thinned_from=32
        keys [[-1,0.0119],[0,-0.0132],[21,0.0312],[46,-0.0166],[60,-0.0166],[78,-0.0191],[84,-0.0216],[92,-0.0166],[139,-0.0166],[146,-0.0332],[160,-0.0039],[172,-0.0039],[178,-0.0665],[183,0.0141],[190,0.0247],[217,0.0247],[221,0.03],[234,0.0211],[253,0.0201],[265,0.0202],[267,0.0201],[272,0.03],[294,0.0116],[310,-0.0132],[331,-0.0132],[357,-0.0166],[376,0.0277],[389,0.03]]
       }
       chest_ctrl.Rotation.X {
        type=MovieSceneScriptingFloatChannel
        default=0
        thinned_from=32
        keys [[-1,7.4245],[0,7.8034],[21,7.7926],[60,7.8034],[78,8.3081],[84,8.7233],[92,7.8034],[135,7.8034],[139,8.8487],[146,7.9617],[160,8.0149],[172,8.0149],[178,8.8556],[183,8.4873],[190,7.6357],[210,8.4768],[217,7.6359],[221,7.4245],[234,7.0672],[253,7.0415],[267,7.0415],[272,7.4245],[294,7.5143],[310,7.8034],[357,7.8034],[376,7.4357],[389,7.4245]]
       }
       chest_ctrl.Rotation.Y {
        type=MovieSceneScriptingFloatChannel
        default=0
        thinned_from=32
        keys [[-1,4.1332],[0,6.2767],[21,6.2767],[46,6.5305],[60,6.5305],[78,6.563],[92,6.5305],[139,6.5305],[146,8.1087],[160,4.4565],[172,4.4565],[178,10.3627],[183,3.7692],[190,2.0609],[217,2.0592],[221,0],[234,-3.0916],[253,-3.2574],[265,-3.2378],[267,-3.2574],[272,0],[294,4.1332],[310,6.2767],[331,6.2767],[357,6.5305],[376,0.5167],[389,0]]
       }
       chest_ctrl.Rotation.Z {
        type=MovieSceneScriptingFloatChannel
        default=0
        thinned_from=32
        keys [[-1,0],[0,0.87],[60,0.87],[78,0.6193],[84,0.682],[92,0.87],[139,0.87],[146,-1.9252],[160,-5.2852],[172,-5.2852],[178,-7.1325],[183,-8.0026],[190,-6.1502],[210,-8.0026],[217,-6.1552],[221,-3.9924],[234,7.2407],[253,7.8176],[265,7.8052],[267,7.8176],[272,0],[294,1.181],[310,0.87],[357,0.87],[376,0.1476],[389,0]]
       }
       chest_ctrl.Scale.X {type=MovieSceneScriptingFloatChannel, default=1, const [1,32]}
       chest_ctrl.Scale.Y {type=MovieSceneScriptingFloatChannel, default=1, const [1,32]}
       chest_ctrl.Scale.Z {type=MovieSceneScriptingFloatChannel, default=1, const [1,32]}
       chest_tan_ctrl.Location.X {
        type=MovieSceneScriptingFloatChannel
        default=0
        thinned_from=32
        keys [[-1,0.3954],[0,0.5903],[21,0.5903],[46,0.6145],[60,0.6145],[78,0.6203],[84,0.6189],[92,0.6145],[139,0.6145],[146,0.7953],[160,0.4831],[172,0.4831],[178,1.0654],[183,0.4474],[190,0.2641],[210,0.284],[217,0.264],[221,0],[234,-0.3745],[253,-0.3965],[265,-0.3945],[267,-0.3965],[272,0],[294,0.3826],[310,0.5903],[331,0.5903],[357,0.6145],[376,0.0478],[389,0]]
       }
       chest_tan_ctrl.Location.Y {
        type=MovieSceneScriptingFloatChannel
        default=0.0049
        thinned_from=32
        keys [[-1,-0.6969],[0,-0.7274],[21,-0.7287],[60,-0.727],[78,-0.774],[84,-0.8125],[92,-0.727],[135,-0.727],[139,-0.5956],[146,-0.7111],[160,-0.7443],[172,-0.7443],[178,-0.6275],[183,-0.603],[190,-0.7141],[210,-0.6059],[217,-0.7141],[221,-0.6987],[234,-0.6588],[253,-0.6554],[267,-0.6554],[272,-0.6987],[294,-0.7053],[310,-0.7274],[357,-0.727],[376,-0.6995],[389,-0.6987]]
       }
       chest_tan_ctrl.Location.Z {
        type=MovieSceneScriptingFloatChannel
        default=1.2539
        thinned_from=32
        keys [[-1,1.1131],[0,1.087],[21,1.0054],[46,1.0844],[60,1.0844],[78,1.0721],[84,1.0619],[92,1.0844],[139,1.0839],[146,1.0677],[160,1.0882],[172,1.0882],[178,1.0293],[183,1.1049],[190,1.1169],[210,1.1157],[217,1.1169],[221,1.1272],[234,1.1239],[267,1.1231],[272,1.1272],[294,1.1119],[310,1.087],[331,1.087],[357,1.0844],[376,1.1253],[389,1.1272]]
       }
       chest_tan_ctrl.Rotation.X {
        type=MovieSceneScriptingFloatChannel
        default=0.6533
        thinned_from=32
        keys [[-1,0.6533],[0,0.6424],[21,0.6786],[46,0.6424],[92,0.6424],[112,0.6402],[119,0.6424],[135,0.6424],[139,-0.0108],[146,0.1621],[160,0.6424],[172,0.6424],[178,-0.0108],[183,0],[190,0.6529],[210,0],[217,0.6533],[272,0.6533],[294,0.7241],[310,0.6424],[357,0.6424],[376,0.6621],[389,0.6533]]
       }
       chest_tan_ctrl.Rotation.Y {
        type=MovieSceneScriptingFloatChannel
        default=0
        thinned_from=32
        keys [[-1,0],[139,0],[146,3.3533],[160,0],[183,0],[190,0.0024],[210,0],[272,0],[294,4.2511],[310,0],[357,0],[376,0.5314],[389,0]]
       }
       chest_tan_ctrl.Rotation.Z {
        type=MovieSceneScriptingFloatChannel
        default=0
        thinned_from=32
        keys [[-1,0],[139,0],[146,1.3697],[160,0],[183,0],[190,0.0048],[210,0],[272,0],[294,0.6361],[310,0],[357,0],[376,0.0795],[389,0]]
       }
       chest_tan_ctrl.Scale.X {type=MovieSceneScriptingFloatChannel, default=1, const [1,32]}
       chest_tan_ctrl.Scale.Y {type=MovieSceneScriptingFloatChannel, default=1, const [1,32]}
       chest_tan_ctrl.Scale.Z {type=MovieSceneScriptingFloatChannel, default=1, const [1,32]}
       leg_r_pv_ik_ctrl.Space {
        type=MovieSceneScriptingSpaceChannel
        const [
         "<Struct 'MovieSceneControlRigSpaceBaseKey' (0x49d263a76c0) {space_type: Parent, control_rig_element: {type: Null, name: \"leg_r_pv_footFollow\"}}>"
         1
        ]
       }
       hand_l_ik_ctrl.Space {
        type=MovieSceneScriptingSpaceChannel
        const [
         "<Struct 'MovieSceneControlRigSpaceBaseKey' (0x49d263a6cc0) {space_type: ControlRig, control_rig_element: {type: Null, name: \"chest_space\"}}>"
         1
        ]
       }
       upperarm_r_fk_ctrl.Space {type=MovieSceneScriptingSpaceChannel}
       hand_r_ik_ctrl.Space {
        type=MovieSceneScriptingSpaceChannel
        const [
         "<Struct 'MovieSceneControlRigSpaceBaseKey' (0x49d263a20a0) {space_type: ControlRig, control_rig_element: {type: Null, name: \"chest_space\"}}>"
         1
        ]
       }
       head_ctrl.Space {
        type=MovieSceneScriptingSpaceChannel
        const [
         "<Struct 'MovieSceneControlRigSpaceBaseKey' (0x49d263a6400) {space_type: Parent, control_rig_element: {type: Null, name: \"head_fk_space\"}}>"
         1
        ]
       }
      ]
     }]
   }
   {type=MovieSceneControlRigParameterTrack, display_name=MR_Biped_v02, sections [{type=MovieSceneControlRigParameterSection}]}
  ]
 }]
display_rate {denominator=1, numerator=30}
tick_resolution {denominator=1, numerator=24000}
tracks []
