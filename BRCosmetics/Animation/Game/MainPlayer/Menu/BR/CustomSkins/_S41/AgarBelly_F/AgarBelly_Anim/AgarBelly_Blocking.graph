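# Level Sequence holding a blocking animation for the AgarBelly_F custom skin's menu pose, judging by its path and name.
# It spawns a FortPlayerMannequin at the origin and animates it through a Fortnite_Biped_CtrlRig Control Rig track, with an empty skeletal animation track alongside.
# The visible channels set FK/IK switches per limb and key body controls from frame 0 to 340, ending on their starting values as a loop would.
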LevelSequence /BRCosmetics/Animation/Game/MainPlayer/Menu/BR/CustomSkins/_S41/AgarBelly_F/AgarBelly_Anim/AgarBelly_Blocking
bindings [FortPlayerMannequin {
  id=#1
  possessed_class=FortPlayerMannequin
  tracks [
   {type=MovieSceneSpawnTrack, sections [{type=MovieSceneSpawnSection, channels [None {type=MovieSceneScriptingBoolChannel, default=true}]}]}
   {
    type=MovieScene3DTransformTrack
    sections [{
      type=MovieScene3DTransformSection
      channels [
       Location.X {type=MovieSceneScriptingDoubleChannel, default=0, const [0,1]}
       Location.Y {type=MovieSceneScriptingDoubleChannel, default=0, const [0,1]}
       Location.Z {type=MovieSceneScriptingDoubleChannel, default=0, const [0,1]}
       Rotation.X {type=MovieSceneScriptingDoubleChannel, default=0, const [0,1]}
       Rotation.Y {type=MovieSceneScriptingDoubleChannel, default=0, const [0,1]}
       Rotation.Z {type=MovieSceneScriptingDoubleChannel, default=0, const [0,1]}
       Scale.X {type=MovieSceneScriptingDoubleChannel, default=1, const [1,1]}
       Scale.Y {type=MovieSceneScriptingDoubleChannel, default=1, const [1,1]}
       Scale.Z {type=MovieSceneScriptingDoubleChannel, default=1, const [1,1]}
      ]
     }]
   }
   {type=MovieSceneSkeletalAnimationTrack, sections []}
   {
    type=MovieSceneControlRigParameterTrack
    display_name=Fortnite_Biped_CtrlRig
    sections [{
      type=MovieSceneControlRigParameterSection
      channels [
       leg_l_stretch_switch {type=MovieSceneScriptingBoolChannel, default=false, const [true,24]}
       leg_r_stretch_switch {type=MovieSceneScriptingBoolChannel, default=false, const [true,24]}
       arm_l_fk_local {type=MovieSceneScriptingBoolChannel, default=true, const [true,29]}
       Auto Interpolate_3 {type=MovieSceneScriptingBoolChannel, default=false, const [false,24]}
       Auto Interpolate {type=MovieSceneScriptingBoolChannel, default=false, const [false,24]}
       arm_l_stretch_switch {type=MovieSceneScriptingBoolChannel, default=false, const [false,25]}
       Keep Aligned {type=MovieSceneScriptingBoolChannel, default=false, const [false,25]}
       arm_r_fk_local {type=MovieSceneScriptingBoolChannel, default=true, const [true,28]}
       Auto Interpolate_4 {type=MovieSceneScriptingBoolChannel, default=false, const [false,24]}
       Auto Interpolate_2 {type=MovieSceneScriptingBoolChannel, default=false, const [false,24]}
       arm_r_stretch_switch {type=MovieSceneScriptingBoolChannel, default=false, const [false,26]}
       Keep Aligned_2 {type=MovieSceneScriptingBoolChannel, default=false, const [false,26]}
       head_local {type=MovieSceneScriptingBoolChannel, default=true, const [true,31]}
       Stretch {type=MovieSceneScriptingBoolChannel, default=false, const [true,31]}
       head_stretch {type=MovieSceneScriptingBoolChannel, default=false, const [false,24]}
       spine_stretch_switch {type=MovieSceneScriptingBoolChannel, default=false, const [false,24]}
       arm_l_fk_ik_switch {type=MovieSceneScriptingBoolChannel, default=false, const [false,24]}
       leg_l_fk_ik_switch {type=MovieSceneScriptingBoolChannel, default=true, const [true,24]}
       arm_r_fk_ik_switch {type=MovieSceneScriptingBoolChannel, default=false, const [false,24]}
       leg_r_fk_ik_switch {type=MovieSceneScriptingBoolChannel, default=true, const [true,24]}
       ShowBodyControls {type=MovieSceneScriptingBoolChannel, default=true, const [true,24]}
       spine_fk_ik_switch {type=MovieSceneScriptingBoolChannel, default=false, const [false,24]}
       neck_fk_ik_switch {type=MovieSceneScriptingBoolChannel, default=false, const [false,24]}
       space {type=MovieSceneScriptingByteChannel, default=1, const [1,31]}
       global_ctrl.Location.X {type=MovieSceneScriptingFloatChannel, default=0, const [0,24]}
       global_ctrl.Location.Y {type=MovieSceneScriptingFloatChannel, default=0, const [0,24]}
       global_ctrl.Location.Z {type=MovieSceneScriptingFloatChannel, default=0, const [0,24]}
       global_ctrl.Rotation.X {type=MovieSceneScriptingFloatChannel, default=0, const [0,24]}
       global_ctrl.Rotation.Y {type=MovieSceneScriptingFloatChannel, default=0, const [0,24]}
       global_ctrl.Rotation.Z {type=MovieSceneScriptingFloatChannel, default=0, const [0,24]}
       root_ctrl.Location.X {type=MovieSceneScriptingFloatChannel, default=0, const [0,24]}
       root_ctrl.Location.Y {type=MovieSceneScriptingFloatChannel, default=0, const [0,24]}
       root_ctrl.Location.Z {type=MovieSceneScriptingFloatChannel, default=0, const [0,24]}
       root_ctrl.Rotation.X {type=MovieSceneScriptingFloatChannel, default=0, const [0,24]}
       root_ctrl.Rotation.Y {type=MovieSceneScriptingFloatChannel, default=0, const [0,24]}
       root_ctrl.Rotation.Z {type=MovieSceneScriptingFloatChannel, default=0, const [0,24]}
       root_ctrl.Scale.X {type=MovieSceneScriptingFloatChannel, default=1, const [1,24]}
       root_ctrl.Scale.Y {type=MovieSceneScriptingFloatChannel, default=1, const [1,24]}
       root_ctrl.Scale.Z {type=MovieSceneScriptingFloatChannel, default=1, const [1,24]}
       body_offset_ctrl.Location.X {type=MovieSceneScriptingFloatChannel, default=0, const [0,24]}
       body_offset_ctrl.Location.Y {type=MovieSceneScriptingFloatChannel, default=0, const [0,24]}
       body_offset_ctrl.Location.Z {type=MovieSceneScriptingFloatChannel, default=0, const [0,24]}
       body_offset_ctrl.Rotation.X {type=MovieSceneScriptingFloatChannel, default=0, const [0,24]}
       body_offset_ctrl.Rotation.Y {type=MovieSceneScriptingFloatChannel, default=0, const [0,24]}
       body_offset_ctrl.Rotation.Z {type=MovieSceneScriptingFloatChannel, default=0, const [0,24]}
       body_offset_ctrl.Scale.X {type=MovieSceneScriptingFloatChannel, default=1, const [1,24]}
       body_offset_ctrl.Scale.Y {type=MovieSceneScriptingFloatChannel, default=1, const [1,24]}
       body_offset_ctrl.Scale.Z {type=MovieSceneScriptingFloatChannel, default=1, const [1,24]}
       thigh_l_fk_ctrl.Location.X {type=MovieSceneScriptingFloatChannel, default=0, thinned_from=24, keys [[0,0],[340,0]]}
       thigh_l_fk_ctrl.Location.Y {type=MovieSceneScriptingFloatChannel, default=0, thinned_from=24, keys [[0,0],[340,0]]}
       thigh_l_fk_ctrl.Location.Z {type=MovieSceneScriptingFloatChannel, default=0, thinned_from=24, keys [[0,0],[340,0]]}
       thigh_l_fk_ctrl.Rotation.X {
        type=MovieSceneScriptingFloatChannel
        default=1.0553
        keys [[0,-5.7771],[25,-5.2463],[46,-5.0694],[60,-4.9693],[84,-4.9693],[92,-4.5993],[112,-4.5993],[122,-4.6335],[135,-4.6335],[139,-4.8263],[146,-5.0038],[160,-5.4921],[175,-5.4921],[178,-5.9507],[185,-6.1801],[205,-5.6424],[217,-5.6424],[236,-3.6654],[255,-3.6392],[269,-3.6392],[295,-5.7479],[311,-5.7479],[327,-5.7735],[340,-5.7771]]
       }
       thigh_l_fk_ctrl.Rotation.Y {
        type=MovieSceneScriptingFloatChannel
        default=-1.0258
        keys [[0,-0.6516],[25,-1.1545],[46,-1.3221],[60,-1.4186],[84,-1.4186],[92,-0.9959],[112,-0.9959],[122,-1.9061],[135,-1.9061],[139,-2.0018],[146,-2.4719],[160,-3.5218],[175,-3.5218],[178,-3.064],[185,-2.7896],[205,3.1419],[217,3.1419],[236,6.0775],[255,6.1166],[269,6.1166],[295,2.541],[311,2.541],[327,-0.2525],[340,-0.6516]]
       }
       thigh_l_fk_ctrl.Rotation.Z {
        type=MovieSceneScriptingFloatChannel
        default=-0.0727
        thinned_from=24
        keys [[0,-48.4383],[25,-48.133],[60,-47.9724],[84,-47.9724],[92,-48.4527],[146,-48.4962],[160,-49.3699],[175,-49.3699],[178,-49.5817],[185,-49.5608],[205,-49.1573],[217,-49.1573],[236,-48.8471],[311,-48.8782],[327,-48.4933],[340,-48.4383]]
       }
       thigh_l_fk_ctrl.Scale.X {type=MovieSceneScriptingFloatChannel, default=1, const [1,24]}
       thigh_l_fk_ctrl.Scale.Y {type=MovieSceneScriptingFloatChannel, default=1, const [1,24]}
       thigh_l_fk_ctrl.Scale.Z {type=MovieSceneScriptingFloatChannel, default=1, const [1,24]}
       calf_l_fk_ctrl.Location.X {type=MovieSceneScriptingFloatChannel, default=0, thinned_from=24, keys [[0,0],[340,0]]}
       calf_l_fk_ctrl.Location.Y {type=MovieSceneScriptingFloatChannel, default=0, thinned_from=24, keys [[0,0],[340,0]]}
       calf_l_fk_ctrl.Location.Z {type=MovieSceneScriptingFloatChannel, default=0, thinned_from=24, keys [[0,0],[340,0]]}
       calf_l_fk_ctrl.Rotation.X {
        type=MovieSceneScriptingFloatChannel
        default=0.0724
        thinned_from=24
        keys [[0,0.8683],[25,0.8629],[60,0.8601],[112,0.859],[122,0.8562],[160,0.8524],[175,0.8524],[178,0.8557],[185,0.8583],[205,0.8909],[217,0.8909],[236,0.8979],[269,0.8979],[295,0.886],[311,0.886],[327,0.8705],[340,0.8683]]
       }
       calf_l_fk_ctrl.Rotation.Y {
        type=MovieSceneScriptingFloatChannel
        default=1.0549
        thinned_from=24
        keys [[0,0.6035],[25,0.6111],[60,0.6151],[84,0.6151],[92,0.6167],[112,0.6167],[122,0.6205],[135,0.6205],[160,0.6258],[175,0.6258],[178,0.6211],[185,0.6176],[205,0.5696],[217,0.5696],[236,0.5585],[269,0.5585],[295,0.5771],[311,0.5771],[327,0.6002],[340,0.6035]]
       }
       calf_l_fk_ctrl.Rotation.Z {
        type=MovieSceneScriptingFloatChannel
        default=0.2163
        thinned_from=24
        keys [[0,51.4925],[25,50.9885],[60,50.7229],[84,50.7229],[92,50.6163],[112,50.6163],[122,50.3591],[139,50.2963],[160,50.0077],[175,50.0077],[178,50.317],[185,50.5554],[205,53.6995],[217,53.6995],[236,54.4078],[269,54.4078],[295,53.2133],[311,53.2133],[327,51.7076],[340,51.4925]]
       }
       calf_l_fk_ctrl.Scale.X {type=MovieSceneScriptingFloatChannel, default=1, const [1,24]}
       calf_l_fk_ctrl.Scale.Y {type=MovieSceneScriptingFloatChannel, default=1, const [1,24]}
       calf_l_fk_ctrl.Scale.Z {type=MovieSceneScriptingFloatChannel, default=1, const [1,24]}
       foot_l_fk_ctrl.Location.X {type=MovieSceneScriptingFloatChannel, default=0, thinned_from=24, keys [[0,0],[340,0]]}
       foot_l_fk_ctrl.Location.Y {
        type=MovieSceneScriptingFloatChannel
        default=0.0627
        thinned_from=24
        keys [[0,0.0627],[236,0.0627],[255,0.0553],[269,0.0553],[295,0.0627],[340,0.0627]]
       }
       foot_l_fk_ctrl.Location.Z {type=MovieSceneScriptingFloatChannel, default=-0.8453, thinned_from=24, keys [[0,-0.8453],[340,-0.8453]]}
       foot_l_fk_ctrl.Rotation.X {
        type=MovieSceneScriptingFloatChannel
        default=-0.9801
        thinned_from=24
        keys [[0,-2.3031],[84,-2.3089],[92,-2.3151],[135,-2.3177],[146,-2.3298],[160,-2.3629],[175,-2.3629],[178,-2.3762],[185,-2.3724],[205,-2.1934],[217,-2.1934],[236,-2.1366],[269,-2.1359],[295,-2.2201],[311,-2.2201],[327,-2.2927],[340,-2.3031]]
       }
       foot_l_fk_ctrl.Rotation.Y {
        type=MovieSceneScriptingFloatChannel
        default=0.1001
        thinned_from=24
        keys [[0,-6.3996],[25,-6.5071],[60,-6.5634],[84,-6.5634],[92,-6.6609],[112,-6.6609],[122,-6.5829],[135,-6.5829],[139,-6.6463],[146,-6.6216],[160,-6.4623],[175,-6.4623],[178,-6.4505],[185,-6.3932],[205,-5.3516],[217,-5.3516],[236,-5.1621],[269,-5.1597],[295,-5.7074],[311,-5.7074],[327,-6.3131],[340,-6.3996]]
       }
       foot_l_fk_ctrl.Rotation.Z {
        type=MovieSceneScriptingFloatChannel
        default=-0.1441
        thinned_from=24
        keys [[0,-19.2888],[25,-19.1145],[60,-19.0225],[112,-19.0131],[122,-18.8207],[139,-18.8072],[146,-18.7042],[160,-18.324],[175,-18.324],[178,-18.3894],[185,-18.4752],[205,-20.0687],[217,-20.0687],[236,-20.5699],[269,-20.5672],[295,-19.9939],[311,-19.9939],[327,-19.3769],[340,-19.2888]]
       }
       foot_l_fk_ctrl.Scale.X {type=MovieSceneScriptingFloatChannel, default=1, const [1,24]}
       foot_l_fk_ctrl.Scale.Y {type=MovieSceneScriptingFloatChannel, default=1, const [1,24]}
       foot_l_fk_ctrl.Scale.Z {type=MovieSceneScriptingFloatChannel, default=1, const [1,24]}
       ball_l_fk_ctrl.Location.X {type=MovieSceneScriptingFloatChannel, default=0, thinned_from=24, keys [[0,-1.1663],[340,-1.1663]]}
       ball_l_fk_ctrl.Location.Y {type=MovieSceneScriptingFloatChannel, default=0, thinned_from=24, keys [[0,7.6543],[340,7.6543]]}
       ball_l_fk_ctrl.Location.Z {type=MovieSceneScriptingFloatChannel, default=0, const [-3.6092,24]}
       ball_l_fk_ctrl.Rotation.X {type=MovieSceneScriptingFloatChannel, default=0, thinned_from=24, keys [[0,0],[340,0]]}
       ball_l_fk_ctrl.Rotation.Y {type=MovieSceneScriptingFloatChannel, default=0, thinned_from=24, keys [[0,0],[340,0]]}
       ball_l_fk_ctrl.Rotation.Z {type=MovieSceneScriptingFloatChannel, default=0, thinned_from=24, keys [[0,0],[340,0]]}
       ball_l_fk_ctrl.Scale.X {type=MovieSceneScriptingFloatChannel, default=1, const [1,24]}
       ball_l_fk_ctrl.Scale.Y {type=MovieSceneScriptingFloatChannel, default=1, const [1,24]}
       ball_l_fk_ctrl.Scale.Z {type=MovieSceneScriptingFloatChannel, default=1, const [1,24]}
       foot_l_ik_ctrl.Location.X {type=MovieSceneScriptingFloatChannel, default=0, const [6.0009,24]}
       foot_l_ik_ctrl.Location.Y {type=MovieSceneScriptingFloatChannel, default=0, const [19.2466,24]}
       foot_l_ik_ctrl.Location.Z {type=MovieSceneScriptingFloatChannel, default=0, const [0,24]}
       foot_l_ik_ctrl.Rotation.X {type=MovieSceneScriptingFloatChannel, default=0, const [0,24]}
       foot_l_ik_ctrl.Rotation.Y {type=MovieSceneScriptingFloatChannel, default=0, const [0,24]}
       foot_l_ik_ctrl.Rotation.Z {type=MovieSceneScriptingFloatChannel, default=0, const [34.1832,24]}
       foot_l_ik_ctrl.Scale.X {type=MovieSceneScriptingFloatChannel, default=1, const [1,24]}
       foot_l_ik_ctrl.Scale.Y {type=MovieSceneScriptingFloatChannel, default=1, const [1,24]}
       foot_l_ik_ctrl.Scale.Z {type=MovieSceneScriptingFloatChannel, default=1, const [1,24]}
       ball_l_ik_ctrl.Location.X {type=MovieSceneScriptingFloatChannel, default=0, thinned_from=24, keys [[0,4.2527],[340,4.2527]]}
       ball_l_ik_ctrl.Location.Y {type=MovieSceneScriptingFloatChannel, default=0, thinned_from=24, keys [[0,7.3059],[340,7.3059]]}
       ball_l_ik_ctrl.Location.Z {type=MovieSceneScriptingFloatChannel, default=0, thinned_from=24, keys [[0,-1.2297],[340,-1.2297]]}
       ball_l_ik_ctrl.Rotation.X {type=MovieSceneScriptingFloatChannel, default=0, const [0,24]}
       ball_l_ik_ctrl.Rotation.Y {type=MovieSceneScriptingFloatChannel, default=0, thinned_from=24, keys [[0,0],[340,0]]}
       ball_l_ik_ctrl.Rotation.Z {type=MovieSceneScriptingFloatChannel, default=0, const [0,24]}
       ball_l_ik_ctrl.Scale.X {type=MovieSceneScriptingFloatChannel, default=1, const [1,24]}
       ball_l_ik_ctrl.Scale.Y {type=MovieSceneScriptingFloatChannel, default=1, const [1,24]}
       ball_l_ik_ctrl.Scale.Z {type=MovieSceneScriptingFloatChannel, default=1, const [1,24]}
       foot_bk1_l_ctrl.Location.X {type=MovieSceneScriptingFloatChannel, default=0, const [0,24]}
       foot_bk1_l_ctrl.Location.Y {type=MovieSceneScriptingFloatChannel, default=0, const [0,24]}
       foot_bk1_l_ctrl.Location.Z {type=MovieSceneScriptingFloatChannel, default=0, const [0,24]}
       foot_bk1_l_ctrl.Rotation.X {type=MovieSceneScriptingFloatChannel, default=0, const [0,24]}
       foot_bk1_l_ctrl.Rotation.Y {type=MovieSceneScriptingFloatChannel, default=0, const [0,24]}
       foot_bk1_l_ctrl.Rotation.Z {type=MovieSceneScriptingFloatChannel, default=0, const [0,24]}
       foot_bk1_l_ctrl.Scale.X {type=MovieSceneScriptingFloatChannel, default=1, const [1,24]}
       foot_bk1_l_ctrl.Scale.Y {type=MovieSceneScriptingFloatChannel, default=1, const [1,24]}
       foot_bk1_l_ctrl.Scale.Z {type=MovieSceneScriptingFloatChannel, default=1, const [1,24]}
       foot_roll_l_ctrl.Location.X {type=MovieSceneScriptingFloatChannel, default=0, const [0,24]}
       foot_roll_l_ctrl.Location.Y {type=MovieSceneScriptingFloatChannel, default=0, const [0,24]}
       foot_roll_l_ctrl.Location.Z {type=MovieSceneScriptingFloatChannel, default=0, const [0,24]}
       foot_roll_l_ctrl.Rotation.X {type=MovieSceneScriptingFloatChannel, default=0, const [0,24]}
       foot_roll_l_ctrl.Rotation.Y {type=MovieSceneScriptingFloatChannel, default=0, const [0,24]}
       foot_roll_l_ctrl.Rotation.Z {type=MovieSceneScriptingFloatChannel, default=0, const [0,24]}
       foot_roll_l_ctrl.Scale.X {type=MovieSceneScriptingFloatChannel, default=1, const [1,24]}
       foot_roll_l_ctrl.Scale.Y {type=MovieSceneScriptingFloatChannel, default=1, const [1,24]}
       foot_roll_l_ctrl.Scale.Z {type=MovieSceneScriptingFloatChannel, default=1, const [1,24]}
       roll_blend_l_ctrl {type=MovieSceneScriptingFloatChannel, default=1, const [1,24]}
       tip_l_ctrl.Location.X {type=MovieSceneScriptingFloatChannel, default=0, const [0,24]}
       tip_l_ctrl.Location.Y {type=MovieSceneScriptingFloatChannel, default=0, const [0,24]}
       tip_l_ctrl.Location.Z {type=MovieSceneScriptingFloatChannel, default=0, const [0,24]}
       tip_l_ctrl.Rotation.X {type=MovieSceneScriptingFloatChannel, default=0, const [0,24]}
       tip_l_ctrl.Rotation.Y {type=MovieSceneScriptingFloatChannel, default=0, const [0,24]}
       tip_l_ctrl.Rotation.Z {type=MovieSceneScriptingFloatChannel, default=0, const [0,24]}
       tip_l_ctrl.Scale.X {type=MovieSceneScriptingFloatChannel, default=1, const [1,24]}
       tip_l_ctrl.Scale.Y {type=MovieSceneScriptingFloatChannel, default=1, const [1,24]}
       tip_l_ctrl.Scale.Z {type=MovieSceneScriptingFloatChannel, default=1, const [1,24]}
       heel_l_ctrl.Location.X {type=MovieSceneScriptingFloatChannel, default=0, const [0,24]}
       heel_l_ctrl.Location.Y {type=MovieSceneScriptingFloatChannel, default=0, const [0,24]}
       heel_l_ctrl.Location.Z {type=MovieSceneScriptingFloatChannel, default=0, const [0,24]}
       heel_l_ctrl.Rotation.X {type=MovieSceneScriptingFloatChannel, default=0, const [0,24]}
       heel_l_ctrl.Rotation.Y {type=MovieSceneScriptingFloatChannel, default=0, const [0,24]}
       heel_l_ctrl.Rotation.Z {type=MovieSceneScriptingFloatChannel, default=0, const [0,24]}
       heel_l_ctrl.Scale.X {type=MovieSceneScriptingFloatChannel, default=1, const [1,24]}
       heel_l_ctrl.Scale.Y {type=MovieSceneScriptingFloatChannel, default=1, const [1,24]}
       heel_l_ctrl.Scale.Z {type=MovieSceneScriptingFloatChannel, default=1, const [1,24]}
       leg_l_softness {type=MovieSceneScriptingFloatChannel, default=0, const [0,24]}
       leg_l_pv_ik_ctrl.Location.X {type=MovieSceneScriptingFloatChannel, default=0, const [0,24]}
       leg_l_pv_ik_ctrl.Location.Y {type=MovieSceneScriptingFloatChannel, default=0, const [0,24]}
       leg_l_pv_ik_ctrl.Location.Z {type=MovieSceneScriptingFloatChannel, default=0, const [0,24]}
       leg_l_pv_ik_ctrl.Rotation.X {type=MovieSceneScriptingFloatChannel, default=0, const [0,24]}
       leg_l_pv_ik_ctrl.Rotation.Y {type=MovieSceneScriptingFloatChannel, default=0, const [0,24]}
       leg_l_pv_ik_ctrl.Rotation.Z {type=MovieSceneScriptingFloatChannel, default=0, const [0,24]}
       leg_l_pv_ik_ctrl.Scale.X {type=MovieSceneScriptingFloatChannel, default=1, const [1,24]}
       leg_l_pv_ik_ctrl.Scale.Y {type=MovieSceneScriptingFloatChannel, default=1, const [1,24]}
       leg_l_pv_ik_ctrl.Scale.Z {type=MovieSceneScriptingFloatChannel, default=1, const [1,24]}
       thigh_r_fk_ctrl.Location.X {type=MovieSceneScriptingFloatChannel, default=0, thinned_from=24, keys [[0,0],[340,0]]}
       thigh_r_fk_ctrl.Location.Y {type=MovieSceneScriptingFloatChannel, default=0, thinned_from=24, keys [[0,0],[340,0]]}
       thigh_r_fk_ctrl.Location.Z {type=MovieSceneScriptingFloatChannel, default=0, thinned_from=24, keys [[0,0],[340,0]]}
       thigh_r_fk_ctrl.Rotation.X {
        type=MovieSceneScriptingFloatChannel
        default=1.0528
        thinned_from=24
        keys [[0,-22.7626],[25,-23.169],[60,-23.3823],[84,-23.3823],[92,-23.7402],[112,-23.7402],[122,-23.5144],[135,-23.5144],[139,-23.2988],[146,-23.021],[160,-22.204],[175,-22.204],[178,-21.83],[185,-21.682],[205,-23.9535],[217,-23.9535],[236,-26.622],[269,-26.657],[295,-23.6586],[311,-23.6586],[327,-22.8746],[340,-22.7626]]
       }
       thigh_r_fk_ctrl.Rotation.Y {
        type=MovieSceneScriptingFloatChannel
        default=-1.026
        keys [[0,2.3235],[25,2.8975],[46,3.0888],[60,3.1982],[84,3.1982],[92,2.9218],[112,2.9218],[122,3.7977],[135,3.7977],[139,3.8442],[146,4.2655],[160,5.2755],[175,5.2755],[178,4.7459],[185,4.4249],[205,-1.0101],[217,-1.0101],[236,-3.2212],[255,-3.2504],[269,-3.2504],[295,-0.5485],[311,-0.5485],[327,1.9645],[340,2.3235]]
       }
       thigh_r_fk_ctrl.Rotation.Z {
        type=MovieSceneScriptingFloatChannel
        default=-0.07
        thinned_from=24
        keys [[0,-34.9668],[25,-35.1423],[60,-35.2336],[84,-35.2336],[92,-36.0192],[135,-35.9894],[146,-35.8544],[160,-36.2987],[175,-36.2987],[185,-35.9723],[205,-34.3377],[217,-34.3377],[236,-34.7813],[269,-34.7871],[295,-34.4389],[311,-34.4389],[327,-34.9008],[340,-34.9668]]
       }
       thigh_r_fk_ctrl.Scale.X {type=MovieSceneScriptingFloatChannel, default=1, const [1,24]}
       thigh_r_fk_ctrl.Scale.Y {type=MovieSceneScriptingFloatChannel, default=1, const [1,24]}
       thigh_r_fk_ctrl.Scale.Z {type=MovieSceneScriptingFloatChannel, default=1, const [1,24]}
       calf_r_fk_ctrl.Location.X {type=MovieSceneScriptingFloatChannel, default=0, thinned_from=24, keys [[0,0],[340,0]]}
       calf_r_fk_ctrl.Location.Y {type=MovieSceneScriptingFloatChannel, default=0, thinned_from=24, keys [[0,0],[340,0]]}
       calf_r_fk_ctrl.Location.Z {type=MovieSceneScriptingFloatChannel, default=0, thinned_from=24, keys [[0,0],[340,0]]}
       calf_r_fk_ctrl.Rotation.X {
        type=MovieSceneScriptingFloatChannel
        default=0.0752
        thinned_from=24
        keys [[0,0.8251],[25,0.829],[112,0.8318],[122,0.8332],[175,0.8338],[178,0.8313],[185,0.8291],[205,0.7945],[217,0.7945],[236,0.7864],[269,0.7864],[295,0.8024],[311,0.8024],[327,0.8222],[340,0.8251]]
       }
       calf_r_fk_ctrl.Rotation.Y {
        type=MovieSceneScriptingFloatChannel
        default=1.0359
        thinned_from=24
        keys [[0,0.6309],[25,0.6257],[60,0.623],[84,0.623],[92,0.622],[112,0.622],[122,0.62],[146,0.6186],[175,0.6193],[178,0.6226],[185,0.6255],[205,0.669],[217,0.669],[236,0.6785],[269,0.6785],[295,0.6595],[311,0.6595],[327,0.6345],[340,0.6309]]
       }
       calf_r_fk_ctrl.Rotation.Z {
        type=MovieSceneScriptingFloatChannel
        default=0.0678
        thinned_from=24
        keys [[0,48.515],[25,48.8744],[60,49.0611],[84,49.0611],[112,49.1281],[122,49.2651],[146,49.3665],[175,49.3199],[178,49.0895],[185,48.8879],[205,45.8196],[217,45.8196],[236,45.1318],[269,45.1318],[295,46.5032],[311,46.5032],[327,48.2635],[340,48.515]]
       }
       calf_r_fk_ctrl.Scale.X {type=MovieSceneScriptingFloatChannel, default=1, const [1,24]}
       calf_r_fk_ctrl.Scale.Y {type=MovieSceneScriptingFloatChannel, default=1, const [1,24]}
       calf_r_fk_ctrl.Scale.Z {type=MovieSceneScriptingFloatChannel, default=1, const [1,24]}
       foot_r_fk_ctrl.Location.X {type=MovieSceneScriptingFloatChannel, default=0, thinned_from=24, keys [[0,0],[340,0]]}
       foot_r_fk_ctrl.Location.Y {
        type=MovieSceneScriptingFloatChannel
        default=-0.0543
        thinned_from=24
        keys [[0,-0.0543],[236,-0.0543],[255,-0.0465],[269,-0.0465],[295,-0.0543],[340,-0.0543]]
       }
       foot_r_fk_ctrl.Location.Z {type=MovieSceneScriptingFloatChannel, default=-0.8609, thinned_from=24, keys [[0,-0.8609],[340,-0.8609]]}
       foot_r_fk_ctrl.Rotation.X {
        type=MovieSceneScriptingFloatChannel
        default=-0.9802
        thinned_from=24
        keys [[0,21.1319],[84,21.1278],[112,21.1014],[185,21.1385],[205,21.0164],[217,21.0164],[236,20.8776],[269,20.8757],[295,21.0519],[311,21.0519],[327,21.1219],[340,21.1319]]
       }
       foot_r_fk_ctrl.Rotation.Y {
        type=MovieSceneScriptingFloatChannel
        default=0.1214
        thinned_from=24
        keys [[0,-10.2089],[25,-10.1072],[60,-10.054],[84,-10.054],[92,-9.9869],[112,-9.9869],[122,-10.0523],[135,-10.0523],[139,-9.9714],[146,-9.986],[160,-10.1765],[175,-10.1765],[185,-10.2504],[205,-11.3789],[217,-11.3789],[236,-11.7858],[269,-11.7913],[295,-10.9747],[311,-10.9747],[327,-10.3046],[340,-10.2089]]
       }
       foot_r_fk_ctrl.Rotation.Z {
        type=MovieSceneScriptingFloatChannel
        default=0.0013
        thinned_from=24
        keys [[0,-30.4945],[25,-30.6424],[60,-30.7192],[84,-30.7192],[92,-30.6335],[112,-30.6335],[122,-30.7927],[135,-30.7927],[146,-30.931],[175,-30.9465],[178,-30.793],[185,-30.7008],[205,-28.8985],[217,-28.8985],[236,-28.1007],[269,-28.0995],[295,-29.3079],[311,-29.3079],[327,-30.3462],[340,-30.4945]]
       }
       foot_r_fk_ctrl.Scale.X {type=MovieSceneScriptingFloatChannel, default=1, const [1,24]}
       foot_r_fk_ctrl.Scale.Y {type=MovieSceneScriptingFloatChannel, default=1, const [1,24]}
       foot_r_fk_ctrl.Scale.Z {type=MovieSceneScriptingFloatChannel, default=1, const [1,24]}
       ball_r_fk_ctrl.Location.X {type=MovieSceneScriptingFloatChannel, default=0, const [0,24]}
       ball_r_fk_ctrl.Location.Y {type=MovieSceneScriptingFloatChannel, default=0, const [0,24]}
       ball_r_fk_ctrl.Location.Z {type=MovieSceneScriptingFloatChannel, default=0, const [0,24]}
       ball_r_fk_ctrl.Rotation.X {type=MovieSceneScriptingFloatChannel, default=0, thinned_from=24, keys [[0,0],[340,0]]}
       ball_r_fk_ctrl.Rotation.Y {type=MovieSceneScriptingFloatChannel, default=0, thinned_from=24, keys [[0,0],[340,0]]}
       ball_r_fk_ctrl.Rotation.Z {type=MovieSceneScriptingFloatChannel, default=0, const [0,24]}
       ball_r_fk_ctrl.Scale.X {type=MovieSceneScriptingFloatChannel, default=1, const [1,24]}
       ball_r_fk_ctrl.Scale.Y {type=MovieSceneScriptingFloatChannel, default=1, const [1,24]}
       ball_r_fk_ctrl.Scale.Z {type=MovieSceneScriptingFloatChannel, default=1, const [1,24]}
       foot_r_ik_ctrl.Location.X {type=MovieSceneScriptingFloatChannel, default=0, thinned_from=24, keys [[0,4.8731],[340,4.8731]]}
       foot_r_ik_ctrl.Location.Y {type=MovieSceneScriptingFloatChannel, default=0, const [-15.684,24]}
       foot_r_ik_ctrl.Location.Z {type=MovieSceneScriptingFloatChannel, default=0, const [0,24]}
       foot_r_ik_ctrl.Rotation.X {type=MovieSceneScriptingFloatChannel, default=0, const [0,24]}
       foot_r_ik_ctrl.Rotation.Y {type=MovieSceneScriptingFloatChannel, default=0, const [0,24]}
       foot_r_ik_ctrl.Rotation.Z {type=MovieSceneScriptingFloatChannel, default=0, const [-26.123,24]}
       foot_r_ik_ctrl.Scale.X {type=MovieSceneScriptingFloatChannel, default=1, const [1,24]}
       foot_r_ik_ctrl.Scale.Y {type=MovieSceneScriptingFloatChannel, default=1, const [1,24]}
       foot_r_ik_ctrl.Scale.Z {type=MovieSceneScriptingFloatChannel, default=1, const [1,24]}
       ball_r_ik_ctrl.Location.X {type=MovieSceneScriptingFloatChannel, default=0, const [0,24]}
       ball_r_ik_ctrl.Location.Y {type=MovieSceneScriptingFloatChannel, default=0, const [0,24]}
       ball_r_ik_ctrl.Location.Z {type=MovieSceneScriptingFloatChannel, default=0, const [0,24]}
       ball_r_ik_ctrl.Rotation.X {type=MovieSceneScriptingFloatChannel, default=0, thinned_from=24, keys [[0,0],[340,0]]}
       ball_r_ik_ctrl.Rotation.Y {type=MovieSceneScriptingFloatChannel, default=0, const [0,24]}
       ball_r_ik_ctrl.Rotation.Z {type=MovieSceneScriptingFloatChannel, default=0, const [0,24]}
       ball_r_ik_ctrl.Scale.X {type=MovieSceneScriptingFloatChannel, default=1, const [1,24]}
       ball_r_ik_ctrl.Scale.Y {type=MovieSceneScriptingFloatChannel, default=1, const [1,24]}
       ball_r_ik_ctrl.Scale.Z {type=MovieSceneScriptingFloatChannel, default=1, const [1,24]}
       foot_bk1_r_ctrl.Location.X {type=MovieSceneScriptingFloatChannel, default=0, const [0,24]}
       foot_bk1_r_ctrl.Location.Y {type=MovieSceneScriptingFloatChannel, default=0, const [0,24]}
       foot_bk1_r_ctrl.Location.Z {type=MovieSceneScriptingFloatChannel, default=0, const [0,24]}
       foot_bk1_r_ctrl.Rotation.X {type=MovieSceneScriptingFloatChannel, default=0, thinned_from=24, keys [[0,0],[340,0]]}
       foot_bk1_r_ctrl.Rotation.Y {type=MovieSceneScriptingFloatChannel, default=0, const [0,24]}
       foot_bk1_r_ctrl.Rotation.Z {type=MovieSceneScriptingFloatChannel, default=0, thinned_from=24, keys [[0,0],[340,0]]}
       foot_bk1_r_ctrl.Scale.X {type=MovieSceneScriptingFloatChannel, default=1, const [1,24]}
       foot_bk1_r_ctrl.Scale.Y {type=MovieSceneScriptingFloatChannel, default=1, const [1,24]}
       foot_bk1_r_ctrl.Scale.Z {type=MovieSceneScriptingFloatChannel, default=1, const [1,24]}
       foot_roll_r_ctrl.Location.X {type=MovieSceneScriptingFloatChannel, default=0, const [0,24]}
       foot_roll_r_ctrl.Location.Y {type=MovieSceneScriptingFloatChannel, default=0, const [0,24]}
       foot_roll_r_ctrl.Location.Z {type=MovieSceneScriptingFloatChannel, default=0, const [0,24]}
       foot_roll_r_ctrl.Rotation.X {type=MovieSceneScriptingFloatChannel, default=0, const [0,24]}
       foot_roll_r_ctrl.Rotation.Y {type=MovieSceneScriptingFloatChannel, default=0, const [0,24]}
       foot_roll_r_ctrl.Rotation.Z {type=MovieSceneScriptingFloatChannel, default=0, thinned_from=24, keys [[0,0],[340,0]]}
       foot_roll_r_ctrl.Scale.X {type=MovieSceneScriptingFloatChannel, default=1, const [1,24]}
       foot_roll_r_ctrl.Scale.Y {type=MovieSceneScriptingFloatChannel, default=1, const [1,24]}
       foot_roll_r_ctrl.Scale.Z {type=MovieSceneScriptingFloatChannel, default=1, const [1,24]}
       roll_blend_r_ctrl {type=MovieSceneScriptingFloatChannel, default=1, const [1,24]}
       tip_r_ctrl.Location.X {type=MovieSceneScriptingFloatChannel, default=0, const [0,24]}
       tip_r_ctrl.Location.Y {type=MovieSceneScriptingFloatChannel, default=0, const [0,24]}
       tip_r_ctrl.Location.Z {type=MovieSceneScriptingFloatChannel, default=0, const [0,24]}
       tip_r_ctrl.Rotation.X {type=MovieSceneScriptingFloatChannel, default=0, const [0,24]}
       tip_r_ctrl.Rotation.Y {type=MovieSceneScriptingFloatChannel, default=0, const [0,24]}
       tip_r_ctrl.Rotation.Z {type=MovieSceneScriptingFloatChannel, default=0, const [0,24]}
       tip_r_ctrl.Scale.X {type=MovieSceneScriptingFloatChannel, default=1, const [1,24]}
       tip_r_ctrl.Scale.Y {type=MovieSceneScriptingFloatChannel, default=1, const [1,24]}
       tip_r_ctrl.Scale.Z {type=MovieSceneScriptingFloatChannel, default=1, const [1,24]}
       heel_r_ctrl.Location.X {type=MovieSceneScriptingFloatChannel, default=0, thinned_from=24, keys [[0,0],[340,0]]}
       heel_r_ctrl.Location.Y {type=MovieSceneScriptingFloatChannel, default=0, thinned_from=24, keys [[0,0],[340,0]]}
       heel_r_ctrl.Location.Z {type=MovieSceneScriptingFloatChannel, default=0, thinned_from=24, keys [[0,0],[340,0]]}
       heel_r_ctrl.Rotation.X {type=MovieSceneScriptingFloatChannel, default=0, const [0,24]}
       heel_r_ctrl.Rotation.Y {type=MovieSceneScriptingFloatChannel, default=0, thinned_from=24, keys [[0,0],[340,0]]}
       heel_r_ctrl.Rotation.Z {type=MovieSceneScriptingFloatChannel, default=0, const [0,24]}
       heel_r_ctrl.Scale.X {type=MovieSceneScriptingFloatChannel, default=1, const [1,24]}
       heel_r_ctrl.Scale.Y {type=MovieSceneScriptingFloatChannel, default=1, const [1,24]}
       heel_r_ctrl.Scale.Z {type=MovieSceneScriptingFloatChannel, default=1, const [1,24]}
       leg_r_softness {type=MovieSceneScriptingFloatChannel, default=0, const [0,24]}
       leg_r_pv_ik_ctrl.Location.X {type=MovieSceneScriptingFloatChannel, default=0, const [-25.2055,24]}
       leg_r_pv_ik_ctrl.Location.Y {type=MovieSceneScriptingFloatChannel, default=0, const [15.679,24]}
       leg_r_pv_ik_ctrl.Location.Z {type=MovieSceneScriptingFloatChannel, default=0, const [19.9869,24]}
       leg_r_pv_ik_ctrl.Rotation.X {type=MovieSceneScriptingFloatChannel, default=0, const [-23.6188,24]}
       leg_r_pv_ik_ctrl.Rotation.Y {type=MovieSceneScriptingFloatChannel, default=0, thinned_from=24, keys [[0,-0.6007],[340,-0.6007]]}
       leg_r_pv_ik_ctrl.Rotation.Z {type=MovieSceneScriptingFloatChannel, default=0, thinned_from=24, keys [[0,1.0725],[340,1.0725]]}
       leg_r_pv_ik_ctrl.Scale.X {type=MovieSceneScriptingFloatChannel, default=1, const [1,24]}
       leg_r_pv_ik_ctrl.Scale.Y {type=MovieSceneScriptingFloatChannel, default=1, const [1,24]}
       leg_r_pv_ik_ctrl.Scale.Z {type=MovieSceneScriptingFloatChannel, default=1, const [1,24]}
       upperarm_l_fk_ctrl.Location.X {type=MovieSceneScriptingFloatChannel, default=0, thinned_from=29, keys [[0,0],[340,0]]}
       upperarm_l_fk_ctrl.Location.Y {type=MovieSceneScriptingFloatChannel, default=0, thinned_from=29, keys [[0,0],[340,0]]}
       upperarm_l_fk_ctrl.Location.Z {type=MovieSceneScriptingFloatChannel, default=0, thinned_from=29, keys [[0,0],[340,0]]}
       upperarm_l_fk_ctrl.Rotation.X {
        type=MovieSceneScriptingFloatChannel
        default=0
        thinned_from=29
        keys [[0,24.9803],[146,24.9803],[160,26.4665],[221,26.4665],[226,27.575],[236,30.9005],[269,30.9005],[287,30.7496],[311,30.7496],[327,25.7014],[340,24.9803]]
       }
       upperarm_l_fk_ctrl.Rotation.Y {
        type=MovieSceneScriptingFloatChannel
        default=0
        thinned_from=29
        keys [[0,4.0567],[146,4.0567],[160,8.0047],[221,8.0047],[226,6.3882],[236,1.5384],[269,1.5384],[282,2.918],[287,3.5609],[311,3.5609],[327,3.9947],[340,4.0567]]
       }
       upperarm_l_fk_ctrl.Rotation.Z {
        type=MovieSceneScriptingFloatChannel
        default=0
        thinned_from=29
        keys [[0,8.9326],[146,8.9326],[160,6.5939],[221,6.5939],[226,7.3415],[236,9.5844],[269,9.5844],[282,7.2738],[287,6.1928],[311,6.1928],[327,8.5901],[340,8.9326]]
       }
       upperarm_l_fk_ctrl.Scale.X {type=MovieSceneScriptingFloatChannel, default=1, const [1,29]}
       upperarm_l_fk_ctrl.Scale.Y {type=MovieSceneScriptingFloatChannel, default=1, const [1,29]}
       upperarm_l_fk_ctrl.Scale.Z {type=MovieSceneScriptingFloatChannel, default=1, const [1,29]}
       lowerarm_l_fk_ctrl.Location.X {type=MovieSceneScriptingFloatChannel, default=0, thinned_from=27, keys [[0,0],[340,0]]}
       lowerarm_l_fk_ctrl.Location.Y {type=MovieSceneScriptingFloatChannel, default=0, thinned_from=27, keys [[0,0],[340,0]]}
       lowerarm_l_fk_ctrl.Location.Z {type=MovieSceneScriptingFloatChannel, default=0, thinned_from=27, keys [[0,0],[340,0]]}
       lowerarm_l_fk_ctrl.Rotation.X {
        type=MovieSceneScriptingFloatChannel
        default=0
        thinned_from=27
        keys [[0,-9.1794],[221,-9.1794],[226,-9.0679],[236,-8.7334],[311,-8.7334],[327,-9.1237],[340,-9.1794]]
       }
       lowerarm_l_fk_ctrl.Rotation.Y {
        type=MovieSceneScriptingFloatChannel
        default=0
        thinned_from=27
        keys [[0,6.1338],[221,6.1338],[226,6.2894],[236,6.756],[311,6.756],[327,6.2116],[340,6.1338]]
       }
       lowerarm_l_fk_ctrl.Rotation.Z {
        type=MovieSceneScriptingFloatChannel
        default=0
        thinned_from=27
        keys [[0,101.2294],[221,101.2294],[226,102.2223],[236,105.2009],[311,105.2009],[327,101.7258],[340,101.2294]]
       }
       lowerarm_l_fk_ctrl.Scale.X {type=MovieSceneScriptingFloatChannel, default=1, const [1,27]}
       lowerarm_l_fk_ctrl.Scale.Y {type=MovieSceneScriptingFloatChannel, default=1, const [1,27]}
       lowerarm_l_fk_ctrl.Scale.Z {type=MovieSceneScriptingFloatChannel, default=1, const [1,27]}
       hand_l_fk_ctrl.Location.X {type=MovieSceneScriptingFloatChannel, default=0, thinned_from=25, keys [[0,0],[340,0]]}
       hand_l_fk_ctrl.Location.Y {type=MovieSceneScriptingFloatChannel, default=0, thinned_from=25, keys [[0,0],[340,0]]}
       hand_l_fk_ctrl.Location.Z {type=MovieSceneScriptingFloatChannel, default=0, thinned_from=25, keys [[0,0],[340,0]]}
       hand_l_fk_ctrl.Rotation.X {type=MovieSceneScriptingFloatChannel, default=0, thinned_from=25, keys [[0,-0.7131],[340,-0.7131]]}
       hand_l_fk_ctrl.Rotation.Y {type=MovieSceneScriptingFloatChannel, default=0, const [23.0471,25]}
       hand_l_fk_ctrl.Rotation.Z {type=MovieSceneScriptingFloatChannel, default=0, const [-37.0127,25]}
       hand_l_fk_ctrl.Scale.X {type=MovieSceneScriptingFloatChannel, default=1, const [1,25]}
       hand_l_fk_ctrl.Scale.Y {type=MovieSceneScriptingFloatChannel, default=1, const [1,25]}
       hand_l_fk_ctrl.Scale.Z {type=MovieSceneScriptingFloatChannel, default=1, const [1,25]}
       thumb_01_l_ctrl.Location.X {type=MovieSceneScriptingFloatChannel, default=0, thinned_from=24, keys [[0,0],[340,0]]}
       thumb_01_l_ctrl.Location.Y {type=MovieSceneScriptingFloatChannel, default=0, thinned_from=24, keys [[0,0],[340,0]]}
       thumb_01_l_ctrl.Location.Z {type=MovieSceneScriptingFloatChannel, default=0, thinned_from=24, keys [[0,0],[340,0]]}
       thumb_01_l_ctrl.Rotation.X {type=MovieSceneScriptingFloatChannel, default=0, const [-11.8305,24]}
       thumb_01_l_ctrl.Rotation.Y {type=MovieSceneScriptingFloatChannel, default=0, thinned_from=24, keys [[0,-6.959],[340,-6.959]]}
       thumb_01_l_ctrl.Rotation.Z {type=MovieSceneScriptingFloatChannel, default=0, thinned_from=24, keys [[0,4.4404],[340,4.4404]]}
       thumb_01_l_ctrl.Scale.X {type=MovieSceneScriptingFloatChannel, default=1, const [1,24]}
       thumb_01_l_ctrl.Scale.Y {type=MovieSceneScriptingFloatChannel, default=1, const [1,24]}
       thumb_01_l_ctrl.Scale.Z {type=MovieSceneScriptingFloatChannel, default=1, const [1,24]}
       thumb_02_l_ctrl.Location.X {type=MovieSceneScriptingFloatChannel, default=0, thinned_from=24, keys [[0,0],[340,0]]}
       thumb_02_l_ctrl.Location.Y {type=MovieSceneScriptingFloatChannel, default=0, thinned_from=24, keys [[0,0],[340,0]]}
       thumb_02_l_ctrl.Location.Z {type=MovieSceneScriptingFloatChannel, default=0, thinned_from=24, keys [[0,0],[340,0]]}
       thumb_02_l_ctrl.Rotation.X {type=MovieSceneScriptingFloatChannel, default=0, thinned_from=24, keys [[0,0],[340,0]]}
       thumb_02_l_ctrl.Rotation.Y {type=MovieSceneScriptingFloatChannel, default=0, thinned_from=24, keys [[0,0],[340,0]]}
       thumb_02_l_ctrl.Rotation.Z {type=MovieSceneScriptingFloatChannel, default=0, const [-12.3705,24]}
       thumb_02_l_ctrl.Scale.X {type=MovieSceneScriptingFloatChannel, default=1, const [1,24]}
       thumb_02_l_ctrl.Scale.Y {type=MovieSceneScriptingFloatChannel, default=1, const [1,24]}
       thumb_02_l_ctrl.Scale.Z {type=MovieSceneScriptingFloatChannel, default=1, const [1,24]}
       thumb_03_l_ctrl.Location.X {type=MovieSceneScriptingFloatChannel, default=0, thinned_from=24, keys [[0,-0.5596],[340,-0.5596]]}
       thumb_03_l_ctrl.Location.Y {type=MovieSceneScriptingFloatChannel, default=0, thinned_from=24, keys [[0,-0.2442],[340,-0.2442]]}
       thumb_03_l_ctrl.Location.Z {type=MovieSceneScriptingFloatChannel, default=0, thinned_from=24, keys [[0,0],[340,0]]}
       thumb_03_l_ctrl.Rotation.X {type=MovieSceneScriptingFloatChannel, default=0, thinned_from=24, keys [[0,0],[340,0]]}
       thumb_03_l_ctrl.Rotation.Y {type=MovieSceneScriptingFloatChannel, default=0, thinned_from=24, keys [[0,0],[340,0]]}
       thumb_03_l_ctrl.Rotation.Z {type=MovieSceneScriptingFloatChannel, default=0, const [-66.4234,24]}
       thumb_03_l_ctrl.Scale.X {type=MovieSceneScriptingFloatChannel, default=1, const [1,24]}
       thumb_03_l_ctrl.Scale.Y {type=MovieSceneScriptingFloatChannel, default=1, const [1,24]}
       thumb_03_l_ctrl.Scale.Z {type=MovieSceneScriptingFloatChannel, default=1, const [1,24]}
       index_metacarpal_l_ctrl.Location.X {type=MovieSceneScriptingFloatChannel, default=0, thinned_from=24, keys [[0,0],[340,0]]}
       index_metacarpal_l_ctrl.Location.Y {type=MovieSceneScriptingFloatChannel, default=0, thinned_from=24, keys [[0,0],[340,0]]}
       index_metacarpal_l_ctrl.Location.Z {type=MovieSceneScriptingFloatChannel, default=0, thinned_from=24, keys [[0,0],[340,0]]}
       index_metacarpal_l_ctrl.Rotation.X {type=MovieSceneScriptingFloatChannel, default=0, const [-2.7256,24]}
       index_metacarpal_l_ctrl.Rotation.Y {type=MovieSceneScriptingFloatChannel, default=0, thinned_from=24, keys [[0,0],[340,0]]}
       index_metacarpal_l_ctrl.Rotation.Z {type=MovieSceneScriptingFloatChannel, default=0, thinned_from=24, keys [[0,-7.3778],[340,-7.3778]]}
       index_metacarpal_l_ctrl.Scale.X {type=MovieSceneScriptingFloatChannel, default=1, const [1,24]}
       index_metacarpal_l_ctrl.Scale.Y {type=MovieSceneScriptingFloatChannel, default=1, const [1,24]}
       index_metacarpal_l_ctrl.Scale.Z {type=MovieSceneScriptingFloatChannel, default=1, const [1,24]}
       index_01_l_ctrl.Location.X {type=MovieSceneScriptingFloatChannel, default=0, thinned_from=24, keys [[0,0],[340,0]]}
       index_01_l_ctrl.Location.Y {type=MovieSceneScriptingFloatChannel, default=0, thinned_from=24, keys [[0,0],[340,0]]}
       index_01_l_ctrl.Location.Z {type=MovieSceneScriptingFloatChannel, default=0, thinned_from=24, keys [[0,0],[340,0]]}
       index_01_l_ctrl.Rotation.X {type=MovieSceneScriptingFloatChannel, default=0, thinned_from=24, keys [[0,0],[340,0]]}
       index_01_l_ctrl.Rotation.Y {type=MovieSceneScriptingFloatChannel, default=0, const [0,24]}
       index_01_l_ctrl.Rotation.Z {type=MovieSceneScriptingFloatChannel, default=0, const [-64.9002,24]}
       index_01_l_ctrl.Scale.X {type=MovieSceneScriptingFloatChannel, default=1, const [1,24]}
       index_01_l_ctrl.Scale.Y {type=MovieSceneScriptingFloatChannel, default=1, const [1,24]}
       index_01_l_ctrl.Scale.Z {type=MovieSceneScriptingFloatChannel, default=1, const [1,24]}
       index_02_l_ctrl.Location.X {type=MovieSceneScriptingFloatChannel, default=0, const [0,24]}
       index_02_l_ctrl.Location.Y {type=MovieSceneScriptingFloatChannel, default=0, const [0,24]}
       index_02_l_ctrl.Location.Z {type=MovieSceneScriptingFloatChannel, default=0, const [0,24]}
       index_02_l_ctrl.Rotation.X {type=MovieSceneScriptingFloatChannel, default=0, const [0,24]}
       index_02_l_ctrl.Rotation.Y {type=MovieSceneScriptingFloatChannel, default=0, const [0,24]}
       index_02_l_ctrl.Rotation.Z {type=MovieSceneScriptingFloatChannel, default=0, const [-49.5772,24]}
       index_02_l_ctrl.Scale.X {type=MovieSceneScriptingFloatChannel, default=1, const [1,24]}
       index_02_l_ctrl.Scale.Y {type=MovieSceneScriptingFloatChannel, default=1, const [1,24]}
       index_02_l_ctrl.Scale.Z {type=MovieSceneScriptingFloatChannel, default=1, const [1,24]}
       index_03_l_ctrl.Location.X {type=MovieSceneScriptingFloatChannel, default=0, const [0,24]}
       index_03_l_ctrl.Location.Y {type=MovieSceneScriptingFloatChannel, default=0, const [0,24]}
       index_03_l_ctrl.Location.Z {type=MovieSceneScriptingFloatChannel, default=0, const [0,24]}
       index_03_l_ctrl.Rotation.X {type=MovieSceneScriptingFloatChannel, default=0, const [0,24]}
       index_03_l_ctrl.Rotation.Y {type=MovieSceneScriptingFloatChannel, default=0, thinned_from=24, keys [[0,0],[340,0]]}
       index_03_l_ctrl.Rotation.Z {type=MovieSceneScriptingFloatChannel, default=0, const [-102.5714,24]}
       index_03_l_ctrl.Scale.X {type=MovieSceneScriptingFloatChannel, default=1, const [1,24]}
       index_03_l_ctrl.Scale.Y {type=MovieSceneScriptingFloatChannel, default=1, const [1,24]}
       index_03_l_ctrl.Scale.Z {type=MovieSceneScriptingFloatChannel, default=1, const [1,24]}
       middle_metacarpal_l_ctrl.Location.X {type=MovieSceneScriptingFloatChannel, default=0, const [0,24]}
       middle_metacarpal_l_ctrl.Location.Y {type=MovieSceneScriptingFloatChannel, default=0, const [0,24]}
       middle_metacarpal_l_ctrl.Location.Z {type=MovieSceneScriptingFloatChannel, default=0, const [0,24]}
       middle_metacarpal_l_ctrl.Rotation.X {type=MovieSceneScriptingFloatChannel, default=0, const [-2.7256,24]}
       middle_metacarpal_l_ctrl.Rotation.Y {type=MovieSceneScriptingFloatChannel, default=0, const [0,24]}
       middle_metacarpal_l_ctrl.Rotation.Z {type=MovieSceneScriptingFloatChannel, default=0, thinned_from=24, keys [[0,-7.3778],[340,-7.3778]]}
       middle_metacarpal_l_ctrl.Scale.X {type=MovieSceneScriptingFloatChannel, default=1, const [1,24]}
       middle_metacarpal_l_ctrl.Scale.Y {type=MovieSceneScriptingFloatChannel, default=1, const [1,24]}
       middle_metacarpal_l_ctrl.Scale.Z {type=MovieSceneScriptingFloatChannel, default=1, const [1,24]}
       middle_01_l_ctrl.Location.X {type=MovieSceneScriptingFloatChannel, default=0, const [0,24]}
       middle_01_l_ctrl.Location.Y {type=MovieSceneScriptingFloatChannel, default=0, const [0,24]}
       middle_01_l_ctrl.Location.Z {type=MovieSceneScriptingFloatChannel, default=0, const [0,24]}
       middle_01_l_ctrl.Rotation.X {type=MovieSceneScriptingFloatChannel, default=0, const [0,24]}
       middle_01_l_ctrl.Rotation.Y {type=MovieSceneScriptingFloatChannel, default=0, const [0,24]}
       middle_01_l_ctrl.Rotation.Z {type=MovieSceneScriptingFloatChannel, default=0, const [-64.9002,24]}
       middle_01_l_ctrl.Scale.X {type=MovieSceneScriptingFloatChannel, default=1, const [1,24]}
       middle_01_l_ctrl.Scale.Y {type=MovieSceneScriptingFloatChannel, default=1, const [1,24]}
       middle_01_l_ctrl.Scale.Z {type=MovieSceneScriptingFloatChannel, default=1, const [1,24]}
       middle_02_l_ctrl.Location.X {type=MovieSceneScriptingFloatChannel, default=0, const [0,24]}
       middle_02_l_ctrl.Location.Y {type=MovieSceneScriptingFloatChannel, default=0, const [0,24]}
       middle_02_l_ctrl.Location.Z {type=MovieSceneScriptingFloatChannel, default=0, const [0,24]}
       middle_02_l_ctrl.Rotation.X {type=MovieSceneScriptingFloatChannel, default=0, const [0,24]}
       middle_02_l_ctrl.Rotation.Y {type=MovieSceneScriptingFloatChannel, default=0, const [0,24]}
       middle_02_l_ctrl.Rotation.Z {type=MovieSceneScriptingFloatChannel, default=0, const [-49.5772,24]}
       middle_02_l_ctrl.Scale.X {type=MovieSceneScriptingFloatChannel, default=1, const [1,24]}
       middle_02_l_ctrl.Scale.Y {type=MovieSceneScriptingFloatChannel, default=1, const [1,24]}
       middle_02_l_ctrl.Scale.Z {type=MovieSceneScriptingFloatChannel, default=1, const [1,24]}
       middle_03_l_ctrl.Location.X {type=MovieSceneScriptingFloatChannel, default=0, const [0,24]}
       middle_03_l_ctrl.Location.Y {type=MovieSceneScriptingFloatChannel, default=0, const [0,24]}
       middle_03_l_ctrl.Location.Z {type=MovieSceneScriptingFloatChannel, default=0, const [0,24]}
       middle_03_l_ctrl.Rotation.X {type=MovieSceneScriptingFloatChannel, default=0, const [0,24]}
       middle_03_l_ctrl.Rotation.Y {type=MovieSceneScriptingFloatChannel, default=0, thinned_from=24, keys [[0,0],[340,0]]}
       middle_03_l_ctrl.Rotation.Z {type=MovieSceneScriptingFloatChannel, default=0, const [-102.5714,24]}
       middle_03_l_ctrl.Scale.X {type=MovieSceneScriptingFloatChannel, default=1, const [1,24]}
       middle_03_l_ctrl.Scale.Y {type=MovieSceneScriptingFloatChannel, default=1, const [1,24]}
       middle_03_l_ctrl.Scale.Z {type=MovieSceneScriptingFloatChannel, default=1, const [1,24]}
       ring_metacarpal_l_ctrl.Location.X {type=MovieSceneScriptingFloatChannel, default=0, const [0,24]}
       ring_metacarpal_l_ctrl.Location.Y {type=MovieSceneScriptingFloatChannel, default=0, const [0,24]}
       ring_metacarpal_l_ctrl.Location.Z {type=MovieSceneScriptingFloatChannel, default=0, const [0,24]}
       ring_metacarpal_l_ctrl.Rotation.X {type=MovieSceneScriptingFloatChannel, default=0, thinned_from=24, keys [[0,1.2867],[340,1.2867]]}
       ring_metacarpal_l_ctrl.Rotation.Y {type=MovieSceneScriptingFloatChannel, default=0, const [0,24]}
       ring_metacarpal_l_ctrl.Rotation.Z {type=MovieSceneScriptingFloatChannel, default=0, thinned_from=24, keys [[0,-7.3778],[340,-7.3778]]}
       ring_metacarpal_l_ctrl.Scale.X {type=MovieSceneScriptingFloatChannel, default=1, const [1,24]}
       ring_metacarpal_l_ctrl.Scale.Y {type=MovieSceneScriptingFloatChannel, default=1, const [1,24]}
       ring_metacarpal_l_ctrl.Scale.Z {type=MovieSceneScriptingFloatChannel, default=1, const [1,24]}
       ring_01_l_ctrl.Location.X {type=MovieSceneScriptingFloatChannel, default=0, const [0,24]}
       ring_01_l_ctrl.Location.Y {type=MovieSceneScriptingFloatChannel, default=0, const [0,24]}
       ring_01_l_ctrl.Location.Z {type=MovieSceneScriptingFloatChannel, default=0, const [0,24]}
       ring_01_l_ctrl.Rotation.X {type=MovieSceneScriptingFloatChannel, default=0, thinned_from=24, keys [[0,0],[340,0]]}
       ring_01_l_ctrl.Rotation.Y {type=MovieSceneScriptingFloatChannel, default=0, const [0,24]}
       ring_01_l_ctrl.Rotation.Z {type=MovieSceneScriptingFloatChannel, default=0, const [-64.9002,24]}
       ring_01_l_ctrl.Scale.X {type=MovieSceneScriptingFloatChannel, default=1, const [1,24]}
       ring_01_l_ctrl.Scale.Y {type=MovieSceneScriptingFloatChannel, default=1, const [1,24]}
       ring_01_l_ctrl.Scale.Z {type=MovieSceneScriptingFloatChannel, default=1, const [1,24]}
       ring_02_l_ctrl.Location.X {type=MovieSceneScriptingFloatChannel, default=0, const [0,24]}
       ring_02_l_ctrl.Location.Y {type=MovieSceneScriptingFloatChannel, default=0, const [0,24]}
       ring_02_l_ctrl.Location.Z {type=MovieSceneScriptingFloatChannel, default=0, const [0,24]}
       ring_02_l_ctrl.Rotation.X {type=MovieSceneScriptingFloatChannel, default=0, thinned_from=24, keys [[0,0],[340,0]]}
       ring_02_l_ctrl.Rotation.Y {type=MovieSceneScriptingFloatChannel, default=0, thinned_from=24, keys [[0,0],[340,0]]}
       ring_02_l_ctrl.Rotation.Z {type=MovieSceneScriptingFloatChannel, default=0, const [-49.5772,24]}
       ring_02_l_ctrl.Scale.X {type=MovieSceneScriptingFloatChannel, default=1, const [1,24]}
       ring_02_l_ctrl.Scale.Y {type=MovieSceneScriptingFloatChannel, default=1, const [1,24]}
       ring_02_l_ctrl.Scale.Z {type=MovieSceneScriptingFloatChannel, default=1, const [1,24]}
       ring_03_l_ctrl.Location.X {type=MovieSceneScriptingFloatChannel, default=0, const [0,24]}
       ring_03_l_ctrl.Location.Y {type=MovieSceneScriptingFloatChannel, default=0, const [0,24]}
       ring_03_l_ctrl.Location.Z {type=MovieSceneScriptingFloatChannel, default=0, const [0,24]}
       ring_03_l_ctrl.Rotation.X {type=MovieSceneScriptingFloatChannel, default=0, const [0,24]}
       ring_03_l_ctrl.Rotation.Y {type=MovieSceneScriptingFloatChannel, default=0, const [0,24]}
       ring_03_l_ctrl.Rotation.Z {type=MovieSceneScriptingFloatChannel, default=0, const [-102.5714,24]}
       ring_03_l_ctrl.Scale.X {type=MovieSceneScriptingFloatChannel, default=1, const [1,24]}
       ring_03_l_ctrl.Scale.Y {type=MovieSceneScriptingFloatChannel, default=1, const [1,24]}
       ring_03_l_ctrl.Scale.Z {type=MovieSceneScriptingFloatChannel, default=1, const [1,24]}
       pinky_metacarpal_l_ctrl.Location.X {type=MovieSceneScriptingFloatChannel, default=0, const [0,24]}
       pinky_metacarpal_l_ctrl.Location.Y {type=MovieSceneScriptingFloatChannel, default=0, const [0,24]}
       pinky_metacarpal_l_ctrl.Location.Z {type=MovieSceneScriptingFloatChannel, default=0, const [0,24]}
       pinky_metacarpal_l_ctrl.Rotation.X {type=MovieSceneScriptingFloatChannel, default=0, thinned_from=24, keys [[0,1.2867],[340,1.2867]]}
       pinky_metacarpal_l_ctrl.Rotation.Y {type=MovieSceneScriptingFloatChannel, default=0, const [0,24]}
       pinky_metacarpal_l_ctrl.Rotation.Z {type=MovieSceneScriptingFloatChannel, default=0, thinned_from=24, keys [[0,-7.3778],[340,-7.3778]]}
       pinky_metacarpal_l_ctrl.Scale.X {type=MovieSceneScriptingFloatChannel, default=1, const [1,24]}
       pinky_metacarpal_l_ctrl.Scale.Y {type=MovieSceneScriptingFloatChannel, default=1, const [1,24]}
       pinky_metacarpal_l_ctrl.Scale.Z {type=MovieSceneScriptingFloatChannel, default=1, const [1,24]}
       pinky_01_l_ctrl.Location.X {type=MovieSceneScriptingFloatChannel, default=0, const [0,24]}
       pinky_01_l_ctrl.Location.Y {type=MovieSceneScriptingFloatChannel, default=0, const [0,24]}
       pinky_01_l_ctrl.Location.Z {type=MovieSceneScriptingFloatChannel, default=0, const [0,24]}
       pinky_01_l_ctrl.Rotation.X {type=MovieSceneScriptingFloatChannel, default=0, thinned_from=24, keys [[0,0],[340,0]]}
       pinky_01_l_ctrl.Rotation.Y {type=MovieSceneScriptingFloatChannel, default=0, const [0,24]}
       pinky_01_l_ctrl.Rotation.Z {type=MovieSceneScriptingFloatChannel, default=0, const [-64.9002,24]}
       pinky_01_l_ctrl.Scale.X {type=MovieSceneScriptingFloatChannel, default=1, const [1,24]}
       pinky_01_l_ctrl.Scale.Y {type=MovieSceneScriptingFloatChannel, default=1, const [1,24]}
       pinky_01_l_ctrl.Scale.Z {type=MovieSceneScriptingFloatChannel, default=1, const [1,24]}
       pinky_02_l_ctrl.Location.X {type=MovieSceneScriptingFloatChannel, default=0, const [0,24]}
       pinky_02_l_ctrl.Location.Y {type=MovieSceneScriptingFloatChannel, default=0, const [0,24]}
       pinky_02_l_ctrl.Location.Z {type=MovieSceneScriptingFloatChannel, default=0, const [0,24]}
       pinky_02_l_ctrl.Rotation.X {type=MovieSceneScriptingFloatChannel, default=0, const [0,24]}
       pinky_02_l_ctrl.Rotation.Y {type=MovieSceneScriptingFloatChannel, default=0, const [0,24]}
       pinky_02_l_ctrl.Rotation.Z {type=MovieSceneScriptingFloatChannel, default=0, const [-49.5772,24]}
       pinky_02_l_ctrl.Scale.X {type=MovieSceneScriptingFloatChannel, default=1, const [1,24]}
       pinky_02_l_ctrl.Scale.Y {type=MovieSceneScriptingFloatChannel, default=1, const [1,24]}
       pinky_02_l_ctrl.Scale.Z {type=MovieSceneScriptingFloatChannel, default=1, const [1,24]}
       pinky_03_l_ctrl.Location.X {type=MovieSceneScriptingFloatChannel, default=0, const [0,24]}
       pinky_03_l_ctrl.Location.Y {type=MovieSceneScriptingFloatChannel, default=0, const [0,24]}
       pinky_03_l_ctrl.Location.Z {type=MovieSceneScriptingFloatChannel, default=0, const [0,24]}
       pinky_03_l_ctrl.Rotation.X {type=MovieSceneScriptingFloatChannel, default=0, thinned_from=24, keys [[0,0],[340,0]]}
       pinky_03_l_ctrl.Rotation.Y {type=MovieSceneScriptingFloatChannel, default=0, const [0,24]}
       pinky_03_l_ctrl.Rotation.Z {type=MovieSceneScriptingFloatChannel, default=0, const [-102.5714,24]}
       pinky_03_l_ctrl.Scale.X {type=MovieSceneScriptingFloatChannel, default=1, const [1,24]}
       pinky_03_l_ctrl.Scale.Y {type=MovieSceneScriptingFloatChannel, default=1, const [1,24]}
       pinky_03_l_ctrl.Scale.Z {type=MovieSceneScriptingFloatChannel, default=1, const [1,24]}
       hand_l_ik_ctrl.Location.X {
        type=MovieSceneScriptingFloatChannel
        default=0
        thinned_from=25
        keys [[0,-15.0711],[25,-15.5572],[60,-15.8118],[84,-15.8118],[92,-15.718],[112,-15.718],[122,-15.8116],[135,-15.8116],[139,-15.8324],[142,-14.7245],[146,-15.8782],[160,-14.4003],[175,-14.4003],[178,-14.079],[185,-13.7646],[205,-11.8518],[217,-11.8518],[236,-12.6765],[255,-12.6152],[269,-12.6152],[295,-14.9915],[311,-14.9915],[327,-14.7559],[340,-14.7223]]
       }
       hand_l_ik_ctrl.Location.Y {
        type=MovieSceneScriptingFloatChannel
        default=0
        thinned_from=25
        keys [[0,24.1831],[25,24.5965],[60,24.8135],[84,24.8135],[92,25.3259],[112,25.3259],[122,25.1646],[139,25.1669],[142,24.8863],[146,25.005],[160,22.4906],[175,22.4906],[185,22.2405],[205,21.0302],[217,21.0302],[236,24.5384],[255,24.6409],[269,24.6409],[295,23.6084],[311,23.6084],[327,24.1],[340,24.1702]]
       }
       hand_l_ik_ctrl.Location.Z {
        type=MovieSceneScriptingFloatChannel
        default=0
        thinned_from=25
        keys [[0,33.9222],[84,33.9368],[92,34.0408],[139,34.0126],[142,34.6063],[146,34.0279],[160,32.2187],[175,32.2187],[178,32.7315],[185,32.4906],[205,32.9405],[217,32.9405],[236,34.0938],[255,34.1743],[269,34.1743],[295,33.0272],[311,33.0272],[327,34.2954],[340,34.4765]]
       }
       hand_l_ik_ctrl.Rotation.X {
        type=MovieSceneScriptingFloatChannel
        default=0
        thinned_from=25
        keys [[0,-204.4979],[25,-205.2823],[84,-205.689],[92,-207.0827],[112,-207.0827],[122,-206.4794],[135,-206.4794],[139,-206.1075],[142,-206.6051],[146,-205.4965],[160,-210.0716],[175,-210.0716],[178,-209.4872],[185,-209.057],[205,-211.9997],[217,-211.9997],[236,-219.3438],[269,-219.4126],[295,-214.8801],[311,-214.8801],[327,-205.774],[340,-204.4732]]
       }
       hand_l_ik_ctrl.Rotation.Y {
        type=MovieSceneScriptingFloatChannel
        default=0
        thinned_from=25
        keys [[0,61.7604],[25,61.4127],[60,61.2289],[84,61.2289],[92,61.6341],[139,61.6312],[142,59.4753],[146,61.6628],[160,63.0485],[175,63.0485],[178,63.3635],[185,63.4133],[205,62.8771],[217,62.8771],[236,61.3949],[269,61.3887],[295,63.8022],[311,63.8022],[327,59.9479],[340,59.3972]]
       }
       hand_l_ik_ctrl.Rotation.Z {
        type=MovieSceneScriptingFloatChannel
        default=0
        thinned_from=25
        keys [[0,-169.785],[25,-171.1706],[60,-171.8922],[84,-171.8922],[92,-173.0535],[112,-173.0535],[135,-173.3396],[139,-173.1324],[142,-168.2621],[146,-173.0002],[160,-182.0659],[175,-182.0659],[178,-180.9929],[185,-180.1951],[205,-175.7202],[217,-175.7202],[236,-168.5132],[269,-168.5382],[295,-171.5247],[311,-171.5247],[327,-167.2984],[340,-166.6947]]
       }
       hand_l_ik_ctrl.Scale.X {type=MovieSceneScriptingFloatChannel, default=1, const [1,25]}
       hand_l_ik_ctrl.Scale.Y {type=MovieSceneScriptingFloatChannel, default=1, const [1,25]}
       hand_l_ik_ctrl.Scale.Z {type=MovieSceneScriptingFloatChannel, default=1, const [1,25]}
       Softness {type=MovieSceneScriptingFloatChannel, default=0, const [0,25]}
       hand_l_ikorient_ctrl.Rotation.Roll {type=MovieSceneScriptingFloatChannel, default=0, const [0,24]}
       hand_l_ikorient_ctrl.Rotation.Pitch {type=MovieSceneScriptingFloatChannel, default=0, const [0,24]}
       hand_l_ikorient_ctrl.Rotation.Yaw {type=MovieSceneScriptingFloatChannel, default=0, const [0,24]}
       arm_l_pv_ik_ctrl.Location.X {
        type=MovieSceneScriptingFloatChannel
        default=0
        thinned_from=25
        keys [[0,11.0665],[25,10.6677],[60,10.4576],[84,10.4576],[92,10.2113],[135,10.2193],[139,10.1386],[142,10.7698],[146,10.1463],[160,13.5771],[178,13.5919],[185,13.8881],[205,16.4846],[217,16.4846],[236,13.1809],[255,13.1058],[269,13.1058],[295,12.9943],[311,12.9943],[327,11.4293],[340,11.2058]]
       }
       arm_l_pv_ik_ctrl.Location.Y {
        type=MovieSceneScriptingFloatChannel
        default=0
        thinned_from=25
        keys [[0,35.6121],[84,35.5764],[92,35.9474],[112,35.9474],[122,35.5336],[135,35.5336],[139,35.3766],[142,38.5738],[146,35.0923],[160,33.4147],[175,33.4147],[178,33.8932],[185,33.797],[205,37.8212],[217,37.8212],[236,44.102],[255,44.2321],[269,44.2321],[295,38.1985],[311,38.1985],[327,37.937],[340,37.8996]]
       }
       arm_l_pv_ik_ctrl.Location.Z {
        type=MovieSceneScriptingFloatChannel
        default=0
        keys [[0,-4.3132],[25,-3.696],[46,-3.4903],[60,-3.3729],[84,-3.3729],[92,-3.265],[112,-3.265],[122,-2.7905],[135,-2.7905],[139,-2.7619],[142,-7.0591],[146,-2.5203],[160,1.1008],[175,1.1008],[178,0.7349],[185,0.1602],[205,-4.3437],[217,-4.3437],[236,-10.724],[255,-10.8015],[269,-10.8015],[295,-4.6631],[311,-4.6631],[327,-6.7405],[340,-7.0373]]
       }
       arm_l_pv_ik_ctrl.Rotation.X {
        type=MovieSceneScriptingFloatChannel
        default=0
        thinned_from=25
        keys [[0,17.6064],[25,17.9885],[60,18.1887],[84,18.1887],[92,18.7503],[112,18.7503],[122,19.0497],[139,19.0554],[142,14.6862],[146,19.1567],[160,24.1034],[175,24.1034],[178,23.7889],[185,23.4594],[205,19.6908],[217,19.6908],[236,15.439],[269,15.4267],[295,20.4518],[311,20.4518],[327,15.0277],[340,14.2528]]
       }
       arm_l_pv_ik_ctrl.Rotation.Y {
        type=MovieSceneScriptingFloatChannel
        default=0
        keys [[0,1.7979],[25,2.553],[46,2.8047],[60,2.9485],[84,2.9485],[92,2.6298],[112,2.6298],[122,3.2419],[135,3.2419],[139,3.2903],[142,0.8264],[146,3.5527],[160,6.1214],[175,6.1214],[178,5.5276],[185,5.2076],[205,0.5255],[217,0.5255],[236,-6.8003],[255,-6.8168],[269,-6.8168],[295,-4.3003],[311,-4.3003],[327,0.6106],[340,1.3121]]
       }
       arm_l_pv_ik_ctrl.Rotation.Z {
        type=MovieSceneScriptingFloatChannel
        default=0
        thinned_from=25
        keys [[0,52.0694],[60,52.3537],[84,52.3537],[92,52.8237],[112,52.8237],[122,52.3023],[135,52.3023],[139,52.0886],[142,54.6646],[146,51.6818],[160,51.2927],[175,51.2927],[178,51.1121],[185,51.0644],[205,54.9256],[217,54.9256],[236,63.0083],[269,63.054],[295,57.2757],[311,57.2757],[327,54.284],[340,53.8566]]
       }
       arm_l_pv_ik_ctrl.Scale.X {type=MovieSceneScriptingFloatChannel, default=1, const [1,25]}
       arm_l_pv_ik_ctrl.Scale.Y {type=MovieSceneScriptingFloatChannel, default=1, const [1,25]}
       arm_l_pv_ik_ctrl.Scale.Z {type=MovieSceneScriptingFloatChannel, default=1, const [1,25]}
       upperarm_r_fk_ctrl.Location.X {type=MovieSceneScriptingFloatChannel, default=0, thinned_from=28, keys [[0,0],[340,0]]}
       upperarm_r_fk_ctrl.Location.Y {type=MovieSceneScriptingFloatChannel, default=0, thinned_from=28, keys [[0,0],[340,0]]}
       upperarm_r_fk_ctrl.Location.Z {type=MovieSceneScriptingFloatChannel, default=0, thinned_from=28, keys [[0,0],[340,0]]}
       upperarm_r_fk_ctrl.Rotation.X {
        type=MovieSceneScriptingFloatChannel
        default=0
        thinned_from=28
        keys [[0,30.6556],[146,30.6556],[160,33.0286],[221,33.0286],[236,26.7687],[269,26.7687],[282,30.2211],[287,30.8963],[311,30.8963],[327,30.6857],[340,30.6556]]
       }
       upperarm_r_fk_ctrl.Rotation.Y {
        type=MovieSceneScriptingFloatChannel
        default=0
        thinned_from=28
        keys [[0,0.4821],[146,0.4821],[160,-3.8073],[221,-3.8073],[236,7.5125],[269,7.5125],[282,7.7534],[287,6.8145],[311,6.8145],[327,1.2736],[340,0.4821]]
       }
       upperarm_r_fk_ctrl.Rotation.Z {
        type=MovieSceneScriptingFloatChannel
        default=0
        thinned_from=28
        keys [[0,10.9777],[146,10.9777],[160,14.0098],[221,14.0098],[236,8.5371],[269,8.5371],[282,12.5356],[287,14.7534],[311,14.7534],[327,11.4497],[340,10.9777]]
       }
       upperarm_r_fk_ctrl.Scale.X {type=MovieSceneScriptingFloatChannel, default=1, const [1,28]}
       upperarm_r_fk_ctrl.Scale.Y {type=MovieSceneScriptingFloatChannel, default=1, const [1,28]}
       upperarm_r_fk_ctrl.Scale.Z {type=MovieSceneScriptingFloatChannel, default=1, const [1,28]}
       lowerarm_r_fk_ctrl.Location.X {type=MovieSceneScriptingFloatChannel, default=0, thinned_from=28, keys [[0,0],[340,0]]}
       lowerarm_r_fk_ctrl.Location.Y {type=MovieSceneScriptingFloatChannel, default=0, thinned_from=28, keys [[0,0],[340,0]]}
       lowerarm_r_fk_ctrl.Location.Z {type=MovieSceneScriptingFloatChannel, default=0, thinned_from=28, keys [[0,0],[340,0]]}
       lowerarm_r_fk_ctrl.Rotation.X {
        type=MovieSceneScriptingFloatChannel
        default=0
        thinned_from=28
        keys [[0,4.607],[146,4.607],[160,21.1029],[221,21.1029],[236,4.6367],[269,4.6367],[287,4.607],[340,4.607]]
       }
       lowerarm_r_fk_ctrl.Rotation.Y {
        type=MovieSceneScriptingFloatChannel
        default=0
        thinned_from=28
        keys [[0,-0.7045],[146,-0.7045],[160,0.4094],[221,0.4094],[236,-0.4702],[269,-0.4702],[282,-0.6752],[287,-0.7045],[340,-0.7045]]
       }
       lowerarm_r_fk_ctrl.Rotation.Z {
        type=MovieSceneScriptingFloatChannel
        default=0
        thinned_from=28
        keys [[0,94.0884],[146,94.0884],[160,88.9121],[221,88.9121],[236,91.1911],[269,91.1911],[282,93.7262],[287,94.0884],[340,94.0884]]
       }
       lowerarm_r_fk_ctrl.Scale.X {type=MovieSceneScriptingFloatChannel, default=1, const [1,28]}
       lowerarm_r_fk_ctrl.Scale.Y {type=MovieSceneScriptingFloatChannel, default=1, const [1,28]}
       lowerarm_r_fk_ctrl.Scale.Z {type=MovieSceneScriptingFloatChannel, default=1, const [1,28]}
       hand_r_fk_ctrl.Location.X {type=MovieSceneScriptingFloatChannel, default=0, thinned_from=26, keys [[0,0],[340,0]]}
       hand_r_fk_ctrl.Location.Y {type=MovieSceneScriptingFloatChannel, default=0, thinned_from=26, keys [[0,0],[340,0]]}
       hand_r_fk_ctrl.Location.Z {type=MovieSceneScriptingFloatChannel, default=0, thinned_from=26, keys [[0,0],[340,0]]}
       hand_r_fk_ctrl.Rotation.X {type=MovieSceneScriptingFloatChannel, default=0, thinned_from=26, keys [[0,-0.6776],[340,-0.6776]]}
       hand_r_fk_ctrl.Rotation.Y {type=MovieSceneScriptingFloatChannel, default=0, const [14.4524,26]}
       hand_r_fk_ctrl.Rotation.Z {type=MovieSceneScriptingFloatChannel, default=0, const [-36.9027,26]}
       hand_r_fk_ctrl.Scale.X {type=MovieSceneScriptingFloatChannel, default=1, const [1,26]}
       hand_r_fk_ctrl.Scale.Y {type=MovieSceneScriptingFloatChannel, default=1, const [1,26]}
       hand_r_fk_ctrl.Scale.Z {type=MovieSceneScriptingFloatChannel, default=1, const [1,26]}
       thumb_01_r_ctrl.Location.X {type=MovieSceneScriptingFloatChannel, default=0, thinned_from=24, keys [[0,0],[340,0]]}
       thumb_01_r_ctrl.Location.Y {type=MovieSceneScriptingFloatChannel, default=0, thinned_from=24, keys [[0,0],[340,0]]}
       thumb_01_r_ctrl.Location.Z {type=MovieSceneScriptingFloatChannel, default=0, thinned_from=24, keys [[0,0],[340,0]]}
       thumb_01_r_ctrl.Rotation.X {type=MovieSceneScriptingFloatChannel, default=0, const [-11.8305,24]}
       thumb_01_r_ctrl.Rotation.Y {type=MovieSceneScriptingFloatChannel, default=0, thinned_from=24, keys [[0,-6.959],[340,-6.959]]}
       thumb_01_r_ctrl.Rotation.Z {type=MovieSceneScriptingFloatChannel, default=0, thinned_from=24, keys [[0,4.4404],[340,4.4404]]}
       thumb_01_r_ctrl.Scale.X {type=MovieSceneScriptingFloatChannel, default=1, const [1,24]}
       thumb_01_r_ctrl.Scale.Y {type=MovieSceneScriptingFloatChannel, default=1, const [1,24]}
       thumb_01_r_ctrl.Scale.Z {type=MovieSceneScriptingFloatChannel, default=1, const [1,24]}
       thumb_02_r_ctrl.Location.X {type=MovieSceneScriptingFloatChannel, default=0, thinned_from=24, keys [[0,0],[340,0]]}
       thumb_02_r_ctrl.Location.Y {type=MovieSceneScriptingFloatChannel, default=0, thinned_from=24, keys [[0,0],[340,0]]}
       thumb_02_r_ctrl.Location.Z {type=MovieSceneScriptingFloatChannel, default=0, thinned_from=24, keys [[0,0],[340,0]]}
       thumb_02_r_ctrl.Rotation.X {type=MovieSceneScriptingFloatChannel, default=0, thinned_from=24, keys [[0,0],[340,0]]}
       thumb_02_r_ctrl.Rotation.Y {type=MovieSceneScriptingFloatChannel, default=0, thinned_from=24, keys [[0,0],[340,0]]}
       thumb_02_r_ctrl.Rotation.Z {type=MovieSceneScriptingFloatChannel, default=0, const [-12.3705,24]}
       thumb_02_r_ctrl.Scale.X {type=MovieSceneScriptingFloatChannel, default=1, const [1,24]}
       thumb_02_r_ctrl.Scale.Y {type=MovieSceneScriptingFloatChannel, default=1, const [1,24]}
       thumb_02_r_ctrl.Scale.Z {type=MovieSceneScriptingFloatChannel, default=1, const [1,24]}
       thumb_03_r_ctrl.Location.X {type=MovieSceneScriptingFloatChannel, default=0, thinned_from=24, keys [[0,0.5596],[340,0.5596]]}
       thumb_03_r_ctrl.Location.Y {type=MovieSceneScriptingFloatChannel, default=0, thinned_from=24, keys [[0,0.2442],[340,0.2442]]}
       thumb_03_r_ctrl.Location.Z {type=MovieSceneScriptingFloatChannel, default=0, thinned_from=24, keys [[0,0],[340,0]]}
       thumb_03_r_ctrl.Rotation.X {type=MovieSceneScriptingFloatChannel, default=0, thinned_from=24, keys [[0,0],[340,0]]}
       thumb_03_r_ctrl.Rotation.Y {type=MovieSceneScriptingFloatChannel, default=0, thinned_from=24, keys [[0,0],[340,0]]}
       thumb_03_r_ctrl.Rotation.Z {type=MovieSceneScriptingFloatChannel, default=0, const [-66.4234,24]}
       thumb_03_r_ctrl.Scale.X {type=MovieSceneScriptingFloatChannel, default=1, const [1,24]}
       thumb_03_r_ctrl.Scale.Y {type=MovieSceneScriptingFloatChannel, default=1, const [1,24]}
       thumb_03_r_ctrl.Scale.Z {type=MovieSceneScriptingFloatChannel, default=1, const [1,24]}
       index_metacarpal_r_ctrl.Location.X {type=MovieSceneScriptingFloatChannel, default=0, thinned_from=24, keys [[0,0],[340,0]]}
       index_metacarpal_r_ctrl.Location.Y {type=MovieSceneScriptingFloatChannel, default=0, thinned_from=24, keys [[0,0],[340,0]]}
       index_metacarpal_r_ctrl.Location.Z {type=MovieSceneScriptingFloatChannel, default=0, thinned_from=24, keys [[0,0],[340,0]]}
       index_metacarpal_r_ctrl.Rotation.X {type=MovieSceneScriptingFloatChannel, default=0, const [-2.7256,24]}
       index_metacarpal_r_ctrl.Rotation.Y {type=MovieSceneScriptingFloatChannel, default=0, thinned_from=24, keys [[0,0],[340,0]]}
       index_metacarpal_r_ctrl.Rotation.Z {type=MovieSceneScriptingFloatChannel, default=0, thinned_from=24, keys [[0,-7.3778],[340,-7.3778]]}
       index_metacarpal_r_ctrl.Scale.X {type=MovieSceneScriptingFloatChannel, default=1, const [1,24]}
       index_metacarpal_r_ctrl.Scale.Y {type=MovieSceneScriptingFloatChannel, default=1, const [1,24]}
       index_metacarpal_r_ctrl.Scale.Z {type=MovieSceneScriptingFloatChannel, default=1, const [1,24]}
       index_01_r_ctrl.Location.X {type=MovieSceneScriptingFloatChannel, default=0, thinned_from=24, keys [[0,0],[340,0]]}
       index_01_r_ctrl.Location.Y {type=MovieSceneScriptingFloatChannel, default=0, thinned_from=24, keys [[0,0],[340,0]]}
       index_01_r_ctrl.Location.Z {type=MovieSceneScriptingFloatChannel, default=0, thinned_from=24, keys [[0,0],[340,0]]}
       index_01_r_ctrl.Rotation.X {type=MovieSceneScriptingFloatChannel, default=0, thinned_from=24, keys [[0,0],[340,0]]}
       index_01_r_ctrl.Rotation.Y {type=MovieSceneScriptingFloatChannel, default=0, thinned_from=24, keys [[0,0],[340,0]]}
       index_01_r_ctrl.Rotation.Z {type=MovieSceneScriptingFloatChannel, default=0, const [-64.9002,24]}
       index_01_r_ctrl.Scale.X {type=MovieSceneScriptingFloatChannel, default=1, const [1,24]}
       index_01_r_ctrl.Scale.Y {type=MovieSceneScriptingFloatChannel, default=1, const [1,24]}
       index_01_r_ctrl.Scale.Z {type=MovieSceneScriptingFloatChannel, default=1, const [1,24]}
       index_02_r_ctrl.Location.X {type=MovieSceneScriptingFloatChannel, default=0, const [0,24]}
       index_02_r_ctrl.Location.Y {type=MovieSceneScriptingFloatChannel, default=0, const [0,24]}
       index_02_r_ctrl.Location.Z {type=MovieSceneScriptingFloatChannel, default=0, const [0,24]}
       index_02_r_ctrl.Rotation.X {type=MovieSceneScriptingFloatChannel, default=0, thinned_from=24, keys [[0,0],[340,0]]}
       index_02_r_ctrl.Rotation.Y {type=MovieSceneScriptingFloatChannel, default=0, thinned_from=24, keys [[0,0],[340,0]]}
       index_02_r_ctrl.Rotation.Z {type=MovieSceneScriptingFloatChannel, default=0, const [-49.5772,24]}
       index_02_r_ctrl.Scale.X {type=MovieSceneScriptingFloatChannel, default=1, const [1,24]}
       index_02_r_ctrl.Scale.Y {type=MovieSceneScriptingFloatChannel, default=1, const [1,24]}
       index_02_r_ctrl.Scale.Z {type=MovieSceneScriptingFloatChannel, default=1, const [1,24]}
       index_03_r_ctrl.Location.X {type=MovieSceneScriptingFloatChannel, default=0, const [0,24]}
       index_03_r_ctrl.Location.Y {type=MovieSceneScriptingFloatChannel, default=0, const [0,24]}
       index_03_r_ctrl.Location.Z {type=MovieSceneScriptingFloatChannel, default=0, const [0,24]}
       index_03_r_ctrl.Rotation.X {type=MovieSceneScriptingFloatChannel, default=0, thinned_from=24, keys [[0,0],[340,0]]}
       index_03_r_ctrl.Rotation.Y {type=MovieSceneScriptingFloatChannel, default=0, thinned_from=24, keys [[0,0],[340,0]]}
       index_03_r_ctrl.Rotation.Z {type=MovieSceneScriptingFloatChannel, default=0, const [-102.5714,24]}
       index_03_r_ctrl.Scale.X {type=MovieSceneScriptingFloatChannel, default=1, const [1,24]}
       index_03_r_ctrl.Scale.Y {type=MovieSceneScriptingFloatChannel, default=1, const [1,24]}
       index_03_r_ctrl.Scale.Z {type=MovieSceneScriptingFloatChannel, default=1, const [1,24]}
       middle_metacarpal_r_ctrl.Location.X {type=MovieSceneScriptingFloatChannel, default=0, const [0,24]}
       middle_metacarpal_r_ctrl.Location.Y {type=MovieSceneScriptingFloatChannel, default=0, const [0,24]}
       middle_metacarpal_r_ctrl.Location.Z {type=MovieSceneScriptingFloatChannel, default=0, const [0,24]}
       middle_metacarpal_r_ctrl.Rotation.X {type=MovieSceneScriptingFloatChannel, default=0, const [-2.7256,24]}
       middle_metacarpal_r_ctrl.Rotation.Y {type=MovieSceneScriptingFloatChannel, default=0, thinned_from=24, keys [[0,0],[340,0]]}
       middle_metacarpal_r_ctrl.Rotation.Z {type=MovieSceneScriptingFloatChannel, default=0, thinned_from=24, keys [[0,-7.3778],[340,-7.3778]]}
       middle_metacarpal_r_ctrl.Scale.X {type=MovieSceneScriptingFloatChannel, default=1, const [1,24]}
       middle_metacarpal_r_ctrl.Scale.Y {type=MovieSceneScriptingFloatChannel, default=1, const [1,24]}
       middle_metacarpal_r_ctrl.Scale.Z {type=MovieSceneScriptingFloatChannel, default=1, const [1,24]}
       middle_01_r_ctrl.Location.X {type=MovieSceneScriptingFloatChannel, default=0, const [0,24]}
       middle_01_r_ctrl.Location.Y {type=MovieSceneScriptingFloatChannel, default=0, const [0,24]}
       middle_01_r_ctrl.Location.Z {type=MovieSceneScriptingFloatChannel, default=0, const [0,24]}
       middle_01_r_ctrl.Rotation.X {type=MovieSceneScriptingFloatChannel, default=0, thinned_from=24, keys [[0,0],[340,0]]}
       middle_01_r_ctrl.Rotation.Y {type=MovieSceneScriptingFloatChannel, default=0, thinned_from=24, keys [[0,0],[340,0]]}
       middle_01_r_ctrl.Rotation.Z {type=MovieSceneScriptingFloatChannel, default=0, const [-64.9002,24]}
       middle_01_r_ctrl.Scale.X {type=MovieSceneScriptingFloatChannel, default=1, const [1,24]}
       middle_01_r_ctrl.Scale.Y {type=MovieSceneScriptingFloatChannel, default=1, const [1,24]}
       middle_01_r_ctrl.Scale.Z {type=MovieSceneScriptingFloatChannel, default=1, const [1,24]}
       middle_02_r_ctrl.Location.X {type=MovieSceneScriptingFloatChannel, default=0, const [0,24]}
       middle_02_r_ctrl.Location.Y {type=MovieSceneScriptingFloatChannel, default=0, const [0,24]}
       middle_02_r_ctrl.Location.Z {type=MovieSceneScriptingFloatChannel, default=0, const [0,24]}
       middle_02_r_ctrl.Rotation.X {type=MovieSceneScriptingFloatChannel, default=0, thinned_from=24, keys [[0,0],[340,0]]}
       middle_02_r_ctrl.Rotation.Y {type=MovieSceneScriptingFloatChannel, default=0, const [0,24]}
       middle_02_r_ctrl.Rotation.Z {type=MovieSceneScriptingFloatChannel, default=0, const [-49.5772,24]}
       middle_02_r_ctrl.Scale.X {type=MovieSceneScriptingFloatChannel, default=1, const [1,24]}
       middle_02_r_ctrl.Scale.Y {type=MovieSceneScriptingFloatChannel, default=1, const [1,24]}
       middle_02_r_ctrl.Scale.Z {type=MovieSceneScriptingFloatChannel, default=1, const [1,24]}
       middle_03_r_ctrl.Location.X {type=MovieSceneScriptingFloatChannel, default=0, const [0,24]}
       middle_03_r_ctrl.Location.Y {type=MovieSceneScriptingFloatChannel, default=0, const [0,24]}
       middle_03_r_ctrl.Location.Z {type=MovieSceneScriptingFloatChannel, default=0, const [0,24]}
       middle_03_r_ctrl.Rotation.X {type=MovieSceneScriptingFloatChannel, default=0, thinned_from=24, keys [[0,0],[340,0]]}
       middle_03_r_ctrl.Rotation.Y {type=MovieSceneScriptingFloatChannel, default=0, thinned_from=24, keys [[0,0],[340,0]]}
       middle_03_r_ctrl.Rotation.Z {type=MovieSceneScriptingFloatChannel, default=0, const [-102.5714,24]}
       middle_03_r_ctrl.Scale.X {type=MovieSceneScriptingFloatChannel, default=1, const [1,24]}
       middle_03_r_ctrl.Scale.Y {type=MovieSceneScriptingFloatChannel, default=1, const [1,24]}
       middle_03_r_ctrl.Scale.Z {type=MovieSceneScriptingFloatChannel, default=1, const [1,24]}
       ring_metacarpal_r_ctrl.Location.X {type=MovieSceneScriptingFloatChannel, default=0, const [0,24]}
       ring_metacarpal_r_ctrl.Location.Y {type=MovieSceneScriptingFloatChannel, default=0, const [0,24]}
       ring_metacarpal_r_ctrl.Location.Z {type=MovieSceneScriptingFloatChannel, default=0, const [0,24]}
       ring_metacarpal_r_ctrl.Rotation.X {type=MovieSceneScriptingFloatChannel, default=0, thinned_from=24, keys [[0,1.2867],[340,1.2867]]}
       ring_metacarpal_r_ctrl.Rotation.Y {type=MovieSceneScriptingFloatChannel, default=0, const [0,24]}
       ring_metacarpal_r_ctrl.Rotation.Z {type=MovieSceneScriptingFloatChannel, default=0, thinned_from=24, keys [[0,-7.3778],[340,-7.3778]]}
       ring_metacarpal_r_ctrl.Scale.X {type=MovieSceneScriptingFloatChannel, default=1, const [1,24]}
       ring_metacarpal_r_ctrl.Scale.Y {type=MovieSceneScriptingFloatChannel, default=1, const [1,24]}
       ring_metacarpal_r_ctrl.Scale.Z {type=MovieSceneScriptingFloatChannel, default=1, const [1,24]}
       ring_01_r_ctrl.Location.X {type=MovieSceneScriptingFloatChannel, default=0, const [0,24]}
       ring_01_r_ctrl.Location.Y {type=MovieSceneScriptingFloatChannel, default=0, const [0,24]}
       ring_01_r_ctrl.Location.Z {type=MovieSceneScriptingFloatChannel, default=0, const [0,24]}
       ring_01_r_ctrl.Rotation.X {type=MovieSceneScriptingFloatChannel, default=0, thinned_from=24, keys [[0,0],[340,0]]}
       ring_01_r_ctrl.Rotation.Y {type=MovieSceneScriptingFloatChannel, default=0, thinned_from=24, keys [[0,0],[340,0]]}
       ring_01_r_ctrl.Rotation.Z {type=MovieSceneScriptingFloatChannel, default=0, const [-64.9002,24]}
       ring_01_r_ctrl.Scale.X {type=MovieSceneScriptingFloatChannel, default=1, const [1,24]}
       ring_01_r_ctrl.Scale.Y {type=MovieSceneScriptingFloatChannel, default=1, const [1,24]}
       ring_01_r_ctrl.Scale.Z {type=MovieSceneScriptingFloatChannel, default=1, const [1,24]}
       ring_02_r_ctrl.Location.X {type=MovieSceneScriptingFloatChannel, default=0, const [0,24]}
       ring_02_r_ctrl.Location.Y {type=MovieSceneScriptingFloatChannel, default=0, const [0,24]}
       ring_02_r_ctrl.Location.Z {type=MovieSceneScriptingFloatChannel, default=0, const [0,24]}
       ring_02_r_ctrl.Rotation.X {type=MovieSceneScriptingFloatChannel, default=0, thinned_from=24, keys [[0,0],[340,0]]}
       ring_02_r_ctrl.Rotation.Y {type=MovieSceneScriptingFloatChannel, default=0, thinned_from=24, keys [[0,0],[340,0]]}
       ring_02_r_ctrl.Rotation.Z {type=MovieSceneScriptingFloatChannel, default=0, const [-49.5772,24]}
       ring_02_r_ctrl.Scale.X {type=MovieSceneScriptingFloatChannel, default=1, const [1,24]}
       ring_02_r_ctrl.Scale.Y {type=MovieSceneScriptingFloatChannel, default=1, const [1,24]}
       ring_02_r_ctrl.Scale.Z {type=MovieSceneScriptingFloatChannel, default=1, const [1,24]}
       ring_03_r_ctrl.Location.X {type=MovieSceneScriptingFloatChannel, default=0, const [0,24]}
       ring_03_r_ctrl.Location.Y {type=MovieSceneScriptingFloatChannel, default=0, const [0,24]}
       ring_03_r_ctrl.Location.Z {type=MovieSceneScriptingFloatChannel, default=0, const [0,24]}
       ring_03_r_ctrl.Rotation.X {type=MovieSceneScriptingFloatChannel, default=0, thinned_from=24, keys [[0,0],[340,0]]}
       ring_03_r_ctrl.Rotation.Y {type=MovieSceneScriptingFloatChannel, default=0, const [0,24]}
       ring_03_r_ctrl.Rotation.Z {type=MovieSceneScriptingFloatChannel, default=0, const [-102.5714,24]}
       ring_03_r_ctrl.Scale.X {type=MovieSceneScriptingFloatChannel, default=1, const [1,24]}
       ring_03_r_ctrl.Scale.Y {type=MovieSceneScriptingFloatChannel, default=1, const [1,24]}
       ring_03_r_ctrl.Scale.Z {type=MovieSceneScriptingFloatChannel, default=1, const [1,24]}
       pinky_metacarpal_r_ctrl.Location.X {type=MovieSceneScriptingFloatChannel, default=0, const [0,24]}
       pinky_metacarpal_r_ctrl.Location.Y {type=MovieSceneScriptingFloatChannel, default=0, const [0,24]}
       pinky_metacarpal_r_ctrl.Location.Z {type=MovieSceneScriptingFloatChannel, default=0, const [0,24]}
       pinky_metacarpal_r_ctrl.Rotation.X {type=MovieSceneScriptingFloatChannel, default=0, thinned_from=24, keys [[0,1.2867],[340,1.2867]]}
       pinky_metacarpal_r_ctrl.Rotation.Y {type=MovieSceneScriptingFloatChannel, default=0, thinned_from=24, keys [[0,0],[340,0]]}
       pinky_metacarpal_r_ctrl.Rotation.Z {type=MovieSceneScriptingFloatChannel, default=0, thinned_from=24, keys [[0,-7.3778],[340,-7.3778]]}
       pinky_metacarpal_r_ctrl.Scale.X {type=MovieSceneScriptingFloatChannel, default=1, const [1,24]}
       pinky_metacarpal_r_ctrl.Scale.Y {type=MovieSceneScriptingFloatChannel, default=1, const [1,24]}
       pinky_metacarpal_r_ctrl.Scale.Z {type=MovieSceneScriptingFloatChannel, default=1, const [1,24]}
       pinky_01_r_ctrl.Location.X {type=MovieSceneScriptingFloatChannel, default=0, const [0,24]}
       pinky_01_r_ctrl.Location.Y {type=MovieSceneScriptingFloatChannel, default=0, const [0,24]}
       pinky_01_r_ctrl.Location.Z {type=MovieSceneScriptingFloatChannel, default=0, const [0,24]}
       pinky_01_r_ctrl.Rotation.X {type=MovieSceneScriptingFloatChannel, default=0, thinned_from=24, keys [[0,0],[340,0]]}
       pinky_01_r_ctrl.Rotation.Y {type=MovieSceneScriptingFloatChannel, default=0, thinned_from=24, keys [[0,0],[340,0]]}
       pinky_01_r_ctrl.Rotation.Z {type=MovieSceneScriptingFloatChannel, default=0, const [-64.9002,24]}
       pinky_01_r_ctrl.Scale.X {type=MovieSceneScriptingFloatChannel, default=1, const [1,24]}
       pinky_01_r_ctrl.Scale.Y {type=MovieSceneScriptingFloatChannel, default=1, const [1,24]}
       pinky_01_r_ctrl.Scale.Z {type=MovieSceneScriptingFloatChannel, default=1, const [1,24]}
       pinky_02_r_ctrl.Location.X {type=MovieSceneScriptingFloatChannel, default=0, const [0,24]}
       pinky_02_r_ctrl.Location.Y {type=MovieSceneScriptingFloatChannel, default=0, const [0,24]}
       pinky_02_r_ctrl.Location.Z {type=MovieSceneScriptingFloatChannel, default=0, const [0,24]}
       pinky_02_r_ctrl.Rotation.X {type=MovieSceneScriptingFloatChannel, default=0, const [0,24]}
       pinky_02_r_ctrl.Rotation.Y {type=MovieSceneScriptingFloatChannel, default=0, const [0,24]}
       pinky_02_r_ctrl.Rotation.Z {type=MovieSceneScriptingFloatChannel, default=0, const [-49.5772,24]}
       pinky_02_r_ctrl.Scale.X {type=MovieSceneScriptingFloatChannel, default=1, const [1,24]}
       pinky_02_r_ctrl.Scale.Y {type=MovieSceneScriptingFloatChannel, default=1, const [1,24]}
       pinky_02_r_ctrl.Scale.Z {type=MovieSceneScriptingFloatChannel, default=1, const [1,24]}
       pinky_03_r_ctrl.Location.X {type=MovieSceneScriptingFloatChannel, default=0, const [0,24]}
       pinky_03_r_ctrl.Location.Y {type=MovieSceneScriptingFloatChannel, default=0, const [0,24]}
       pinky_03_r_ctrl.Location.Z {type=MovieSceneScriptingFloatChannel, default=0, const [0,24]}
       pinky_03_r_ctrl.Rotation.X {type=MovieSceneScriptingFloatChannel, default=0, const [0,24]}
       pinky_03_r_ctrl.Rotation.Y {type=MovieSceneScriptingFloatChannel, default=0, const [0,24]}
       pinky_03_r_ctrl.Rotation.Z {type=MovieSceneScriptingFloatChannel, default=0, const [-102.5714,24]}
       pinky_03_r_ctrl.Scale.X {type=MovieSceneScriptingFloatChannel, default=1, const [1,24]}
       pinky_03_r_ctrl.Scale.Y {type=MovieSceneScriptingFloatChannel, default=1, const [1,24]}
       pinky_03_r_ctrl.Scale.Z {type=MovieSceneScriptingFloatChannel, default=1, const [1,24]}
       hand_r_ik_ctrl.Location.X {
        type=MovieSceneScriptingFloatChannel
        default=0
        thinned_from=26
        keys [[0,-8.8686],[25,-8.6508],[60,-8.5367],[84,-8.5367],[92,-8.2795],[112,-8.2795],[122,-8.1227],[135,-8.1227],[139,-8.0642],[142,-10.2634],[146,-7.9976],[160,-6.1406],[175,-6.1406],[178,-6.346],[185,-6.4437],[205,-8.6872],[217,-8.6872],[236,-12.3773],[255,-12.5081],[269,-12.5081],[287,-10.5934],[295,-10.0204],[311,-10.0204],[327,-10.5217],[340,-10.5934]]
       }
       hand_r_ik_ctrl.Location.Y {
        type=MovieSceneScriptingFloatChannel
        default=0
        thinned_from=26
        keys [[0,10.9949],[25,10.6708],[60,10.5015],[84,10.5015],[92,10.3419],[112,10.3419],[122,10.4465],[135,10.4465],[139,10.3951],[142,10.6953],[146,10.473],[160,12.3595],[175,12.3595],[205,13.7674],[217,13.7674],[236,11.2447],[255,11.3298],[269,11.3298],[287,11.011],[295,10.5567],[311,10.5567],[327,10.9542],[340,11.011]]
       }
       hand_r_ik_ctrl.Location.Z {
        type=MovieSceneScriptingFloatChannel
        default=0
        thinned_from=26
        keys [[0,15.0624],[25,14.7163],[60,14.5351],[84,14.5351],[92,14.3253],[112,14.3253],[122,14.4106],[135,14.4106],[139,14.458],[142,15.076],[146,14.5608],[160,18.2181],[175,18.2181],[178,18.4269],[185,18.4182],[205,19.0568],[217,19.0568],[236,14.1932],[269,14.1799],[287,15.3978],[295,15.5167],[311,15.5167],[327,15.4126],[340,15.3978]]
       }
       hand_r_ik_ctrl.Rotation.X {
        type=MovieSceneScriptingFloatChannel
        default=0
        thinned_from=26
        keys [[0,-105.7069],[84,-105.9981],[92,-104.5015],[112,-104.5015],[122,-105.5141],[135,-105.5141],[139,-105.8204],[142,-105.5194],[146,-106.5026],[160,-111.5045],[175,-111.5045],[185,-111.8747],[205,-105.9473],[217,-105.9473],[236,-80.0779],[269,-80.0125],[287,-107.0648],[295,-108.0639],[311,-108.0639],[327,-107.1896],[340,-107.0648]]
       }
       hand_r_ik_ctrl.Rotation.Y {
        type=MovieSceneScriptingFloatChannel
        default=0
        thinned_from=26
        keys [[0,60.501],[25,59.8426],[60,59.4977],[135,59.3741],[139,59.4865],[142,65.4344],[146,59.5823],[160,60.4607],[175,60.4607],[178,60.9771],[185,61.1919],[205,62.7253],[217,62.7253],[236,62.4],[269,62.3788],[287,65.5639],[295,65.2556],[311,65.2556],[327,65.5254],[340,65.5639]]
       }
       hand_r_ik_ctrl.Rotation.Z {
        type=MovieSceneScriptingFloatChannel
        default=0
        thinned_from=26
        keys [[0,-102.8446],[25,-102.4594],[84,-102.2588],[92,-100.7935],[112,-100.7935],[135,-100.9708],[139,-101.1284],[142,-112.508],[146,-101.369],[160,-118.0045],[175,-118.0045],[178,-118.6153],[205,-120.4441],[217,-120.4441],[236,-93.9352],[269,-93.9089],[287,-113.8281],[311,-113.9548],[340,-113.8281]]
       }
       hand_r_ik_ctrl.Scale.X {type=MovieSceneScriptingFloatChannel, default=1, const [1,26]}
       hand_r_ik_ctrl.Scale.Y {type=MovieSceneScriptingFloatChannel, default=1, const [1,26]}
       hand_r_ik_ctrl.Scale.Z {type=MovieSceneScriptingFloatChannel, default=1, const [1,26]}
       Softness_2 {type=MovieSceneScriptingFloatChannel, default=0, const [0,26]}
       hand_r_ikorient_ctrl.Rotation.Roll {type=MovieSceneScriptingFloatChannel, default=0, const [0,24]}
       hand_r_ikorient_ctrl.Rotation.Pitch {type=MovieSceneScriptingFloatChannel, default=0, const [0,24]}
       hand_r_ikorient_ctrl.Rotation.Yaw {type=MovieSceneScriptingFloatChannel, default=0, const [0,24]}
       arm_r_pv_ik_ctrl.Location.X {
        type=MovieSceneScriptingFloatChannel
        default=0
        keys [[0,1.8515],[25,2.0334],[46,2.094],[60,2.1274],[84,2.1274],[92,2.0377],[112,2.0377],[122,2.2918],[135,2.2918],[139,2.5084],[142,2.3432],[146,2.7144],[160,2.6925],[175,2.6925],[178,2.6821],[185,2.4549],[205,-1.4231],[217,-1.4231],[236,-6.4697],[255,-6.6325],[269,-6.6325],[287,2.4087],[295,3.2839],[311,3.2839],[327,2.5181],[340,2.4087]]
       }
       arm_r_pv_ik_ctrl.Location.Y {
        type=MovieSceneScriptingFloatChannel
        default=0
        keys [[0,-2.3038],[25,-2.4976],[46,-2.5621],[60,-2.5984],[84,-2.5984],[92,-2.8855],[112,-2.8855],[122,-2.5629],[135,-2.5629],[139,-2.4393],[142,-3.3639],[146,-2.1937],[160,2.8312],[175,2.8312],[178,2.9137],[185,2.8605],[205,1.1682],[217,1.1682],[236,-7.2335],[255,-7.3076],[269,-7.3076],[287,-2.9671],[295,-2.4557],[311,-2.4557],[327,-2.9032],[340,-2.9671]]
       }
       arm_r_pv_ik_ctrl.Location.Z {
        type=MovieSceneScriptingFloatChannel
        default=0
        thinned_from=26
        keys [[0,-11.2601],[25,-11.8679],[60,-12.1861],[84,-12.1861],[92,-12.0018],[112,-12.0018],[122,-12.5121],[135,-12.5121],[139,-12.5504],[142,-6.0452],[146,-12.766],[160,-15.2509],[175,-15.2509],[178,-14.7114],[185,-14.6398],[205,-10.4701],[217,-10.4701],[236,-1.3411],[255,-1.2729],[269,-1.2729],[287,-6.214],[295,-6.8635],[311,-6.8635],[327,-6.2952],[340,-6.214]]
       }
       arm_r_pv_ik_ctrl.Rotation.X {
        type=MovieSceneScriptingFloatChannel
        default=0
        thinned_from=26
        keys [[0,13.7492],[25,13.3073],[60,13.0757],[84,13.0757],[92,13.6508],[112,13.6508],[122,13.2854],[135,13.2854],[139,13.2467],[142,16.439],[146,13.075],[160,9.2882],[175,9.2882],[178,9.6833],[185,9.8201],[205,12.2486],[217,12.2486],[236,19.5755],[269,19.5834],[287,15.8756],[295,15.4662],[311,15.4662],[327,15.8244],[340,15.8756]]
       }
       arm_r_pv_ik_ctrl.Rotation.Y {
        type=MovieSceneScriptingFloatChannel
        default=0
        keys [[0,-7.8978],[25,-8.6223],[46,-8.8638],[60,-9.0017],[84,-9.0017],[92,-9.3082],[112,-9.3082],[122,-9.8863],[135,-9.8863],[139,-9.9168],[142,2.6089],[146,-10.1415],[160,-13.7216],[175,-13.7216],[178,-13.1714],[185,-12.7329],[205,-7.2225],[217,-7.2225],[236,3.8839],[255,3.9028],[269,3.9028],[287,2.913],[295,2.1984],[311,2.1984],[327,2.8236],[340,2.913]]
       }
       arm_r_pv_ik_ctrl.Rotation.Z {
        type=MovieSceneScriptingFloatChannel
        default=0
        thinned_from=26
        keys [[0,5.8001],[25,5.6938],[60,5.6389],[84,5.6389],[92,5.1048],[112,5.1048],[122,5.7024],[135,5.7024],[139,5.923],[142,4.6402],[146,6.3658],[160,11.4845],[175,11.4845],[178,11.5413],[185,11.5276],[205,7.0563],[217,7.0563],[236,-4.0334],[255,-4.0771],[269,-4.0771],[287,5.4469],[295,5.8772],[311,5.8772],[327,5.5007],[340,5.4469]]
       }
       arm_r_pv_ik_ctrl.Scale.X {type=MovieSceneScriptingFloatChannel, default=1, const [1,26]}
       arm_r_pv_ik_ctrl.Scale.Y {type=MovieSceneScriptingFloatChannel, default=1, const [1,26]}
       arm_r_pv_ik_ctrl.Scale.Z {type=MovieSceneScriptingFloatChannel, default=1, const [1,26]}
       body_ctrl.Location.X {
        type=MovieSceneScriptingFloatChannel
        default=0
        thinned_from=32
        keys [[0,0.2661],[25,-0.0516],[46,-0.1574],[60,-0.2179],[84,-0.2179],[92,-0.3874],[135,-0.3874],[139,-0.4898],[146,-0.4898],[160,-0.2585],[175,-0.2585],[178,-0.077],[185,0.0952],[195,1.7137],[205,2.5521],[217,2.5521],[221,2.6041],[226,2.9684],[236,3.0999],[269,3.1072],[274,3.0125],[282,2.3498],[287,2.2416],[295,1.8603],[311,1.8603],[327,0.4653],[340,0.2661]]
       }
       body_ctrl.Location.Y {
        type=MovieSceneScriptingFloatChannel
        default=0
        thinned_from=32
        keys [[0,2.1288],[25,2.0117],[46,1.9727],[60,1.9505],[178,1.9505],[185,1.9822],[195,2.2508],[205,2.6204],[287,2.6204],[295,2.5255],[311,2.5255],[327,2.1784],[340,2.1288]]
       }
       body_ctrl.Location.Z {
        type=MovieSceneScriptingFloatChannel
        default=0
        thinned_from=32
        keys [[0,-12.8916],[185,-12.8516],[195,-12.8231],[311,-12.8363],[327,-12.8847],[340,-12.8916]]
       }
       body_ctrl.Rotation.X {
        type=MovieSceneScriptingFloatChannel
        default=0
        thinned_from=32
        keys [[0,11.4255],[84,11.4431],[92,12.0891],[135,12.0891],[146,12.0484],[160,12.5564],[178,12.544],[185,12.3943],[195,11.1265],[217,11.1265],[221,11.1131],[226,11.0195],[236,10.9857],[269,10.9839],[274,10.9995],[282,11.1086],[287,11.1265],[295,11.1842],[311,11.1842],[327,11.3953],[340,11.4255]]
       }
       body_ctrl.Rotation.Y {
        type=MovieSceneScriptingFloatChannel
        default=0
        thinned_from=32
        keys [[0,3.2221],[25,4.0682],[46,4.3502],[60,4.5114],[84,4.5114],[92,4.4713],[112,4.4713],[122,5.1525],[135,5.1525],[139,5.1985],[142,5.3018],[146,5.4787],[160,5.9425],[175,5.9425],[178,5.2715],[185,4.8381],[195,0.3108],[205,-1.0335],[217,-1.0335],[221,-1.1801],[226,-2.2064],[236,-2.577],[255,-2.5974],[269,-2.5974],[274,-2.4264],[282,-1.229],[287,-1.0335],[295,-0.2123],[311,-0.2123],[327,2.7928],[340,3.2221]]
       }
       body_ctrl.Rotation.Z {
        type=MovieSceneScriptingFloatChannel
        default=0
        thinned_from=32
        keys [[0,26.0415],[25,26.213],[60,26.3028],[84,26.3028],[92,26.7961],[112,26.7961],[122,26.2593],[135,26.2593],[142,25.8908],[146,25.6287],[160,24.5416],[175,24.5416],[178,24.3917],[185,24.3616],[195,24.1065],[205,28.3693],[217,28.3693],[221,28.6871],[226,30.9113],[236,31.7144],[269,31.7587],[274,31.388],[282,28.793],[287,28.3693],[295,27.9201],[311,27.9201],[327,26.2763],[340,26.0415]]
       }
       body_ctrl.Scale.X {type=MovieSceneScriptingFloatChannel, default=1, const [1,32]}
       body_ctrl.Scale.Y {type=MovieSceneScriptingFloatChannel, default=1, const [1,32]}
       body_ctrl.Scale.Z {type=MovieSceneScriptingFloatChannel, default=1, const [1,32]}
       hips_ctrl.Location.X {type=MovieSceneScriptingFloatChannel, default=0, const [0,30]}
       hips_ctrl.Location.Y {type=MovieSceneScriptingFloatChannel, default=0, const [0,30]}
       hips_ctrl.Location.Z {type=MovieSceneScriptingFloatChannel, default=0, const [0,30]}
       hips_ctrl.Rotation.X {type=MovieSceneScriptingFloatChannel, default=0, thinned_from=30, keys [[0,3.792],[340,3.792]]}
       hips_ctrl.Rotation.Y {type=MovieSceneScriptingFloatChannel, default=0, thinned_from=30, keys [[0,0],[340,0]]}
       hips_ctrl.Rotation.Z {type=MovieSceneScriptingFloatChannel, default=0, thinned_from=30, keys [[0,0],[340,0]]}
       hips_ctrl.Scale.X {type=MovieSceneScriptingFloatChannel, default=1, const [1,30]}
       hips_ctrl.Scale.Y {type=MovieSceneScriptingFloatChannel, default=1, const [1,30]}
       hips_ctrl.Scale.Z {type=MovieSceneScriptingFloatChannel, default=1, const [1,30]}
       hips_tan_ctrl.Location.X {
        type=MovieSceneScriptingFloatChannel
        default=0
        thinned_from=26
        keys [[0,-0.1598],[146,-0.1598],[160,-0.0708],[217,-0.0708],[221,0],[236,0.1116],[255,0.1174],[269,0.1174],[274,0],[295,-0.1613],[311,-0.1613],[327,-0.0202],[340,0]]
       }
       hips_tan_ctrl.Location.Y {
        type=MovieSceneScriptingFloatChannel
        default=0.0032
        thinned_from=26
        keys [[0,-0.163],[146,-0.163],[160,-0.1471],[217,-0.1471],[221,-0.1634],[255,-0.1614],[269,-0.1614],[274,-0.1634],[295,-0.1629],[311,-0.1629],[327,-0.1633],[340,-0.1634]]
       }
       hips_tan_ctrl.Location.Z {
        type=MovieSceneScriptingFloatChannel
        default=-0.0015
        thinned_from=26
        keys [[0,-0.0103],[146,-0.0103],[160,-0.0052],[217,-0.0052],[221,-0.0045],[236,-0.0073],[255,-0.0076],[269,-0.0076],[274,-0.0045],[295,-0.0104],[311,-0.0104],[327,-0.0053],[340,-0.0045]]
       }
       hips_tan_ctrl.Rotation.X {type=MovieSceneScriptingFloatChannel, default=0, const [0,26]}
       hips_tan_ctrl.Rotation.Y {type=MovieSceneScriptingFloatChannel, default=0, const [0,26]}
       hips_tan_ctrl.Rotation.Z {type=MovieSceneScriptingFloatChannel, default=0, const [0,26]}
       hips_tan_ctrl.Scale.X {type=MovieSceneScriptingFloatChannel, default=1, const [1,26]}
       hips_tan_ctrl.Scale.Y {type=MovieSceneScriptingFloatChannel, default=1, const [1,26]}
       hips_tan_ctrl.Scale.Z {type=MovieSceneScriptingFloatChannel, default=1, const [1,26]}
       spine_01_ctrl.Location.X {type=MovieSceneScriptingFloatChannel, default=0, const [0,31]}
       spine_01_ctrl.Location.Y {type=MovieSceneScriptingFloatChannel, default=0, const [0,31]}
       spine_01_ctrl.Location.Z {type=MovieSceneScriptingFloatChannel, default=0, const [0,31]}
       spine_01_ctrl.Rotation.X {type=MovieSceneScriptingFloatChannel, default=0, const [0,31]}
       spine_01_ctrl.Rotation.Y {type=MovieSceneScriptingFloatChannel, default=0, const [0,31]}
       spine_01_ctrl.Rotation.Z {type=MovieSceneScriptingFloatChannel, default=0, const [0,31]}
       spine_01_ctrl.Scale.X {type=MovieSceneScriptingFloatChannel, default=1, const [1,31]}
       spine_01_ctrl.Scale.Y {type=MovieSceneScriptingFloatChannel, default=1, const [1,31]}
       spine_01_ctrl.Scale.Z {type=MovieSceneScriptingFloatChannel, default=1, const [1,31]}
       spine_02_ctrl.Location.X {type=MovieSceneScriptingFloatChannel, default=0, const [0,31]}
       spine_02_ctrl.Location.Y {type=MovieSceneScriptingFloatChannel, default=0, const [0,31]}
       spine_02_ctrl.Location.Z {type=MovieSceneScriptingFloatChannel, default=0, const [0,31]}
       spine_02_ctrl.Rotation.X {
        type=MovieSceneScriptingFloatChannel
        default=0
        thinned_from=31
        keys [[0,7.4245],[146,7.4245],[160,7.6359],[217,7.6359],[221,7.623],[226,7.533],[236,7.0672],[255,7.0415],[269,7.0415],[274,7.1065],[282,7.5616],[287,7.6359],[295,7.5143],[311,7.5143],[327,7.4357],[340,7.4245]]
       }
       spine_02_ctrl.Rotation.Y {
        type=MovieSceneScriptingFloatChannel
        default=0
        thinned_from=31
        keys [[0,4.1332],[146,4.1332],[160,2.0592],[217,2.0592],[221,1.7913],[226,-0.0841],[236,-3.0916],[255,-3.2574],[269,-3.2574],[274,-2.5224],[282,2.6229],[287,4.1332],[340,4.1332]]
       }
       spine_02_ctrl.Rotation.Z {
        type=MovieSceneScriptingFloatChannel
        default=0
        thinned_from=31
        keys [[0,0],[146,0],[160,-6.1552],[217,-6.1552],[221,-5.7887],[226,-3.2228],[236,7.2407],[255,7.8176],[269,7.8176],[274,7.3563],[282,4.1274],[287,2.7805],[295,1.181],[311,1.181],[327,0.1476],[340,0]]
       }
       spine_02_ctrl.Scale.X {type=MovieSceneScriptingFloatChannel, default=1, const [1,31]}
       spine_02_ctrl.Scale.Y {type=MovieSceneScriptingFloatChannel, default=1, const [1,31]}
       spine_02_ctrl.Scale.Z {type=MovieSceneScriptingFloatChannel, default=1, const [1,31]}
       spine_03_ctrl.Location.X {type=MovieSceneScriptingFloatChannel, default=0, const [0,32]}
       spine_03_ctrl.Location.Y {type=MovieSceneScriptingFloatChannel, default=0, const [0,32]}
       spine_03_ctrl.Location.Z {type=MovieSceneScriptingFloatChannel, default=0, const [0,32]}
       spine_03_ctrl.Rotation.X {type=MovieSceneScriptingFloatChannel, default=0, const [0,32]}
       spine_03_ctrl.Rotation.Y {type=MovieSceneScriptingFloatChannel, default=0, const [0,32]}
       spine_03_ctrl.Rotation.Z {type=MovieSceneScriptingFloatChannel, default=0, const [0,32]}
       spine_03_ctrl.Scale.X {type=MovieSceneScriptingFloatChannel, default=1, const [1,32]}
       spine_03_ctrl.Scale.Y {type=MovieSceneScriptingFloatChannel, default=1, const [1,32]}
       spine_03_ctrl.Scale.Z {type=MovieSceneScriptingFloatChannel, default=1, const [1,32]}
       spine_04_ctrl.Location.X {type=MovieSceneScriptingFloatChannel, default=0, const [0,31]}
       spine_04_ctrl.Location.Y {type=MovieSceneScriptingFloatChannel, default=0, const [0,31]}
       spine_04_ctrl.Location.Z {type=MovieSceneScriptingFloatChannel, default=0, const [0,31]}
       spine_04_ctrl.Rotation.X {type=MovieSceneScriptingFloatChannel, default=0, thinned_from=31, keys [[0,9.803],[282,9.803],[340,9.803]]}
       spine_04_ctrl.Rotation.Y {
        type=MovieSceneScriptingFloatChannel
        default=0
        thinned_from=31
        keys [[0,0.7539],[282,0.7539],[287,2.9622],[295,2.536],[311,2.536],[327,0.9767],[340,0.7539]]
       }
       spine_04_ctrl.Rotation.Z {
        type=MovieSceneScriptingFloatChannel
        default=0
        thinned_from=31
        keys [[0,9.078],[282,9.078],[287,9.46],[295,9.3862],[311,9.3862],[327,9.1165],[340,9.078]]
       }
       spine_04_ctrl.Scale.X {type=MovieSceneScriptingFloatChannel, default=1, const [1,31]}
       spine_04_ctrl.Scale.Y {type=MovieSceneScriptingFloatChannel, default=1, const [1,31]}
       spine_04_ctrl.Scale.Z {type=MovieSceneScriptingFloatChannel, default=1, const [1,31]}
       head_ctrl.Location.X {type=MovieSceneScriptingFloatChannel, default=0, thinned_from=32, keys [[0,0],[340,0]]}
       head_ctrl.Location.Y {type=MovieSceneScriptingFloatChannel, default=0, thinned_from=32, keys [[0,-0.0875],[340,-0.0875]]}
       head_ctrl.Location.Z {type=MovieSceneScriptingFloatChannel, default=0, thinned_from=32, keys [[0,-0.3874],[340,-0.3874]]}
       head_ctrl.Rotation.X {
        type=MovieSceneScriptingFloatChannel
        default=0
        thinned_from=32
        keys [[0,3.8354],[84,3.8354],[92,3.2137],[112,3.2137],[122,1.8226],[135,1.8226],[139,1.7426],[142,1.3154],[146,-0.0854],[160,-0.8896],[175,-0.8896],[178,-0.9812],[185,-0.9489],[195,0.3508],[217,0.3508],[221,0.2072],[226,-0.2234],[236,-0.8351],[255,-0.8689],[269,-0.8689],[274,0.0624],[282,0.2565],[287,0.4159],[295,0.7833],[311,0.7833],[327,3.4539],[340,3.8354]]
       }
       head_ctrl.Rotation.Y {
        type=MovieSceneScriptingFloatChannel
        default=0
        thinned_from=32
        keys [[0,-8.3104],[46,-8.3104],[60,-8.0797],[84,-8.0797],[92,-8.0344],[112,-8.0344],[122,-8.0158],[135,-8.0158],[139,-7.977],[142,-7.3689],[146,-6.98],[160,-6.4826],[175,-6.4826],[178,-6.9651],[185,-5.189],[195,-6.0271],[217,-6.0271],[221,-5.5547],[226,-4.1376],[236,-3.4611],[255,-3.4238],[269,-3.4238],[274,-1.9473],[282,2.4823],[287,1.1881],[295,-0.2195],[311,-0.2195],[327,-7.2991],[340,-8.3104]]
       }
       head_ctrl.Rotation.Z {
        type=MovieSceneScriptingFloatChannel
        default=0
        thinned_from=32
        keys [[0,-17.7188],[46,-17.7188],[60,-16.2364],[84,-16.2364],[92,-18.8884],[112,-18.8884],[122,-23.7992],[135,-23.7992],[139,-25.8414],[142,-33.6009],[146,-47.2087],[160,-51.4447],[175,-51.4447],[178,-49.6795],[185,-25.1176],[195,-14.1812],[217,-14.1812],[221,-7.8414],[226,11.1777],[236,23.8247],[255,24.522],[269,24.522],[274,18.1027],[282,-1.155],[287,-6.7936],[295,-8.27],[311,-8.27],[327,-16.5377],[340,-17.7188]]
       }
       head_ctrl.Scale.X {type=MovieSceneScriptingFloatChannel, default=1, const [1,32]}
       head_ctrl.Scale.Y {type=MovieSceneScriptingFloatChannel, default=1, const [1,32]}
       head_ctrl.Scale.Z {type=MovieSceneScriptingFloatChannel, default=1, const [1,32]}
       clavicle_l_ctrl.Location.X {type=MovieSceneScriptingFloatChannel, default=0, thinned_from=27, keys [[0,0],[340,0]]}
       clavicle_l_ctrl.Location.Y {type=MovieSceneScriptingFloatChannel, default=0, thinned_from=27, keys [[0,0],[340,0]]}
       clavicle_l_ctrl.Location.Z {type=MovieSceneScriptingFloatChannel, default=0, thinned_from=27, keys [[0,0],[340,0]]}
       clavicle_l_ctrl.Rotation.X {type=MovieSceneScriptingFloatChannel, default=0, thinned_from=27, keys [[0,0],[175,0],[178,-0.1574],[311,-0.1574],[327,-0.0197],[340,0]]}
       clavicle_l_ctrl.Rotation.Y {
        type=MovieSceneScriptingFloatChannel
        default=0
        thinned_from=27
        keys [[0,-3.3908],[139,-3.3908],[142,-3.2549],[146,-3.0416],[175,-3.0416],[178,-0.549],[185,-2.048],[195,-0.9992],[311,-0.9992],[327,-3.0918],[340,-3.3908]]
       }
       clavicle_l_ctrl.Rotation.Z {
        type=MovieSceneScriptingFloatChannel
        default=0
        thinned_from=27
        keys [[0,13.1328],[175,13.1282],[178,15.7878],[311,15.7878],[327,13.4647],[340,13.1328]]
       }
       clavicle_l_ctrl.Scale.X {type=MovieSceneScriptingFloatChannel, default=1, const [1,27]}
       clavicle_l_ctrl.Scale.Y {type=MovieSceneScriptingFloatChannel, default=1, const [1,27]}
       clavicle_l_ctrl.Scale.Z {type=MovieSceneScriptingFloatChannel, default=1, const [1,27]}
       clavicle_r_ctrl.Location.X {type=MovieSceneScriptingFloatChannel, default=0, thinned_from=28, keys [[0,0],[340,0]]}
       clavicle_r_ctrl.Location.Y {type=MovieSceneScriptingFloatChannel, default=0, thinned_from=28, keys [[0,0],[340,0]]}
       clavicle_r_ctrl.Location.Z {type=MovieSceneScriptingFloatChannel, default=0, thinned_from=28, keys [[0,0],[340,0]]}
       clavicle_r_ctrl.Rotation.X {type=MovieSceneScriptingFloatChannel, default=0, thinned_from=28, keys [[0,-0.1574],[340,-0.1574]]}
       clavicle_r_ctrl.Rotation.Y {
        type=MovieSceneScriptingFloatChannel
        default=0
        thinned_from=28
        keys [[0,-1.3483],[175,-1.3483],[178,-0.8982],[185,-2.3971],[195,-1.3483],[340,-1.3483]]
       }
       clavicle_r_ctrl.Rotation.Z {type=MovieSceneScriptingFloatChannel, default=0, thinned_from=28, keys [[0,15.7924],[340,15.7924]]}
       clavicle_r_ctrl.Scale.X {type=MovieSceneScriptingFloatChannel, default=1, const [1,28]}
       clavicle_r_ctrl.Scale.Y {type=MovieSceneScriptingFloatChannel, default=1, const [1,28]}
       clavicle_r_ctrl.Scale.Z {type=MovieSceneScriptingFloatChannel, default=1, const [1,28]}
       neck_01_ctrl.Location.X {type=MovieSceneScriptingFloatChannel, default=0, const [0,31]}
       neck_01_ctrl.Location.Y {type=MovieSceneScriptingFloatChannel, default=0, thinned_from=31, keys [[0,0],[340,0]]}
       neck_01_ctrl.Location.Z {type=MovieSceneScriptingFloatChannel, default=0, thinned_from=31, keys [[0,0],[340,0]]}
       neck_01_ctrl.Rotation.X {
        type=MovieSceneScriptingFloatChannel
        default=0
        thinned_from=31
        keys [[0,-2.6694],[25,-2.8173],[46,-2.8666],[84,-2.8666],[92,-2.9102],[217,-2.9102],[221,-2.8642],[226,-2.5422],[236,-2.4259],[255,-2.4195],[282,-2.4195],[287,-2.6683],[340,-2.6694]]
       }
       neck_01_ctrl.Rotation.Y {
        type=MovieSceneScriptingFloatChannel
        default=0
        thinned_from=31
        keys [[0,2.0231],[25,1.1889],[46,0.9108],[84,0.9108],[92,2.4851],[217,2.4851],[221,2.5301],[226,2.845],[236,2.9587],[282,2.965],[287,5.0915],[295,5.1009],[311,5.1009],[327,2.4078],[340,2.0231]]
       }
       neck_01_ctrl.Rotation.Z {
        type=MovieSceneScriptingFloatChannel
        default=0
        thinned_from=31
        keys [[0,-1.7035],[25,-7.273],[46,-9.1294],[84,-9.1294],[92,-12.0215],[217,-12.0215],[221,-11.0564],[226,-4.3007],[236,-1.8616],[255,-1.7271],[282,-1.7271],[287,-1.7036],[340,-1.7035]]
       }
       neck_01_ctrl.Scale.X {type=MovieSceneScriptingFloatChannel, default=1, const [1,31]}
       neck_01_ctrl.Scale.Y {type=MovieSceneScriptingFloatChannel, default=1, const [1,31]}
       neck_01_ctrl.Scale.Z {type=MovieSceneScriptingFloatChannel, default=1, const [1,31]}
       neck_02_ctrl.Location.X {type=MovieSceneScriptingFloatChannel, default=0, const [0,31]}
       neck_02_ctrl.Location.Y {type=MovieSceneScriptingFloatChannel, default=0, thinned_from=31, keys [[0,0.1375],[25,0.0344],[46,0],[311,0],[327,0.1203],[340,0.1375]]}
       neck_02_ctrl.Location.Z {type=MovieSceneScriptingFloatChannel, default=0, thinned_from=31, keys [[0,0.4106],[25,0.1027],[46,0],[311,0],[327,0.3593],[340,0.4106]]}
       neck_02_ctrl.Rotation.X {type=MovieSceneScriptingFloatChannel, default=0, const [0,31]}
       neck_02_ctrl.Rotation.Y {type=MovieSceneScriptingFloatChannel, default=0, thinned_from=31, keys [[0,-6.2034],[25,-1.5508],[46,0],[311,0],[327,-5.4279],[340,-6.2034]]}
       neck_02_ctrl.Rotation.Z {type=MovieSceneScriptingFloatChannel, default=0, const [0,31]}
       neck_02_ctrl.Scale.X {type=MovieSceneScriptingFloatChannel, default=1, const [1,31]}
       neck_02_ctrl.Scale.Y {type=MovieSceneScriptingFloatChannel, default=1, const [1,31]}
       neck_02_ctrl.Scale.Z {type=MovieSceneScriptingFloatChannel, default=1, const [1,31]}
       head_ik_ctrl.Location.X {
        type=MovieSceneScriptingFloatChannel
        default=0
        thinned_from=24
        keys [[0,0.2651],[25,0.2445],[46,0.2377],[84,0.2377],[92,0.0818],[146,0.0818],[160,0.0489],[217,0.0489],[236,-0.299],[255,-0.2998],[269,-0.2998],[295,-0.7258],[311,-0.7258],[327,0.4953],[340,0.6698]]
       }
       head_ik_ctrl.Location.Y {
        type=MovieSceneScriptingFloatChannel
        default=0
        thinned_from=24
        keys [[0,-0.2115],[25,-0.4135],[46,-0.4808],[84,-0.4808],[92,-0.4731],[146,-0.4731],[160,-0.4797],[217,-0.4797],[236,-0.5421],[255,-0.5469],[269,-0.5469],[295,-0.6149],[311,-0.6149],[327,-1.1941],[340,-1.2768]]
       }
       head_ik_ctrl.Location.Z {
        type=MovieSceneScriptingFloatChannel
        default=0
        thinned_from=24
        keys [[0,0.2198],[25,-0.031],[46,-0.1145],[84,-0.1145],[92,-0.1413],[146,-0.1413],[160,-0.143],[217,-0.143],[236,-0.1655],[255,-0.1646],[269,-0.1646],[295,-0.2617],[311,-0.2617],[327,0.595],[340,0.7174]]
       }
       head_ik_ctrl.Rotation.X {
        type=MovieSceneScriptingFloatChannel
        default=0
        thinned_from=24
        keys [[0,3.8354],[84,3.8354],[92,3.2137],[112,3.2137],[122,1.8226],[135,1.8226],[139,1.7426],[146,-0.0854],[160,-0.8896],[175,-0.8896],[178,-0.9812],[185,-0.9489],[205,0.3508],[217,0.3508],[236,-0.8351],[255,-0.8689],[269,-0.8689],[295,0.7833],[311,0.7833],[327,0.1538],[340,0.0639]]
       }
       head_ik_ctrl.Rotation.Y {
        type=MovieSceneScriptingFloatChannel
        default=0
        thinned_from=24
        keys [[0,-8.3104],[46,-8.3104],[60,-8.0797],[84,-8.0797],[92,-8.0344],[112,-8.0344],[122,-8.0158],[135,-8.0158],[139,-7.977],[146,-6.98],[160,-6.4826],[175,-6.4826],[178,-6.9651],[185,-5.189],[205,-6.0271],[217,-6.0271],[236,-3.4611],[255,-3.4238],[269,-3.4238],[295,-0.2195],[311,-0.2195],[327,-3.4699],[340,-3.9343]]
       }
       head_ik_ctrl.Rotation.Z {
        type=MovieSceneScriptingFloatChannel
        default=0
        thinned_from=24
        keys [[0,-17.7188],[46,-17.7188],[60,-16.2364],[84,-16.2364],[92,-18.8884],[112,-18.8884],[122,-23.7992],[135,-23.7992],[139,-25.8414],[146,-47.2087],[160,-51.4447],[175,-51.4447],[178,-49.6795],[185,-25.1176],[205,-14.1812],[217,-14.1812],[236,23.8247],[255,24.522],[269,24.522],[295,-8.27],[311,-8.27],[327,-16.2512],[340,-17.3914]]
       }
       head_ik_ctrl.Scale.X {type=MovieSceneScriptingFloatChannel, default=1, const [1,24]}
       head_ik_ctrl.Scale.Y {type=MovieSceneScriptingFloatChannel, default=1, const [1,24]}
       head_ik_ctrl.Scale.Z {type=MovieSceneScriptingFloatChannel, default=1, const [1,24]}
       chest_ctrl.Location.X {
        type=MovieSceneScriptingFloatChannel
        default=0
        thinned_from=26
        keys [[0,-0.5014],[146,-0.5014],[160,-0.2223],[217,-0.2223],[221,0],[236,0.3501],[255,0.3682],[269,0.3682],[274,0],[295,-0.506],[311,-0.506],[327,-0.0633],[340,0]]
       }
       chest_ctrl.Location.Y {
        type=MovieSceneScriptingFloatChannel
        default=0
        thinned_from=26
        keys [[0,0.9008],[146,0.9008],[160,0.9518],[217,0.9518],[221,0.9008],[255,0.9065],[269,0.9065],[274,0.9008],[340,0.9008]]
       }
       chest_ctrl.Location.Z {
        type=MovieSceneScriptingFloatChannel
        default=0
        thinned_from=26
        keys [[0,0.0119],[146,0.0119],[160,0.0247],[217,0.0247],[221,0.03],[236,0.0211],[255,0.0201],[269,0.0201],[274,0.03],[295,0.0116],[311,0.0116],[327,0.0277],[340,0.03]]
       }
       chest_ctrl.Rotation.X {
        type=MovieSceneScriptingFloatChannel
        default=0
        thinned_from=26
        keys [[0,7.4245],[146,7.4245],[160,7.6359],[217,7.6359],[221,7.4245],[236,7.0672],[255,7.0415],[269,7.0415],[274,7.4245],[295,7.5143],[311,7.5143],[327,7.4357],[340,7.4245]]
       }
       chest_ctrl.Rotation.Y {
        type=MovieSceneScriptingFloatChannel
        default=0
        thinned_from=26
        keys [[0,4.1332],[146,4.1332],[160,2.0592],[217,2.0592],[221,0],[236,-3.0916],[255,-3.2574],[269,-3.2574],[274,0],[295,4.1332],[311,4.1332],[327,0.5167],[340,0]]
       }
       chest_ctrl.Rotation.Z {
        type=MovieSceneScriptingFloatChannel
        default=0
        thinned_from=26
        keys [[0,0],[146,0],[160,-6.1552],[217,-6.1552],[221,0],[236,7.2407],[255,7.8176],[269,7.8176],[274,0],[295,1.181],[311,1.181],[327,0.1476],[340,0]]
       }
       chest_ctrl.Scale.X {type=MovieSceneScriptingFloatChannel, default=1, const [1,26]}
       chest_ctrl.Scale.Y {type=MovieSceneScriptingFloatChannel, default=1, const [1,26]}
       chest_ctrl.Scale.Z {type=MovieSceneScriptingFloatChannel, default=1, const [1,26]}
       chest_tan_ctrl.Location.X {
        type=MovieSceneScriptingFloatChannel
        default=0
        thinned_from=26
        keys [[0,0.3954],[146,0.3954],[160,0.264],[217,0.264],[221,0],[236,-0.3745],[255,-0.3965],[269,-0.3965],[274,0],[295,0.3826],[311,0.3826],[327,0.0478],[340,0]]
       }
       chest_tan_ctrl.Location.Y {
        type=MovieSceneScriptingFloatChannel
        default=0.0049
        thinned_from=26
        keys [[0,-0.6969],[146,-0.6969],[160,-0.7141],[217,-0.7141],[221,-0.6987],[236,-0.6588],[255,-0.6554],[269,-0.6554],[274,-0.6987],[295,-0.7053],[311,-0.7053],[327,-0.6995],[340,-0.6987]]
       }
       chest_tan_ctrl.Location.Z {
        type=MovieSceneScriptingFloatChannel
        default=1.2539
        thinned_from=26
        keys [[0,1.1131],[146,1.1131],[160,1.1169],[217,1.1169],[221,1.1272],[236,1.1239],[269,1.1231],[274,1.1272],[295,1.1119],[311,1.1119],[327,1.1253],[340,1.1272]]
       }
       chest_tan_ctrl.Rotation.X {
        type=MovieSceneScriptingFloatChannel
        default=0.6533
        thinned_from=26
        keys [[0,0.6533],[274,0.6533],[295,0.7241],[311,0.7241],[327,0.6621],[340,0.6533]]
       }
       chest_tan_ctrl.Rotation.Y {type=MovieSceneScriptingFloatChannel, default=0, thinned_from=26, keys [[0,0],[274,0],[295,4.2511],[311,4.2511],[327,0.5314],[340,0]]}
       chest_tan_ctrl.Rotation.Z {
        type=MovieSceneScriptingFloatChannel
        default=0
        thinned_from=26
        keys [[0,0],[60,0],[84,0.0015],[92,0],[274,0],[295,0.6361],[311,0.6361],[327,0.0795],[340,0]]
       }
       chest_tan_ctrl.Scale.X {type=MovieSceneScriptingFloatChannel, default=1, const [1,26]}
       chest_tan_ctrl.Scale.Y {type=MovieSceneScriptingFloatChannel, default=1, const [1,26]}
       chest_tan_ctrl.Scale.Z {type=MovieSceneScriptingFloatChannel, default=1, const [1,26]}
      ]
     }]
   }
   {type=MovieSceneControlRigParameterTrack, display_name=MR_Biped_v02, sections [{type=MovieSceneControlRigParameterSection}]}
  ]
 }]
display_rate {denominator=1, numerator=30}
tick_resolution {denominator=1, numerator=24000}
tracks []
